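# Control Rig for the F_MED_TealMink_Pelt cosmetic body that runs a shared DeformRig function, then procedurally poses left and right hip strap bones.
# Spherical pose readers on pelvis and thighs produce alphas that, with crouch, DBNO, front-end and emote flags, weight corrective offsets on base and lock bones.
# Those alphas also blend parent constraints, stored control poses and control offsets, with stretchy B-spline IK driving front and back strap chains.
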ControlRigBlueprint /BRCosmetics/Characters/Player/Female/Medium/Bodies/F_MED_TealMink_Pelt/Meshes/F_MED_TealMink_Pelt_CtrlRig
graphs [
 RigVMFunctionLibrary {links [], nodes []}
 RigVMFunction_Sequence_1_ContainedGraph {nodes [
   n1 = RigVMFunctionEntryNode()
   n2 = &RigVMFunction_Sequence(ExecuteContext=@n1.ExecuteContext)
   n3 = &RigVMFunction_Sequence(ExecuteContext=@n2.B)
   RigVMFunctionReturnNode(B=@n3.A, C=@n3.B, A=@n2.A)
  ]}
 RigVMFunction_Sequence_2_1_ContainedGraph {nodes [
   n1 = RigVMFunctionEntryNode()
   n2 = &RigVMFunction_Sequence(ExecuteContext=@n1.ExecuteContext)
   n3 = &RigVMFunction_Sequence(ExecuteContext=@n2.B)
   n4 = &RigVMFunction_Sequence(ExecuteContext=@n3.B)
   RigVMFunctionReturnNode(A=@n2.A, B=@n3.A, C=@n4.A, D=@n4.B)
  ]}
 RigVMFunction_Sequence_3_1_ContainedGraph {nodes [
   n1 = RigVMFunctionEntryNode()
   n2 = &RigVMFunction_Sequence(ExecuteContext=@n1.ExecuteContext)
   n3 = &RigVMFunction_Sequence(ExecuteContext=@n2.B)
   n4 = &RigVMFunction_Sequence(ExecuteContext=@n3.B)
   RigVMFunctionReturnNode(C=@n4.A, D=@n4.B, A=@n2.A, B=@n3.A)
  ]}
 RigVMFunction_Sequence_3_1_ContainedGraph {nodes [
   n1 = RigVMFunctionEntryNode()
   n2 = &RigVMFunction_Sequence(ExecuteContext=@n1.ExecuteContext)
   n3 = &RigVMFunction_Sequence(ExecuteContext=@n2.B)
   n4 = &RigVMFunction_Sequence(ExecuteContext=@n3.B)
   RigVMFunctionReturnNode(C=@n4.A, D=@n4.B, A=@n2.A, B=@n3.A)
  ]}
 RigVMFunction_Sequence_3_4_ContainedGraph {nodes [
   n1 = RigVMFunctionEntryNode()
   n2 = &RigVMFunction_Sequence(ExecuteContext=@n1.ExecuteContext)
   n3 = &RigVMFunction_Sequence(ExecuteContext=@n2.B)
   RigVMFunctionReturnNode(B=@n3.A, C=@n3.B, A=@n2.A)
  ]}
 RigVMFunction_Sequence_3_4_ContainedGraph {nodes [
   n1 = RigVMFunctionEntryNode()
   n2 = &RigVMFunction_Sequence(ExecuteContext=@n1.ExecuteContext)
   n3 = &RigVMFunction_Sequence(ExecuteContext=@n2.B)
   RigVMFunctionReturnNode(B=@n3.A, C=@n3.B, A=@n2.A)
  ]}
 RigVMModel {nodes [
   RigUnit_BeginExecution = &RigUnit_BeginExecution()
   FN_DeformRig_FNC = RigVMFunctionReferenceNode(
    DeformRig_ Curve="(deformRig_l_upperarm_up_alpha_72_#1=False,deformRig_r_upperarm_up_alpha_73_#2=False,deformRig_l_upperarm_fr_alpha_74_#3=False,deformRig_r_upperarm_fr_alpha_75_CB51B8 ...(truncated)"
    DeformRig_Bool="(upperarm_1_#4=True,elbow_3_#5=True,wrist_7_#6=True,groin_8_#7=True,glute_17_#8=True,knee_9_40C408324CA261 ...(truncated)"
    DeformRig_Debug=(upperarm_debug_18_#4=False,wrist_debug_20_#6=False,lat_debug_26_#9=False,groinGlute_debug_27_#7=False)
    DeformRig_Value="(upperarm_up_18_#10=(Rotation=(X=0,Y=0,Z=0,W=1),Translation=(X=0,Y=0,Z=8),Scale3D=(X=0.7,Y=1,Z=1)),upperarm_fwd_fr_19_#11=(Rotation=(X=0,Y=0,Z=0,W=1),Translation=(X=0,Y=10,Z=0),Scale3D=(X=0.5,Y=1,Z= ...(truncated)"
    ExecuteContext=@RigUnit_BeginExecution.ExecutePin
   )
   > n1 = &RigUnit_SphericalPoseReader(
    ActiveRegionScaleFactors=(PositiveWidth=0.15,NegativeWidth=0.15,PositiveHeight=0.15,NegativeHeight=0.15)
    ActiveRegionSize=0.1
    Debug=(bDrawDebug=true,DebugScale=25,DebugSegments=20,DebugThickness=0.25)
    DriverAxis=(X=0,Y=-1,Z=0)
    DriverItem=(Type=Bone,Name="pelvis")
    FalloffRegionScaleFactors=(PositiveWidth=1,NegativeWidth=1,PositiveHeight=1,NegativeHeight=1)
    FalloffSize=0.3
    FlipHeightScaling=false
    FlipWidthScaling=true
    OptionalParentItem=(Type=Bone)
    RotationOffset=(X=-110,Y=0,Z=0)
   )
   > RigVMVariableNode(Value=@n1.OutputParam) {variable=Dyn_pelvis_rotate_alfa}
   > RigVMFunction_Sequence_2_1 = RigVMAggregateNode()
   n2 = &RigVMFunction_MathBoolToFloat(Value=RigVMVariableNode() {variable=is_crouch}.Value)
   n3 = RigVMVariableNode() {variable=Dyn_move_alfa_l}
   n4 = /Script/RigVM.RigVMFunction_MathFloatMake(Value=&RigVMFunction_MathDoubleClamp(Value=&RigVMFunction_MathDoubleSub(A=@n2.Result, B=@n3.Value).Result).Result) {is=struct RigVMUnitNode}
   n5 = &RigVMFunction_MathDoubleMul(A=0, B=@n4.Value)
   n6 = RigVMVariableNode() {variable=spr_thigh_l_up_alpha}
   n7 = &RigVMFunction_MathDoubleMul(A=20, B=&RigVMFunction_MathDoubleMul(A=&RigVMFunction_MathDoubleSub(A=1, B=@n4.Value).Result, B=@n6.Value).Result)
   n8 = &RigVMFunction_MathDoubleMul(A=0, B=@n4.Value)
   n9 = &RigVMFunction_MathDoubleMul(A=0, B=@n6.Value)
   n10 = &RigVMFunction_MathDoubleMul(A=10, B=@n4.Value)
   n11 = &RigVMFunction_MathDoubleMul(A=10, B=@n6.Value)
   &RigUnit_ModifyTransforms(
    ItemToModify=((Item=(Type=Bone,Name="dyn_straps_base_l"),Transform=(Rotation=(X=0,Y=0,Z=0,W=1),Translation=(X=0,Y=0,Z=0),Scale3D=(X=1,Y=1,Z=1))))
    Weight=0
    WeightMinimum=0
    ItemToModify.0.Transform.Rotation=&RigVMFunction_MathQuaternionFromEuler(
     Euler=(X=0,Y=0,Z=-20)
     Euler.X=&RigVMFunction_MathDoubleAdd(A=@n5.Result, B=@n7.Result).Result
     Euler.Y=&RigVMFunction_MathDoubleAdd(A=@n8.Result, B=@n9.Result).Result
     Euler.Z=&RigVMFunction_MathDoubleAdd(A=@n10.Result, B=@n11.Result).Result
    ).Result
    ExecutePin=@RigVMFunction_Sequence_2_1.A
   )
   > &RigUnit_ParentConstraint(
    AdvancedSettings=(RotationOrderForFilter=XZY)
    Child=(Type=Null,Name="dyn_straps_front_1_l_null")
    Parents=((Item=(Type=Bone,Name="dyn_straps_front_base_l"),Weight=1))
    Weight=1
   )
   > &RigUnit_ParentConstraint(
    AdvancedSettings=(RotationOrderForFilter=XZY)
    Child=(Type=Null,Name="dyn_straps_back_1_l_null")
    Parents=((Item=(Type=Bone,Name="dyn_straps_back_base_l"),Weight=1))
    Weight=1
   )
   > n12 = &RigUnit_SphericalPoseReader(
    ActiveRegionScaleFactors=(PositiveWidth=0.25,NegativeWidth=2,PositiveHeight=2,NegativeHeight=2)
    ActiveRegionSize=0.1
    Debug=(bDrawDebug=true,DebugScale=25,DebugSegments=20,DebugThickness=0.25)
    DriverAxis=(X=-1,Y=0,Z=0)
    DriverItem=(Type=Bone,Name="thigh_l")
    FalloffRegionScaleFactors=(PositiveWidth=1,NegativeWidth=1,PositiveHeight=1,NegativeHeight=1)
    FalloffSize=0.7
    FlipHeightScaling=false
    FlipWidthScaling=true
    OptionalParentItem=(Type=Bone)
    RotationOffset=(X=120,Y=0,Z=-25)
   )
   > RigVMVariableNode(Value=@n12.OutputParam) {variable=spr_thigh_l_up_alpha}
   > n13 = &RigUnit_SphericalPoseReader(
    ActiveRegionScaleFactors=(PositiveWidth=1,NegativeWidth=1,PositiveHeight=1,NegativeHeight=1)
    ActiveRegionSize=0.05
    Debug=(bDrawDebug=false,DebugScale=25,DebugSegments=20,DebugThickness=0.25)
    DriverAxis=(X=-1,Y=0,Z=0)
    DriverItem=(Type=Bone,Name="thigh_l")
    FalloffRegionScaleFactors=(PositiveWidth=1,NegativeWidth=1,PositiveHeight=1,NegativeHeight=1)
    FalloffSize=0.09
    OptionalParentItem=(Type=Bone)
    RotationOffset=(X=165,Y=-80,Z=0)
   )
   > RigVMVariableNode(Value=@n13.OutputParam) {variable=spr_thigh_l_slight_back_out_apha}
   n14 = &RigVMFunction_MathBoolToFloat(Value=RigVMVariableNode() {variable=is_crouch}.Value)
   n15 = &RigVMFunction_MathBoolToFloat(Value=RigVMVariableNode() {variable=is_DBNO}.Value)
   n16 = &RigVMFunction_MathDoubleAdd(
    A=&RigVMFunction_MathDoubleMul(A=-20, B=@n14.Result).Result
    B=&RigVMFunction_MathDoubleMul(A=0, B=@n15.Result).Result
   )
   n17 = &RigVMFunction_MathDoubleAdd(
    A=&RigVMFunction_MathDoubleMul(A=-10, B=@n14.Result).Result
    B=&RigVMFunction_MathDoubleMul(A=-10, B=@n15.Result).Result
   )
   n18 = &RigVMFunction_MathDoubleAdd(
    A=&RigVMFunction_MathDoubleMul(A=30, B=@n14.Result).Result
    B=&RigVMFunction_MathDoubleMul(A=-20, B=@n15.Result).Result
   )
   > &RigUnit_ModifyTransforms(
    ItemToModify=((Item=(Type=Bone,Name="dyn_straps_base_l"),Transform=(Rotation=(X=0,Y=0,Z=0.2588,W=0.9659),Translation=(X=-2,Y=0,Z=0),Scale3D=(X=1,Y=1,Z=1))))
    WeightMinimum=0
    ItemToModify.0.Transform.Translation.X=&RigVMFunction_MathDoubleMul(A=-2, B=@n14.Result).Result
    ItemToModify.0.Transform.Rotation=&RigVMFunction_MathQuaternionFromEuler(
     Euler=(X=0,Y=0,Z=30)
     Euler.Y=&RigVMFunction_MathDoubleAdd(A=20, B=@n16.Result).Result
     Euler.X=&RigVMFunction_MathDoubleAdd(A=-20, B=@n17.Result).Result
     Euler.Z=&RigVMFunction_MathDoubleAdd(A=0, B=@n18.Result).Result
    ).Result
    Weight=@n12.OutputParam
   )
   n19 = &RigVMFunction_MathFloatMake(Value=&RigVMFunction_MathBoolToFloat(Value=RigVMVariableNode() {variable=is_front_end}.Value).Result)
   n20 = RigVMVariableNode() {variable=is_emote}
   n21 = &RigVMFunction_MathDoubleSub(A=1, B=&RigVMFunction_MathFloatMake(Value=&RigVMFunction_MathBoolToFloat(Value=@n20.Value).Result).Value)
   > &RigUnit_ModifyTransforms(
    ItemToModify=((Item=(Type=Bone,Name="dyn_straps_base_l"),Transform=(Rotation=(X=0,Y=0,Z=0,W=1),Translation=(X=0,Y=0,Z=-3),Scale3D=(X=1,Y=1,Z=1))))
    WeightMinimum=0
    Weight=&RigVMFunction_MathDoubleMul(
     B=&RigVMFunction_MathDoubleMul(A=@n19.Value, B=@n21.Result).Result
     A=&RigVMFunction_MathDoubleSub(A=1, B=@n12.OutputParam).Result
    ).Result
   )
   > &RigUnit_ParentConstraint(
    AdvancedSettings=(RotationOrderForFilter=XZY)
    Child=(Type=Null,Name="dyn_straps_Main_l_null")
    Parents=((Item=(Type=Bone,Name="dyn_straps_end_l"),Weight=1),(Item=(Type=Bone,Name="thigh_l"),Weight=0))
    Weight=1
   )
   > n22 = &RigUnit_SphericalPoseReader(
    ActiveRegionScaleFactors=(PositiveWidth=0.5,NegativeWidth=2,PositiveHeight=0.5,NegativeHeight=0.5)
    ActiveRegionSize=0.1
    Debug=(bDrawDebug=false,DebugScale=25,DebugSegments=20,DebugThickness=0.25)
    DriverAxis=(X=-1,Y=0,Z=0)
    DriverItem=(Type=Bone,Name="thigh_l")
    FalloffRegionScaleFactors=(PositiveWidth=1,NegativeWidth=1,PositiveHeight=1,NegativeHeight=1)
    FalloffSize=0.3
    FlipHeightScaling=false
    FlipWidthScaling=true
    OptionalParentItem=(Type=Bone)
    RotationOffset=(X=130,Y=0,Z=45)
   )
   > n23 = RigVMVariableNode(Value=@n22.OutputParam) {variable=Dyn_move_alfa_l}
   n24 = &RigVMFunction_Sequence(ExecuteContext=@n23.ExecuteContext)
   RigVMFunction_Sequence_3_2 = RigVMAggregateNode(ExecuteContext=@n24.B)
   &RigUnit_ParentConstraint(
    AdvancedSettings=(InterpolationType=Average,RotationOrderForFilter=XZY)
    Child=(Type=Control,Name="dyn_straps_front_1_l_ctrl")
    Parents=((Item=(Type=Bone,Name="dyn_straps_front_base_l"),Weight=1))
    Weight=1
    ExecutePin=@RigVMFunction_Sequence_3_2.A
   )
   > &RigUnit_ParentConstraint(
    AdvancedSettings=(InterpolationType=Average,RotationOrderForFilter=XZY)
    Child=(Type=Control,Name="dyn_straps_back_1_l_ctrl")
    Parents=((Item=(Type=Bone,Name="dyn_straps_back_base_l"),Weight=1))
    Weight=1
   )
   n25 = RigVMVariableNode() {variable=spr_thigh_l_up_alpha}
   n26 = &RigVMFunction_MathDoubleSub(A=1, B=@n25.Value)
   &RigUnit_ParentConstraint(
    AdvancedSettings=(RotationOrderForFilter=XZY)
    Child=(Type=Null,Name="dyn_straps_front_3_l_null")
    Parents=((Item=(Type=Bone,Name="thigh_l"),Weight=0.15),(Item=(Type=Control,Name="dyn_straps_Main_l_ctrl"),Weight=0.85),(Item=(Type=Bone,Name="deform_groin_l"),Weight=0.25))
    ExecutePin=@RigVMFunction_Sequence_3_2.B
    Weight=@n26.Result
   )
   > &RigUnit_ParentConstraint(
    AdvancedSettings=(RotationOrderForFilter=XZY)
    Child=(Type=Null,Name="dyn_straps_front_5_l_null")
    Parents=((Item=(Type=Bone,Name="thigh_l"),Weight=0.25),(Item=(Type=Control,Name="dyn_straps_Main_l_ctrl"),Weight=0.75))
    Weight=@n26.Result
   )
   > &RigUnit_ParentConstraint(
    AdvancedSettings=(RotationOrderForFilter=XZY)
    Child=(Type=Null,Name="dyn_straps_back_3_l_null")
    Parents=((Item=(Type=Bone,Name="thigh_l"),Weight=0.5),(Item=(Type=Control,Name="dyn_straps_Main_l_ctrl"),Weight=0.25),(Item=(Type=Control,Name="dyn_straps_back_1_l_ctrl"),Weight=0.25))
    Weight=@n26.Result
   )
   > &RigUnit_ParentConstraint(
    AdvancedSettings=(RotationOrderForFilter=XZY)
    Child=(Type=Null,Name="dyn_straps_back_5_l_null")
    Parents=((Item=(Type=Bone,Name="deform_glute_l"),Weight=0.15),(Item=(Type=Control,Name="dyn_straps_Main_l_ctrl"),Weight=0.85))
    Weight=@n26.Result
   )
   > &RigUnit_ParentConstraint(
    AdvancedSettings=(RotationOrderForFilter=XZY)
    Child=(Type=Null,Name="dyn_straps_front_3_l_null")
    Parents=((Item=(Type=Bone,Name="thigh_l"),Weight=0.25),(Item=(Type=Control,Name="dyn_straps_Main_l_ctrl"),Weight=0.5),(Item=(Type=Bone,Name="deform_groin_l"),Weight=0.25))
    Weight=@n25.Value
   )
   > &RigUnit_ParentConstraint(
    AdvancedSettings=(RotationOrderForFilter=XZY)
    Child=(Type=Null,Name="dyn_straps_front_5_l_null")
    Parents=((Item=(Type=Bone,Name="thigh_l"),Weight=0.1),(Item=(Type=Control,Name="dyn_straps_Main_l_ctrl"),Weight=0.9))
    Weight=@n25.Value
   )
   > &RigUnit_ParentConstraint(
    AdvancedSettings=(RotationOrderForFilter=XZY)
    Child=(Type=Null,Name="dyn_straps_back_3_l_null")
    Parents=((Item=(Type=Control,Name="dyn_straps_back_1_l_ctrl"),Weight=0.15),(Item=(Type=Control,Name="dyn_straps_Main_l_ctrl"),Weight=0.85))
    Weight=@n25.Value
   )
   > &RigUnit_ParentConstraint(
    AdvancedSettings=(RotationOrderForFilter=XZY)
    Child=(Type=Null,Name="dyn_straps_back_5_l_null")
    Parents=((Item=(Type=Bone,Name="deform_glute_l"),Weight=0.15),(Item=(Type=Control,Name="dyn_straps_Main_l_ctrl"),Weight=0.85))
    Weight=@n25.Value
   )
   > RigVMFunction_Sequence_3_3 = RigVMAggregateNode()
   n27 = RigVMVariableNode() {variable=spr_thigh_l_up_alpha}
   Subtract = &RigVMFunction_MathDoubleSub(A=1, B=@n27.Value)
   &RigUnit_ModifyTransforms(
    ItemToModify=((Item=(Type=Control,Name="dyn_straps_front_3_l_ctrl"),Transform=(Rotation=(X=0,Y=0,Z=0,W=1),Translation=(X=0,Y=-3,Z=0),Scale3D=(X=1,Y=1,Z=1))))
    WeightMinimum=0
    ExecutePin=@RigVMFunction_Sequence_3_3.C
    Weight=@Subtract.Result
   )
   > &RigUnit_ModifyTransforms(
    ItemToModify=((Item=(Type=Control,Name="dyn_straps_front_5_l_ctrl"),Transform=(Rotation=(X=0,Y=0,Z=0,W=1),Translation=(X=0,Y=-10,Z=-3),Scale3D=(X=1,Y=1,Z=1))))
    WeightMinimum=0
    Weight=@Subtract.Result
   )
   > &RigUnit_ModifyTransforms(
    ItemToModify=((Item=(Type=Control,Name="dyn_straps_back_1_l_ctrl"),Transform=(Rotation=(X=0,Y=0,Z=0,W=1),Translation=(X=0,Y=0,Z=-2),Scale3D=(X=1,Y=1,Z=1))))
    WeightMinimum=0
    Weight=@Subtract.Result
   )
   > &RigUnit_ModifyTransforms(
    ItemToModify=((Item=(Type=Control,Name="dyn_straps_back_3_l_ctrl"),Transform=(Rotation=(X=0,Y=0,Z=0,W=1),Translation=(X=0,Y=-5,Z=-7),Scale3D=(X=1,Y=1,Z=1))))
    WeightMinimum=0
    Weight=@Subtract.Result
   )
   > &RigUnit_ModifyTransforms(
    ItemToModify=((Item=(Type=Control,Name="dyn_straps_back_5_l_ctrl"),Transform=(Rotation=(X=0,Y=0,Z=0,W=1),Translation=(X=0,Y=-5,Z=-5),Scale3D=(X=1,Y=1,Z=1))))
    WeightMinimum=0
    Weight=@Subtract.Result
   )
   Make_Array = RigVMDispatchNode(Values=((Type=Bone,Name="dyn_straps_front_1_l"),(Type=Bone,Name="dyn_straps_front_2_l"),(Type=Bone,Name="dyn_straps_front_3_l"),(Type=Bone,Name="dyn_straps_front_4_l"),(Type=Bone,Name="dyn_straps_front_5_l"),(Type=Bone,Name="dyn_straps_front_6_l"))) {template="DISPATCH_RigVMDispatch_ArrayMake(in Values,out Array)"}
   Make_Array_1 = RigVMDispatchNode(Values=((Type=Control,Name="dyn_straps_front_1_l_ctrl"),(Type=Control,Name="dyn_straps_front_3_l_ctrl"),(Type=Control,Name="dyn_straps_front_5_l_ctrl"),(Type=Control,Name="dyn_straps_front_6_l_ctrl"))) {template="DISPATCH_RigVMDispatch_ArrayMake(in Values,out Array)"}
   > SplineIK = RigVMFunctionReferenceNode(
    Primary Axis=(X=1,Y=0,Z=0)
    Samples Per Segment=16
    Secondary Spline Direction=(X=0,Y=1,Z=0)
    Spline Mode=BSpline
    Stretch=true
    Up Axis=(X=0,Y=0,Z=0)
    Bones=@Make_Array.Array
    Controls=@Make_Array_1.Array
   )
   Make_Array_2 = RigVMDispatchNode(Values=((Type=Bone,Name="dyn_straps_back_1_l"),(Type=Bone,Name="dyn_straps_back_2_l"),(Type=Bone,Name="dyn_straps_back_3_l"),(Type=Bone,Name="dyn_straps_back_4_l"),(Type=Bone,Name="dyn_straps_back_5_l"),(Type=Bone,Name="dyn_straps_back_6_l"))) {template="DISPATCH_RigVMDispatch_ArrayMake(in Values,out Array)"}
   Make_Array_3 = RigVMDispatchNode(Values=((Type=Control,Name="dyn_straps_back_1_l_ctrl"),(Type=Control,Name="dyn_straps_back_3_l_ctrl"),(Type=Control,Name="dyn_straps_back_5_l_ctrl"),(Type=Control,Name="dyn_straps_back_6_l_ctrl"))) {template="DISPATCH_RigVMDispatch_ArrayMake(in Values,out Array)"}
   > SplineIK_1 = RigVMFunctionReferenceNode(
    Primary Axis=(X=1,Y=0,Z=0)
    Samples Per Segment=16
    Secondary Spline Direction=(X=0,Y=1,Z=0)
    Spline Mode=BSpline
    Stretch=true
    Up Axis=(X=0,Y=0,Z=0)
    Bones=@Make_Array_2.Array
    Controls=@Make_Array_3.Array
   )
   &RigUnit_ModifyTransforms(
    ItemToModify=((Item=(Type=Control,Name="dyn_straps_front_3_l_ctrl"),Transform=(Rotation=(X=0,Y=0,Z=0,W=1),Translation=(X=-4,Y=5,Z=-1),Scale3D=(X=1,Y=1,Z=1))))
    WeightMinimum=0
    Weight=@n27.Value
    ExecutePin=@RigVMFunction_Sequence_3_3.B
   )
   > &RigUnit_ModifyTransforms(
    ItemToModify=((Item=(Type=Control,Name="dyn_straps_front_5_l_ctrl"),Transform=(Rotation=(X=0,Y=0,Z=0,W=1),Translation=(X=0,Y=6.5,Z=-1),Scale3D=(X=1,Y=1,Z=1))))
    WeightMinimum=0
    Weight=@n27.Value
   )
   > &RigUnit_ModifyTransforms(
    ItemToModify=((Item=(Type=Control,Name="dyn_straps_back_1_l_ctrl"),Transform=(Rotation=(X=0,Y=0,Z=0,W=1),Translation=(X=0,Y=-2,Z=-2),Scale3D=(X=1,Y=1,Z=1))))
    WeightMinimum=0
    Weight=@n27.Value
   )
   > &RigUnit_ModifyTransforms(
    ItemToModify=((Item=(Type=Control,Name="dyn_straps_back_3_l_ctrl"),Transform=(Rotation=(X=0,Y=0,Z=0,W=1),Translation=(X=0,Y=-5,Z=-2),Scale3D=(X=1,Y=1,Z=1))))
    WeightMinimum=0
    Weight=@n27.Value
   )
   > &RigUnit_ModifyTransforms(
    ItemToModify=((Item=(Type=Control,Name="dyn_straps_back_5_l_ctrl"),Transform=(Rotation=(X=0,Y=0,Z=0,W=1),Translation=(X=0,Y=-5,Z=-2.5),Scale3D=(X=1,Y=1,Z=1))))
    WeightMinimum=0
    Weight=@n27.Value
   )
   > &RigUnit_ModifyTransforms(
    ItemToModify=((Item=(Type=Control,Name="dyn_straps_front_6_l_ctrl"),Transform=(Rotation=(X=0,Y=0,Z=0,W=1),Translation=(X=0,Y=0,Z=0),Scale3D=(X=1,Y=1,Z=1))))
    WeightMinimum=0
    Weight=@n27.Value
   )
   n28 = &RigVMFunction_MathDoubleMul(B=1, A=RigVMVariableNode() {variable=Dyn_move_alfa_l}.Value)
   &RigUnit_ModifyTransforms(
    ItemToModify=((Item=(Type=Control,Name="dyn_straps_front_3_l_ctrl"),Transform=(Rotation=(X=0,Y=0,Z=0,W=1),Translation=(X=0,Y=1.5,Z=0),Scale3D=(X=1,Y=1,Z=1))))
    WeightMinimum=0
    Weight=@n28.Result
    ExecutePin=@RigVMFunction_Sequence_3_3.A
   )
   > &RigUnit_ModifyTransforms(
    ItemToModify=((Item=(Type=Control,Name="dyn_straps_front_5_l_ctrl"),Transform=(Rotation=(X=0,Y=0,Z=0,W=1),Translation=(X=0,Y=3,Z=0),Scale3D=(X=1,Y=1,Z=1))))
    WeightMinimum=0
    Weight=@n28.Result
   )
   n29 = &RigVMFunction_MathDoubleMul(A=5, B=&RigVMFunction_MathBoolToFloat(Value=RigVMVariableNode() {variable=is_crouch}.Value).Result)
   n30 = RigVMVariableNode() {variable=is_front_end}
   n31 = &RigVMFunction_MathBoolToFloat(Value=RigVMVariableNode() {variable=is_emote}.Value)
   n32 = &RigVMFunction_MathDoubleMul(
    A=&RigVMFunction_MathFloatMake(Value=&RigVMFunction_MathBoolToFloat(Value=@n30.Value).Result).Value
    B=&RigVMFunction_MathDoubleSub(A=1, B=&RigVMFunction_MathFloatMake(Value=@n31.Result).Value).Result
   )
   > &RigUnit_ModifyTransforms(
    ItemToModify=((Item=(Type=Control,Name="dyn_straps_back_3_l_ctrl"),Transform=(Rotation=(X=0,Y=0,Z=0,W=1),Translation=(X=0,Y=0,Z=0.5),Scale3D=(X=1,Y=1,Z=1))))
    WeightMinimum=0
    ItemToModify.0.Transform.Translation.Y=&RigVMFunction_MathDoubleAdd(
     A=&RigVMFunction_MathDoubleAdd(A=0, B=@n29.Result).Result
     B=&RigVMFunction_MathDoubleMul(A=-5, B=@n32.Result).Result
    ).Result
    Weight=@n28.Result
   )
   > &RigUnit_ModifyTransforms(
    ItemToModify=((Item=(Type=Control,Name="dyn_straps_back_5_l_ctrl"),Transform=(Rotation=(X=0,Y=0,Z=0,W=1),Translation=(X=0,Y=2,Z=-2),Scale3D=(X=1,Y=1,Z=1))))
    WeightMinimum=0
    Weight=@n28.Result
   )
   n33 = RigVMVariableNode() {variable=spr_thigh_l_up_alpha}
   n34 = RigVMVariableNode() {variable=SplineIK_chain_l_ctrls}
   /Script/ControlRig.RigUnit_HierarchySetPoseItemArray(
    ElementType=All
    Weight=&RigVMFunction_MathDoubleSub(A=1, B=@n33.Value).Result
    ItemsToSet=@n34.Value
    Pose=RigVMVariableNode() {variable=SplineIK_ctrls_l_bind_pose}.Value
    ExecutePin=@RigVMFunction_Sequence_3_2.C
   ) {is=struct RigVMUnitNode}
   > n35 = &RigUnit_HierarchySetPoseItemArray(ElementType=All, Weight=@n33.Value, ItemsToSet=@n34.Value, Pose=RigVMVariableNode() {variable=SplineIK_ctrl_thigh_l_up_pose}.Value)
   Branch_RigVMUnitNode_1_2 = &RigVMFunction_ControlFlowBranch(ExecuteContext=@n35.ExecutePin, Condition=RigVMVariableNode() {variable=is_front_end}.Value)
   n36 = RigVMVariableNode() {variable=spr_thigh_r_slight_back_out_apha}
   &RigUnit_HierarchySetPoseItemArray(
    ElementType=All
    ExecutePin=@Branch_RigVMUnitNode_1_2.True
    Weight=&RigVMFunction_MathDoubleSub(A=1, B=@n36.Value).Result
    Pose=RigVMVariableNode() {variable=SplineIK_ctrls_l_bind_pose}.Value
    ItemsToSet=@n34.Value
   )
   > &RigUnit_HierarchySetPoseItemArray(
    ElementType=All
    Pose=RigVMVariableNode() {variable=SplineIK_ctrl_thigh_r_leg_slight_back_out_pose}.Value
    Weight=@n36.Value
    ItemsToSet=@n34.Value
   )
   RigVMFunction_MathBoolToFloat = &RigVMFunction_MathBoolToFloat(Value=RigVMVariableNode() {variable=is_crouch}.Value)
   Clamp = &RigVMFunction_MathDoubleClamp(Value=&RigVMFunction_MathDoubleSub(A=@RigVMFunction_MathBoolToFloat.Result, B=RigVMVariableNode() {variable=Dyn_move_alfa_l}.Value).Result)
   RigVMFunction_MathFloatMake = &RigVMFunction_MathFloatMake(Value=@Clamp.Result)
   n37 = &RigVMFunction_MathDoubleMul(A=-2, B=@RigVMFunction_MathFloatMake.Value)
   n38 = RigVMVariableNode() {variable=spr_thigh_l_up_alpha}
   n39 = &RigVMFunction_MathDoubleMul(A=-1, B=@n38.Value)
   n40 = &RigVMFunction_MathDoubleSub(A=1, B=RigVMVariableNode() {variable=Dyn_pelvis_rotate_alfa}.Value)
   n41 = &RigVMFunction_MathDoubleMul(A=0, B=@RigVMFunction_MathFloatMake.Value)
   n42 = &RigVMFunction_MathDoubleMul(A=-0.5, B=@n38.Value)
   n43 = &RigVMFunction_MathDoubleMul(A=0, B=@RigVMFunction_MathFloatMake.Value)
   n44 = &RigVMFunction_MathDoubleMul(A=-0.3, B=@n38.Value)
   n45 = &RigVMFunction_MathDoubleSub(A=1, B=RigVMVariableNode() {variable=Dyn_pelvis_rotate_alfa}.Value)
   Multiply = &RigVMFunction_MathDoubleMul(A=10, B=@n45.Result)
   n46 = RigVMVariableNode() {variable=spr_thigh_l_up_alpha}
   Add = &RigVMFunction_MathDoubleAdd(A=@Multiply.Result, B=&RigVMFunction_MathDoubleMul(A=15, B=@n46.Value).Result)
   n47 = &RigVMFunction_MathDoubleMul(A=30, B=@RigVMFunction_MathFloatMake.Value)
   n48 = RigVMVariableNode() {variable=is_front_end}
   n49 = &RigVMFunction_MathBoolToFloat(Value=RigVMVariableNode() {variable=is_emote}.Value)
   n50 = &RigVMFunction_MathDoubleMul(
    A=&RigVMFunction_MathFloatMake(Value=&RigVMFunction_MathBoolToFloat(Value=@n48.Value).Result).Value
    B=&RigVMFunction_MathDoubleSub(A=1, B=&RigVMFunction_MathFloatMake(Value=@n49.Result).Value).Result
   )
   n51 = &RigVMFunction_MathDoubleAdd(
    A=&RigVMFunction_MathDoubleAdd(
     A=&RigVMFunction_MathDoubleMul(A=-10, B=@n45.Result).Result
     B=&RigVMFunction_MathDoubleMul(A=-45, B=@n46.Value).Result
    ).Result
    B=&RigVMFunction_MathDoubleMul(A=10, B=@n50.Result).Result
   )
   n52 = &RigVMFunction_MathDoubleMul(A=30, B=@RigVMFunction_MathFloatMake.Value)
   n53 = &RigVMFunction_MathDoubleAdd(
    A=&RigVMFunction_MathDoubleAdd(
     A=&RigVMFunction_MathDoubleMul(A=20, B=@n45.Result).Result
     B=&RigVMFunction_MathDoubleMul(A=60, B=@n46.Value).Result
    ).Result
    B=&RigVMFunction_MathDoubleMul(A=-15, B=@n50.Result).Result
   )
   n54 = &RigVMFunction_MathDoubleMul(A=5, B=@RigVMFunction_MathFloatMake.Value)
   &RigUnit_ModifyTransforms(
    ItemToModify=((Item=(Type=Bone,Name="dyn_straps_lock_front_1_l"),Transform=(Rotation=(X=0,Y=0,Z=0.7071,W=0.7071),Translation=(X=-0.5,Y=0,Z=-0.25),Scale3D=(X=1,Y=1,Z=1))))
    Weight=1
    WeightMinimum=0
    ItemToModify.0.Transform.Translation.X=&RigVMFunction_MathDoubleAdd(
     B=&RigVMFunction_MathDoubleAdd(B=@n37.Result, A=@n39.Result).Result
     A=&RigVMFunction_MathDoubleMul(A=0, B=@n40.Result).Result
    ).Result
    ItemToModify.0.Transform.Translation.Y=&RigVMFunction_MathDoubleAdd(
     B=&RigVMFunction_MathDoubleAdd(B=@n41.Result, A=@n42.Result).Result
     A=&RigVMFunction_MathDoubleMul(A=0, B=@n40.Result).Result
    ).Result
    ItemToModify.0.Transform.Translation.Z=&RigVMFunction_MathDoubleAdd(
     B=&RigVMFunction_MathDoubleAdd(B=@n43.Result, A=@n44.Result).Result
     A=&RigVMFunction_MathDoubleMul(A=0, B=@n40.Result).Result
    ).Result
    ItemToModify.0.Transform.Rotation=&RigVMFunction_MathQuaternionFromEuler(
     Euler=(X=0,Y=0,Z=30)
     Euler.X=&RigVMFunction_MathDoubleAdd(B=@Add.Result, A=@n47.Result).Result
     Euler.Y=&RigVMFunction_MathDoubleAdd(B=@n51.Result, A=@n52.Result).Result
     Euler.Z=&RigVMFunction_MathDoubleAdd(B=@n53.Result, A=@n54.Result).Result
    ).Result
    ExecutePin=@n24.A
   )
   n55 = RigVMVariableNode() {variable=spr_thigh_l_up_alpha}
   n56 = &RigVMFunction_MathDoubleSub(A=1, B=RigVMVariableNode() {variable=Dyn_pelvis_rotate_alfa}.Value)
   n57 = RigVMVariableNode() {variable=spr_thigh_l_up_alpha}
   n58 = &RigVMFunction_MathDoubleAdd(
    A=&RigVMFunction_MathDoubleMul(A=0, B=@n56.Result).Result
    B=&RigVMFunction_MathDoubleMul(A=20, B=@n57.Value).Result
   )
   n59 = &RigVMFunction_MathFloatMake(Value=&RigVMFunction_MathBoolToFloat(Value=RigVMVariableNode() {variable=is_crouch}.Value).Result)
   n60 = &RigVMFunction_MathDoubleMul(A=0, B=@n59.Value)
   n61 = &RigVMFunction_MathDoubleAdd(
    B=&RigVMFunction_MathDoubleMul(A=-10, B=@n57.Value).Result
    A=&RigVMFunction_MathDoubleMul(A=-10, B=@n56.Result).Result
   )
   n62 = &RigVMFunction_MathDoubleMul(A=0, B=@n59.Value)
   n63 = &RigVMFunction_MathDoubleAdd(
    B=&RigVMFunction_MathDoubleMul(A=10, B=@n57.Value).Result
    A=&RigVMFunction_MathDoubleMul(A=10, B=@n56.Result).Result
   )
   n64 = &RigVMFunction_MathDoubleMul(A=0, B=&RigVMFunction_MathDoubleSub(A=@n59.Value, B=RigVMVariableNode() {variable=Dyn_move_alfa_l}.Value).Result)
   > &RigUnit_ModifyTransforms(
    ItemToModify=((Item=(Type=Bone,Name="dyn_straps_lock_front_2_l"),Transform=(Rotation=(X=0,Y=0,Z=0,W=1),Translation=(X=0,Y=0,Z=0),Scale3D=(X=1,Y=1,Z=1))))
    Weight=1
    WeightMinimum=0
    ItemToModify.0.Transform.Translation.Y=&RigVMFunction_MathDoubleMul(A=-0.15, B=@n55.Value).Result
    ItemToModify.0.Transform.Translation.X=&RigVMFunction_MathDoubleMul(A=-0.33, B=@n55.Value).Result
    ItemToModify.0.Transform.Translation.Z=&RigVMFunction_MathDoubleMul(A=0, B=@n55.Value).Result
    ItemToModify.0.Transform.Rotation=&RigVMFunction_MathQuaternionFromEuler(
     Euler=(X=0,Y=0,Z=30)
     Euler.X=&RigVMFunction_MathDoubleAdd(B=@n58.Result, A=@n60.Result).Result
     Euler.Y=&RigVMFunction_MathDoubleAdd(B=@n61.Result, A=@n62.Result).Result
     Euler.Z=&RigVMFunction_MathDoubleAdd(B=@n63.Result, A=@n64.Result).Result
    ).Result
   )
   n65 = &RigVMFunction_MathBoolToFloat(Value=RigVMVariableNode() {variable=is_DBNO}.Value)
   n66 = &RigVMFunction_MathDoubleMul(A=1, B=@n65.Result)
   n67 = RigVMVariableNode() {variable=spr_thigh_l_up_alpha}
   n68 = &RigVMFunction_MathFloatMake(Value=&RigVMFunction_MathBoolToFloat(Value=RigVMVariableNode() {variable=is_crouch}.Value).Result)
   n69 = &RigVMFunction_MathDoubleSub(A=1, B=@n68.Value)
   n70 = &RigVMFunction_MathDoubleMul(A=&RigVMFunction_MathDoubleMul(A=-1, B=@n67.Value).Result, B=@n69.Result)
   n71 = RigVMVariableNode() {variable=is_front_end}
   n72 = &RigVMFunction_MathBoolToFloat(Value=RigVMVariableNode() {variable=is_emote}.Value)
   n73 = &RigVMFunction_MathDoubleMul(
    A=&RigVMFunction_MathFloatMake(Value=&RigVMFunction_MathBoolToFloat(Value=@n71.Value).Result).Value
    B=&RigVMFunction_MathDoubleSub(A=1, B=&RigVMFunction_MathFloatMake(Value=@n72.Result).Value).Result
   )
   n74 = &RigVMFunction_MathDoubleMul(A=0.75, B=@n65.Result)
   n75 = &RigVMFunction_MathDoubleMul(A=&RigVMFunction_MathDoubleMul(A=-1, B=@n67.Value).Result, B=@n69.Result)
   n76 = &RigVMFunction_MathDoubleMul(A=-0.25, B=@n65.Result)
   n77 = &RigVMFunction_MathDoubleMul(A=&RigVMFunction_MathDoubleMul(A=0.5, B=@n67.Value).Result, B=@n69.Result)
   n78 = &RigVMFunction_MathDoubleSub(A=1, B=RigVMVariableNode() {variable=Dyn_pelvis_rotate_alfa}.Value)
   n79 = &RigVMFunction_MathDoubleMul(A=0, B=@n78.Result)
   n80 = RigVMVariableNode() {variable=spr_thigh_l_up_alpha}
   n81 = &RigVMFunction_MathDoubleMul(A=-10, B=@n80.Value)
   n82 = &RigVMFunction_MathDoubleAdd(
    B=&RigVMFunction_MathDoubleAdd(
     A=&RigVMFunction_MathDoubleAdd(A=@n79.Result, B=@n81.Result).Result
     B=&RigVMFunction_MathDoubleMul(A=5, B=@n68.Value).Result
    ).Result
    A=&RigVMFunction_MathDoubleMul(A=0, B=@n65.Result).Result
   )
   n83 = RigVMVariableNode() {variable=Dyn_move_alfa_l}
   n84 = &RigVMFunction_MathDoubleMul(A=-10, B=@n83.Value)
   n85 = &RigVMFunction_MathDoubleAdd(
    B=&RigVMFunction_MathDoubleMul(A=30, B=@n78.Result).Result
    A=&RigVMFunction_MathDoubleMul(A=-10, B=@n73.Result).Result
   )
   n86 = &RigVMFunction_MathDoubleMul(A=10, B=@n80.Value)
   n87 = &RigVMFunction_MathDoubleAdd(
    B=&RigVMFunction_MathDoubleAdd(
     A=&RigVMFunction_MathDoubleAdd(A=@n85.Result, B=@n86.Result).Result
     B=&RigVMFunction_MathDoubleMul(A=-15, B=@n68.Value).Result
    ).Result
    A=&RigVMFunction_MathDoubleMul(A=-10, B=@n65.Result).Result
   )
   n88 = &RigVMFunction_MathDoubleMul(A=30, B=@n83.Value)
   n89 = &RigVMFunction_MathDoubleMul(A=-20, B=@n73.Result)
   n90 = &RigVMFunction_MathDoubleAdd(
    A=&RigVMFunction_MathDoubleMul(A=&RigVMFunction_MathDoubleAdd(A=30, B=@n89.Result).Result, B=@n78.Result).Result
    B=&RigVMFunction_MathDoubleMul(A=5, B=@n73.Result).Result
   )
   n91 = &RigVMFunction_MathDoubleAdd(
    B=&RigVMFunction_MathDoubleMul(A=-20, B=&RigVMFunction_MathDoubleMul(B=@n73.Result, A=@n80.Value).Result).Result
    A=&RigVMFunction_MathDoubleMul(A=0, B=@n80.Value).Result
   )
   n92 = &RigVMFunction_MathDoubleAdd(
    B=&RigVMFunction_MathDoubleAdd(
     A=&RigVMFunction_MathDoubleAdd(A=@n90.Result, B=@n91.Result).Result
     B=&RigVMFunction_MathDoubleMul(A=-10, B=@n68.Value).Result
    ).Result
    A=&RigVMFunction_MathDoubleMul(A=-30, B=@n65.Result).Result
   )
   n93 = &RigVMFunction_MathBoolToFloat(Value=RigVMVariableNode() {variable=is_crouch}.Value)
   n94 = &RigVMFunction_MathDoubleMul(A=20, B=&RigVMFunction_MathDoubleMul(B=&RigVMFunction_MathFloatMake(Value=@n93.Result).Value, A=@n83.Value).Result)
   > &RigUnit_ModifyTransforms(
    ItemToModify=((Item=(Type=Bone,Name="dyn_straps_lock_back_1_l"),Transform=(Rotation=(X=0,Y=0,Z=0.0436,W=0.999),Translation=(X=-0.25,Y=0,Z=0.25),Scale3D=(X=1,Y=1,Z=1))))
    Weight=1
    WeightMinimum=0
    ItemToModify.0.Transform.Translation.Y=&RigVMFunction_MathDoubleAdd(
     B=&RigVMFunction_MathDoubleAdd(A=@n66.Result, B=@n70.Result).Result
     A=&RigVMFunction_MathDoubleMul(A=0.25, B=@n73.Result).Result
    ).Result
    ItemToModify.0.Transform.Translation.X=&RigVMFunction_MathDoubleAdd(
     B=&RigVMFunction_MathDoubleAdd(A=@n74.Result, B=@n75.Result).Result
     A=&RigVMFunction_MathDoubleMul(A=-0.5, B=@n73.Result).Result
    ).Result
    ItemToModify.0.Transform.Translation.Z=&RigVMFunction_MathDoubleAdd(
     B=&RigVMFunction_MathDoubleAdd(A=@n76.Result, B=@n77.Result).Result
     A=&RigVMFunction_MathDoubleMul(A=-0.15, B=@n73.Result).Result
    ).Result
    ItemToModify.0.Transform.Rotation=&RigVMFunction_MathQuaternionFromEuler(
     Euler=(X=0,Y=0,Z=30)
     Euler.X=&RigVMFunction_MathDoubleAdd(A=@n82.Result, B=@n84.Result).Result
     Euler.Y=&RigVMFunction_MathDoubleAdd(A=@n87.Result, B=@n88.Result).Result
     Euler.Z=&RigVMFunction_MathDoubleAdd(A=@n92.Result, B=@n94.Result).Result
    ).Result
   )
   n95 = &RigVMFunction_MathDoubleSub(A=1, B=RigVMVariableNode() {variable=Dyn_pelvis_rotate_alfa}.Value)
   n96 = RigVMVariableNode() {variable=spr_thigh_l_up_alpha}
   n97 = &RigVMFunction_MathDoubleAdd(
    A=&RigVMFunction_MathDoubleMul(A=0, B=@n95.Result).Result
    B=&RigVMFunction_MathDoubleMul(A=0, B=@n96.Value).Result
   )
   n98 = &RigVMFunction_MathFloatMake(Value=&RigVMFunction_MathBoolToFloat(Value=RigVMVariableNode() {variable=is_crouch}.Value).Result)
   n99 = &RigVMFunction_MathDoubleMul(A=0, B=@n98.Value)
   n100 = &RigVMFunction_MathDoubleAdd(
    A=&RigVMFunction_MathDoubleMul(A=10, B=@n95.Result).Result
    B=&RigVMFunction_MathDoubleMul(A=10, B=@n96.Value).Result
   )
   n101 = &RigVMFunction_MathDoubleMul(A=-10, B=@n98.Value)
   n102 = &RigVMFunction_MathDoubleAdd(
    A=&RigVMFunction_MathDoubleMul(A=0, B=@n95.Result).Result
    B=&RigVMFunction_MathDoubleMul(A=10, B=@n96.Value).Result
   )
   n103 = &RigVMFunction_MathDoubleMul(A=-10, B=@n98.Value)
   > &RigUnit_ModifyTransforms(
    ItemToModify=((Item=(Type=Bone,Name="dyn_straps_lock_back_2_l"),Transform=(Rotation=(X=0,Y=0,Z=0,W=1),Translation=(X=0,Y=0,Z=0),Scale3D=(X=1,Y=1,Z=1))))
    Weight=1
    WeightMinimum=0
    ItemToModify.0.Transform.Rotation=&RigVMFunction_MathQuaternionFromEuler(
     Euler=(X=0,Y=0,Z=30)
     Euler.X=&RigVMFunction_MathDoubleAdd(A=@n97.Result, B=@n99.Result).Result
     Euler.Y=&RigVMFunction_MathDoubleAdd(A=@n100.Result, B=@n101.Result).Result
     Euler.Z=&RigVMFunction_MathDoubleAdd(B=@n102.Result, A=@n103.Result).Result
    ).Result
   )
   n104 = &RigUnit_SphericalPoseReader(
    ActiveRegionScaleFactors=(PositiveWidth=2,NegativeWidth=1,PositiveHeight=2,NegativeHeight=2)
    ActiveRegionSize=0.1
    Debug=(bDrawDebug=true,DebugScale=25,DebugSegments=20,DebugThickness=0.25)
    DriverAxis=(X=1,Y=0,Z=0)
    DriverItem=(Type=Bone,Name="thigh_r")
    FalloffRegionScaleFactors=(PositiveWidth=1,NegativeWidth=1,PositiveHeight=1,NegativeHeight=1)
    FalloffSize=0.45
    FlipHeightScaling=false
    FlipWidthScaling=true
    OptionalParentItem=(Type=Bone)
    RotationOffset=(X=-90,Y=0,Z=-15)
    ExecutePin=@RigVMFunction_Sequence_2_1.B
   )
   > RigVMVariableNode(Value=@n104.OutputParam) {variable=spr_thigh_r_up_alpha}
   > n105 = &RigUnit_SphericalPoseReader(
    ActiveRegionScaleFactors=(PositiveWidth=1,NegativeWidth=1,PositiveHeight=1,NegativeHeight=1)
    ActiveRegionSize=0.05
    Debug=(bDrawDebug=false,DebugScale=25,DebugSegments=20,DebugThickness=0.25)
    DriverAxis=(X=1,Y=0,Z=0)
    DriverItem=(Type=Bone,Name="thigh_r")
    FalloffRegionScaleFactors=(PositiveWidth=1,NegativeWidth=1,PositiveHeight=1,NegativeHeight=1)
    FalloffSize=0.09
    OptionalParentItem=(Type=Bone)
    RotationOffset=(X=15,Y=-80,Z=0)
   )
   > RigVMVariableNode(Value=@n105.OutputParam) {variable=spr_thigh_r_slight_back_out_apha}
   n106 = RigVMVariableNode() {variable=is_front_end}
   n107 = &RigVMFunction_MathBoolToFloat(Value=RigVMVariableNode() {variable=is_emote}.Value)
   n108 = &RigVMFunction_MathDoubleMul(
    A=&RigVMFunction_MathFloatMake(Value=&RigVMFunction_MathBoolToFloat(Value=@n106.Value).Result).Value
    B=&RigVMFunction_MathDoubleSub(A=1, B=&RigVMFunction_MathFloatMake(Value=@n107.Result).Value).Result
   )
   n109 = &RigVMFunction_MathDoubleMul(A=-15, B=@n108.Result)
   > &RigUnit_ModifyTransforms(
    ItemToModify=((Item=(Type=Bone,Name="dyn_straps_base_r"),Transform=(Rotation=(X=0.1305,Y=0,Z=0,W=0.9914),Translation=(X=0,Y=2,Z=-2),Scale3D=(X=1,Y=1,Z=1))))
    WeightMinimum=0
    ItemToModify.0.Transform.Translation.Z=&RigVMFunction_MathDoubleAdd(A=0, B=&RigVMFunction_MathDoubleMul(A=-3, B=@n108.Result).Result).Result
    ItemToModify.0.Transform.Translation.Y=&RigVMFunction_MathDoubleAdd(A=0, B=&RigVMFunction_MathDoubleMul(A=2.5, B=@n108.Result).Result).Result
    ItemToModify.0.Transform.Rotation=&RigVMFunction_MathQuaternionFromEuler(Euler=(X=0,Y=0,Z=0), Euler.X=&RigVMFunction_MathDoubleAdd(A=0, B=@n109.Result).Result).Result
    Weight=@n105.OutputParam
   )
   > &RigUnit_ParentConstraint(
    AdvancedSettings=(RotationOrderForFilter=XZY)
    Child=(Type=Null,Name="dyn_straps_Main_r_null")
    Parents=((Item=(Type=Bone,Name="dyn_straps_end_r"),Weight=1),(Item=(Type=Bone,Name="thigh_r"),Weight=0))
    Weight=1
   )
   > n110 = &RigUnit_SphericalPoseReader(
    ActiveRegionScaleFactors=(PositiveWidth=1,NegativeWidth=0.5,PositiveHeight=2,NegativeHeight=1)
    ActiveRegionSize=0.1
    Debug=(bDrawDebug=true,DebugScale=25,DebugSegments=20,DebugThickness=0.25)
    DriverAxis=(X=1,Y=0,Z=0)
    DriverItem=(Type=Bone,Name="thigh_r")
    FalloffRegionScaleFactors=(PositiveWidth=1,NegativeWidth=1,PositiveHeight=1,NegativeHeight=1)
    FalloffSize=0.5
    FlipHeightScaling=false
    FlipWidthScaling=false
    OptionalParentItem=(Type=Bone)
    RotationOffset=(X=-45,Y=-20,Z=0)
   )
   > RigVMVariableNode(Value=@n110.OutputParam) {variable=Dyn_move_alfa_r}
   > RigVMFunction_Sequence_3 = RigVMAggregateNode()
   RigVMFunction_Sequence_3_1 = RigVMAggregateNode(ExecuteContext=@RigVMFunction_Sequence_3.B)
   &RigUnit_ParentConstraint(
    AdvancedSettings=(InterpolationType=Average,RotationOrderForFilter=XZY)
    Child=(Type=Control,Name="dyn_straps_front_1_r_ctrl")
    Parents=((Item=(Type=Bone,Name="dyn_straps_front_base_r"),Weight=1))
    Weight=1
    ExecutePin=@RigVMFunction_Sequence_3_1.A
   )
   > &RigUnit_ParentConstraint(
    AdvancedSettings=(InterpolationType=Average,RotationOrderForFilter=XZY)
    Child=(Type=Control,Name="dyn_straps_back_1_r_ctrl")
    Parents=((Item=(Type=Bone,Name="dyn_straps_back_base_r"),Weight=1))
    Weight=1
   )
   n111 = RigVMVariableNode() {variable=spr_thigh_r_up_alpha}
   n112 = &RigVMFunction_MathDoubleSub(A=1, B=@n111.Value)
   &RigUnit_ParentConstraint(
    AdvancedSettings=(RotationOrderForFilter=XZY)
    Child=(Type=Null,Name="dyn_straps_front_3_r_null")
    Parents=((Item=(Type=Bone,Name="thigh_r"),Weight=0.15),(Item=(Type=Control,Name="dyn_straps_Main_r_ctrl"),Weight=0.75),(Item=(Type=Bone,Name="deform_groin_r"),Weight=0.1))
    ExecutePin=@RigVMFunction_Sequence_3_1.B
    Weight=@n112.Result
   )
   > &RigUnit_ParentConstraint(
    AdvancedSettings=(RotationOrderForFilter=XZY)
    Child=(Type=Null,Name="dyn_straps_front_5_r_null")
    Parents=((Item=(Type=Bone,Name="thigh_r"),Weight=0.15),(Item=(Type=Control,Name="dyn_straps_Main_r_ctrl"),Weight=0.85))
    Weight=@n112.Result
   )
   > &RigUnit_ParentConstraint(
    AdvancedSettings=(RotationOrderForFilter=XZY)
    Child=(Type=Null,Name="dyn_straps_back_3_r_null")
    Parents=((Item=(Type=Bone,Name="thigh_r"),Weight=0.1),(Item=(Type=Control,Name="dyn_straps_Main_r_ctrl"),Weight=0.75),(Item=(Type=Control,Name="dyn_straps_back_1_r_ctrl"),Weight=0.15))
    Weight=@n112.Result
   )
   > &RigUnit_ParentConstraint(
    AdvancedSettings=(RotationOrderForFilter=XZY)
    Child=(Type=Null,Name="dyn_straps_back_5_r_null")
    Parents=((Item=(Type=Bone,Name="thigh_r"),Weight=0.25),(Item=(Type=Control,Name="dyn_straps_Main_r_ctrl"),Weight=0.75),(Item=(Type=Control,Name="dyn_straps_back_1_r_ctrl"),Weight=0))
    Weight=@n112.Result
   )
   > &RigUnit_ParentConstraint(
    AdvancedSettings=(RotationOrderForFilter=XZY)
    Child=(Type=Null,Name="dyn_straps_front_3_r_null")
    Parents=((Item=(Type=Bone,Name="thigh_r"),Weight=0.5),(Item=(Type=Control,Name="dyn_straps_Main_r_ctrl"),Weight=0.5),(Item=(Type=Bone,Name="deform_groin_r"),Weight=0))
    Weight=@n111.Value
   )
   > &RigUnit_ParentConstraint(
    AdvancedSettings=(RotationOrderForFilter=XZY)
    Child=(Type=Null,Name="dyn_straps_front_5_r_null")
    Parents=((Item=(Type=Bone,Name="thigh_r"),Weight=0.25),(Item=(Type=Control,Name="dyn_straps_Main_r_ctrl"),Weight=0.75))
    Weight=@n111.Value
   )
   > &RigUnit_ParentConstraint(
    AdvancedSettings=(RotationOrderForFilter=XZY)
    Child=(Type=Null,Name="dyn_straps_back_3_r_null")
    Parents=((Item=(Type=Control,Name="dyn_straps_back_1_r_ctrl"),Weight=0.5),(Item=(Type=Control,Name="dyn_straps_Main_r_ctrl"),Weight=0.5))
    Weight=@n111.Value
   )
   > &RigUnit_ParentConstraint(
    AdvancedSettings=(RotationOrderForFilter=XZY)
    Child=(Type=Null,Name="dyn_straps_back_5_r_null")
    Parents=((Item=(Type=Control,Name="dyn_straps_back_1_r_ctrl"),Weight=0.25),(Item=(Type=Control,Name="dyn_straps_Main_r_ctrl"),Weight=0.75))
    Weight=@n111.Value
   )
   > RigVMFunction_Sequence_3_4 = RigVMAggregateNode()
   n113 = &RigVMFunction_MathDoubleMul(A=1, B=RigVMVariableNode() {variable=Dyn_move_alfa_r}.Value)
   &RigUnit_ModifyTransforms(
    ItemToModify=((Item=(Type=Control,Name="dyn_straps_front_3_r_ctrl"),Transform=(Rotation=(X=0,Y=0,Z=0,W=1),Translation=(X=0,Y=2.5,Z=-0.25),Scale3D=(X=1,Y=1,Z=1))))
    WeightMinimum=0
    ExecutePin=@RigVMFunction_Sequence_3_4.A
    Weight=@n113.Result
   )
   > &RigUnit_ModifyTransforms(
    ItemToModify=((Item=(Type=Control,Name="dyn_straps_front_5_r_ctrl"),Transform=(Rotation=(X=0,Y=0,Z=0,W=1),Translation=(X=0,Y=0,Z=0),Scale3D=(X=1,Y=1,Z=1))))
    WeightMinimum=0
    Weight=@n113.Result
   )
   > &RigUnit_ModifyTransforms(
    ItemToModify=((Item=(Type=Control,Name="dyn_straps_back_3_r_ctrl"),Transform=(Rotation=(X=0,Y=0,Z=0,W=1),Translation=(X=-2.5,Y=2.5,Z=2.5),Scale3D=(X=1,Y=1,Z=1))))
    WeightMinimum=0
    Weight=@n113.Result
   )
   > &RigUnit_ModifyTransforms(
    ItemToModify=((Item=(Type=Control,Name="dyn_straps_back_5_r_ctrl"),Transform=(Rotation=(X=0,Y=0,Z=0,W=1),Translation=(X=0,Y=0,Z=0),Scale3D=(X=1,Y=1,Z=1))))
    WeightMinimum=0
    Weight=@n113.Result
   )
   n114 = RigVMVariableNode() {variable=spr_thigh_r_up_alpha}
   &RigUnit_ModifyTransforms(
    ItemToModify=((Item=(Type=Control,Name="dyn_straps_front_3_r_ctrl"),Transform=(Rotation=(X=0,Y=0,Z=0,W=1),Translation=(X=0,Y=-5,Z=1),Scale3D=(X=1,Y=1,Z=1))))
    WeightMinimum=0
    ExecutePin=@RigVMFunction_Sequence_3_4.B
    Weight=@n114.Value
   )
   > &RigUnit_ModifyTransforms(
    ItemToModify=((Item=(Type=Control,Name="dyn_straps_front_5_r_ctrl"),Transform=(Rotation=(X=0,Y=0,Z=0,W=1),Translation=(X=0,Y=-5,Z=0),Scale3D=(X=1,Y=1,Z=1))))
    WeightMinimum=0
    Weight=@n114.Value
   )
   > &RigUnit_ModifyTransforms(
    ItemToModify=((Item=(Type=Control,Name="dyn_straps_back_1_r_ctrl"),Transform=(Rotation=(X=0,Y=0,Z=0,W=1),Translation=(X=0,Y=2,Z=0),Scale3D=(X=1,Y=1,Z=1))))
    WeightMinimum=0
    Weight=@n114.Value
   )
   > &RigUnit_ModifyTransforms(
    ItemToModify=((Item=(Type=Control,Name="dyn_straps_back_3_r_ctrl"),Transform=(Rotation=(X=0,Y=0,Z=0,W=1),Translation=(X=0,Y=-2,Z=-10),Scale3D=(X=1,Y=1,Z=1))))
    WeightMinimum=0
    Weight=@n114.Value
   )
   > &RigUnit_ModifyTransforms(
    ItemToModify=((Item=(Type=Control,Name="dyn_straps_back_5_r_ctrl"),Transform=(Rotation=(X=0,Y=0,Z=0,W=1),Translation=(X=0,Y=0,Z=0),Scale3D=(X=1,Y=1,Z=1))))
    WeightMinimum=0
    Weight=@n114.Value
   )
   &RigUnit_ModifyTransforms(
    ItemToModify=((Item=(Type=Control,Name="dyn_straps_front_1_r_ctrl"),Transform=(Rotation=(X=0,Y=0,Z=0,W=1),Translation=(X=0,Y=0,Z=0),Scale3D=(X=1,Y=1,Z=1))))
    Weight=0
    WeightMinimum=0
    ExecutePin=@RigVMFunction_Sequence_3_4.C
   )
   n115 = &RigVMFunction_MathDoubleSub(A=1, B=@n114.Value)
   > &RigUnit_ModifyTransforms(
    ItemToModify=((Item=(Type=Control,Name="dyn_straps_front_3_r_ctrl"),Transform=(Rotation=(X=0,Y=0,Z=0,W=1),Translation=(X=0,Y=-2.5,Z=0),Scale3D=(X=1,Y=1,Z=1))))
    WeightMinimum=0
    Weight=@n115.Result
   )
   > &RigUnit_ModifyTransforms(
    ItemToModify=((Item=(Type=Control,Name="dyn_straps_front_5_r_ctrl"),Transform=(Rotation=(X=0,Y=0,Z=0,W=1),Translation=(X=0,Y=-2.5,Z=-5),Scale3D=(X=1,Y=1,Z=1))))
    WeightMinimum=0
    Weight=@n115.Result
   )
   > &RigUnit_ModifyTransforms(
    ItemToModify=((Item=(Type=Control,Name="dyn_straps_back_1_r_ctrl"),Transform=(Rotation=(X=0,Y=0,Z=0,W=1),Translation=(X=1,Y=1,Z=-2),Scale3D=(X=1,Y=1,Z=1))))
    WeightMinimum=0
    Weight=@n115.Result
   )
   > &RigUnit_ModifyTransforms(
    ItemToModify=((Item=(Type=Control,Name="dyn_straps_back_3_r_ctrl"),Transform=(Rotation=(X=0,Y=0,Z=0,W=1),Translation=(X=0,Y=0,Z=-5),Scale3D=(X=1,Y=1,Z=1))))
    WeightMinimum=0
    Weight=@n115.Result
   )
   > &RigUnit_ModifyTransforms(
    ItemToModify=((Item=(Type=Control,Name="dyn_straps_back_5_r_ctrl"),Transform=(Rotation=(X=0,Y=0,Z=0,W=1),Translation=(X=0,Y=0,Z=-10),Scale3D=(X=1,Y=1,Z=1))))
    WeightMinimum=0
    Weight=@n115.Result
   )
   n116 = RigVMVariableNode() {variable=is_front_end}
   n117 = &RigVMFunction_MathBoolToFloat(Value=RigVMVariableNode() {variable=is_emote}.Value)
   n118 = &RigVMFunction_MathDoubleMul(
    A=&RigVMFunction_MathFloatMake(Value=&RigVMFunction_MathBoolToFloat(Value=@n116.Value).Result).Value
    B=&RigVMFunction_MathDoubleSub(A=1, B=&RigVMFunction_MathFloatMake(Value=@n117.Result).Value).Result
   )
   > &RigUnit_ModifyTransforms(
    ItemToModify=((Item=(Type=Control,Name="dyn_straps_front_3_r_ctrl"),Transform=(Rotation=(X=0,Y=0,Z=0,W=1),Translation=(X=0,Y=-5,Z=0),Scale3D=(X=1,Y=1,Z=1))))
    WeightMinimum=0
    Weight=@n118.Result
   )
   > &RigUnit_ModifyTransforms(
    ItemToModify=((Item=(Type=Control,Name="dyn_straps_back_3_r_ctrl"),Transform=(Rotation=(X=0,Y=0,Z=0,W=1),Translation=(X=0,Y=5,Z=0),Scale3D=(X=1,Y=1,Z=1))))
    WeightMinimum=0
    Weight=@n118.Result
   )
   > &RigUnit_ModifyTransforms(
    ItemToModify=((Item=(Type=Bone,Name="dyn_bear_pendant_base"),Transform=(Rotation=(X=0,Y=-0.0436,Z=0,W=0.999),Translation=(X=0,Y=0,Z=0),Scale3D=(X=1,Y=1,Z=1))))
    WeightMinimum=0
    Weight=@n118.Result
   )
   RerouteNode_11_1 = RigVMDispatchNode(Values=((Type=Bone,Name="dyn_straps_front_1_r"),((),Type=Bone,Name="dyn_straps_front_2_r"),((),Type=Bone,Name="dyn_straps_front_3_r"),((),Type=Bone,Name="dyn_straps_front_4_r"),((),Type=Bone,Name="dyn_straps_front_5_r"),((),Type=Bone,Name="dyn_straps_front_6_r"))) {template="DISPATCH_RigVMDispatch_ArrayMake(in Values,out Array)"}
   RerouteNode_15_1 = RigVMDispatchNode(Values=((Type=Control,Name="dyn_straps_front_1_r_ctrl"),((),Type=Control,Name="dyn_straps_front_3_r_ctrl"),((),Type=Control,Name="dyn_straps_front_5_r_ctrl"),((),Type=Control,Name="dyn_straps_front_6_r_ctrl"))) {template="DISPATCH_RigVMDispatch_ArrayMake(in Values,out Array)"}
   > SplineIK_2_2 = RigVMFunctionReferenceNode(
    Primary Axis=(X=1,Y=0,Z=0)
    Samples Per Segment=16
    Secondary Spline Direction=(X=0,Y=1,Z=0)
    Spline Mode=BSpline
    Stretch=true
    Up Axis=(X=0,Y=0,Z=0)
    Bones=@RerouteNode_11_1.Array
    Controls=@RerouteNode_15_1.Array
   )
   RerouteNode_11_2 = RigVMDispatchNode(Values=((Type=Bone,Name="dyn_straps_back_1_r"),((),Type=Bone,Name="dyn_straps_back_2_r"),((),Type=Bone,Name="dyn_straps_back_3_r"),((),Type=Bone,Name="dyn_straps_back_4_r"),((),Type=Bone,Name="dyn_straps_back_5_r"),((),Type=Bone,Name="dyn_straps_back_6_r"))) {template="DISPATCH_RigVMDispatch_ArrayMake(in Values,out Array)"}
   RerouteNode_15_2 = RigVMDispatchNode(Values=((Type=Control,Name="dyn_straps_back_1_r_ctrl"),((),Type=Control,Name="dyn_straps_back_3_r_ctrl"),((),Type=Control,Name="dyn_straps_back_5_r_ctrl"),((),Type=Control,Name="dyn_straps_back_6_r_ctrl"))) {template="DISPATCH_RigVMDispatch_ArrayMake(in Values,out Array)"}
   > SplineIK_2_1 = RigVMFunctionReferenceNode(
    Primary Axis=(X=1,Y=0,Z=0)
    Samples Per Segment=16
    Secondary Spline Direction=(X=0,Y=1,Z=0)
    Spline Mode=BSpline
    Stretch=true
    Up Axis=(X=0,Y=0,Z=0)
    Bones=@RerouteNode_11_2.Array
    Controls=@RerouteNode_15_2.Array
   )
   n119 = RigVMVariableNode() {variable=spr_thigh_r_up_alpha}
   n120 = RigVMVariableNode() {variable=SplineIK_chain_r_ctrls}
   &RigUnit_HierarchySetPoseItemArray(
    ElementType=All
    Weight=&RigVMFunction_MathDoubleSub(A=1, B=@n119.Value).Result
    Pose=RigVMVariableNode() {variable=SplineIK_ctrls_r_bind_pose}.Value
    ItemsToSet=@n120.Value
    ExecutePin=@RigVMFunction_Sequence_3_1.C
   )
   > n121 = &RigUnit_HierarchySetPoseItemArray(ElementType=All, Pose=RigVMVariableNode() {variable=SplineIK_ctrl_thigh_r_up_pose}.Value, ItemsToSet=@n120.Value, Weight=@n119.Value)
   Branch_RigVMUnitNode_1_1 = &RigVMFunction_ControlFlowBranch(ExecuteContext=@n121.ExecutePin, Condition=RigVMVariableNode() {variable=is_front_end}.Value)
   n122 = RigVMVariableNode() {variable=spr_thigh_r_slight_back_out_apha}
   &RigUnit_HierarchySetPoseItemArray(
    ElementType=All
    ExecutePin=@Branch_RigVMUnitNode_1_1.True
    Weight=&RigVMFunction_MathDoubleSub(A=1, B=@n122.Value).Result
    Pose=RigVMVariableNode() {variable=SplineIK_ctrls_r_bind_pose}.Value
    ItemsToSet=@n120.Value
   )
   n123 = &RigVMFunction_MathFloatMake(Value=&RigVMFunction_MathBoolToFloat(Value=RigVMVariableNode() {variable=is_crouch}.Value).Result)
   n124 = RigVMVariableNode() {variable=spr_thigh_r_up_alpha}
   n125 = &RigVMFunction_MathDoubleMul(A=-10, B=@n123.Value)
   n126 = &RigVMFunction_MathDoubleMul(A=0, B=@n124.Value)
   n127 = &RigVMFunction_MathDoubleMul(A=0, B=@n123.Value)
   n128 = &RigVMFunction_MathDoubleMul(A=0, B=@n124.Value)
   n129 = &RigVMFunction_MathDoubleMul(A=0, B=@n123.Value)
   n130 = &RigVMFunction_MathDoubleMul(A=0, B=@n124.Value)
   &RigUnit_ModifyTransforms(
    ItemToModify=((Item=(Type=Bone,Name="dyn_straps_base_r"),Transform=(Rotation=(X=0,Y=0,Z=-0.2588,W=0.9659),Translation=(X=0,Y=0,Z=0),Scale3D=(X=1,Y=1,Z=1))))
    Weight=1
    WeightMinimum=0
    ItemToModify.0.Transform.Translation.X=&RigVMFunction_MathDoubleAdd(
     A=&RigVMFunction_MathDoubleMul(A=0, B=@n123.Value).Result
     B=&RigVMFunction_MathDoubleMul(A=0, B=@n124.Value).Result
    ).Result
    ItemToModify.0.Transform.Translation.Y=&RigVMFunction_MathDoubleAdd(
     A=&RigVMFunction_MathDoubleMul(A=-2.5, B=@n123.Value).Result
     B=&RigVMFunction_MathDoubleMul(A=0, B=@n124.Value).Result
    ).Result
    ItemToModify.0.Transform.Translation.Z=&RigVMFunction_MathDoubleAdd(
     A=&RigVMFunction_MathDoubleMul(A=5, B=@n123.Value).Result
     B=&RigVMFunction_MathDoubleMul(A=0, B=@n124.Value).Result
    ).Result
    ItemToModify.0.Transform.Rotation=&RigVMFunction_MathQuaternionFromEuler(
     Euler=(X=0,Y=0,Z=45)
     Euler.X=&RigVMFunction_MathDoubleAdd(A=@n125.Result, B=@n126.Result).Result
     Euler.Y=&RigVMFunction_MathDoubleAdd(A=@n127.Result, B=@n128.Result).Result
     Euler.Z=&RigVMFunction_MathDoubleAdd(A=@n129.Result, B=@n130.Result).Result
    ).Result
    ExecutePin=@RigVMFunction_Sequence_3.A
   )
   n131 = &RigVMFunction_MathFloatMake(Value=&RigVMFunction_MathBoolToFloat(Value=RigVMVariableNode() {variable=is_crouch}.Value).Result)
   n132 = &RigVMFunction_MathDoubleMul(A=-0.5, B=@n131.Value)
   n133 = RigVMVariableNode() {variable=spr_thigh_r_up_alpha}
   n134 = &RigVMFunction_MathDoubleMul(A=-1.25, B=@n133.Value)
   n135 = &RigVMFunction_MathDoubleSub(A=1, B=RigVMVariableNode() {variable=Dyn_pelvis_rotate_alfa}.Value)
   n136 = &RigVMFunction_MathDoubleMul(A=0, B=@n131.Value)
   n137 = &RigVMFunction_MathDoubleMul(A=-0.25, B=@n133.Value)
   n138 = &RigVMFunction_MathDoubleMul(A=0.33, B=@n131.Value)
   n139 = &RigVMFunction_MathDoubleMul(A=0, B=@n133.Value)
   n140 = &RigVMFunction_MathFloatMake(Value=&RigVMFunction_MathBoolToFloat(Value=RigVMVariableNode() {variable=is_crouch}.Value).Result)
   n141 = RigVMVariableNode() {variable=spr_thigh_r_up_alpha}
   n142 = &RigVMFunction_MathDoubleSub(A=1, B=RigVMVariableNode() {variable=Dyn_pelvis_rotate_alfa}.Value)
   n143 = &RigVMFunction_MathDoubleAdd(
    A=&RigVMFunction_MathDoubleAdd(
     A=&RigVMFunction_MathDoubleMul(A=0, B=@n140.Value).Result
     B=&RigVMFunction_MathDoubleMul(A=-20, B=@n141.Value).Result
    ).Result
    B=&RigVMFunction_MathDoubleMul(A=-10, B=@n142.Result).Result
   )
   n144 = RigVMVariableNode() {variable=Dyn_move_alfa_r}
   n145 = &RigVMFunction_MathDoubleMul(A=-20, B=@n144.Value)
   n146 = &RigVMFunction_MathDoubleMul(A=-10, B=@n140.Value)
   n147 = &RigVMFunction_MathDoubleMul(A=-10, B=@n141.Value)
   n148 = &RigVMFunction_MathDoubleAdd(
    A=&RigVMFunction_MathDoubleAdd(
     A=&RigVMFunction_MathDoubleAdd(A=@n146.Result, B=@n147.Result).Result
     B=&RigVMFunction_MathDoubleMul(A=20, B=@n142.Result).Result
    ).Result
    B=&RigVMFunction_MathDoubleMul(A=20, B=@n144.Value).Result
   )
   n149 = RigVMVariableNode() {variable=is_front_end}
   n150 = &RigVMFunction_MathBoolToFloat(Value=RigVMVariableNode() {variable=is_emote}.Value)
   n151 = &RigVMFunction_MathDoubleMul(
    A=&RigVMFunction_MathFloatMake(Value=&RigVMFunction_MathBoolToFloat(Value=@n149.Value).Result).Value
    B=&RigVMFunction_MathDoubleSub(A=1, B=&RigVMFunction_MathFloatMake(Value=@n150.Result).Value).Result
   )
   n152 = &RigVMFunction_MathDoubleMul(A=-10, B=@n151.Result)
   n153 = &RigVMFunction_MathDoubleMul(A=20, B=@n140.Value)
   n154 = &RigVMFunction_MathDoubleMul(A=20, B=@n141.Value)
   n155 = &RigVMFunction_MathDoubleAdd(
    A=&RigVMFunction_MathDoubleAdd(
     A=&RigVMFunction_MathDoubleAdd(A=@n153.Result, B=@n154.Result).Result
     B=&RigVMFunction_MathDoubleMul(A=20, B=@n142.Result).Result
    ).Result
    B=&RigVMFunction_MathDoubleMul(A=5, B=@n144.Value).Result
   )
   n156 = &RigVMFunction_MathDoubleMul(A=-10, B=@n151.Result)
   > &RigUnit_ModifyTransforms(
    ItemToModify=((Item=(Type=Bone,Name="dyn_straps_lock_front_1_r"),Transform=(Rotation=(X=0,Y=0,Z=0,W=1),Translation=(X=-0.5,Y=0,Z=0.5),Scale3D=(X=1,Y=1,Z=1))))
    Weight=1
    WeightMinimum=0
    ItemToModify.0.Transform.Translation.X=&RigVMFunction_MathDoubleAdd(
     A=&RigVMFunction_MathDoubleAdd(A=@n132.Result, B=@n134.Result).Result
     B=&RigVMFunction_MathDoubleMul(A=-1.25, B=@n135.Result).Result
    ).Result
    ItemToModify.0.Transform.Translation.Y=&RigVMFunction_MathDoubleAdd(
     A=&RigVMFunction_MathDoubleAdd(A=@n136.Result, B=@n137.Result).Result
     B=&RigVMFunction_MathDoubleMul(A=0.25, B=@n135.Result).Result
    ).Result
    ItemToModify.0.Transform.Translation.Z=&RigVMFunction_MathDoubleAdd(
     A=&RigVMFunction_MathDoubleAdd(A=@n138.Result, B=@n139.Result).Result
     B=&RigVMFunction_MathDoubleMul(A=0.33, B=@n135.Result).Result
    ).Result
    ItemToModify.0.Transform.Rotation=&RigVMFunction_MathQuaternionFromEuler(
     Euler=(X=0,Y=0,Z=45)
     Euler.X=&RigVMFunction_MathDoubleAdd(A=@n143.Result, B=@n145.Result).Result
     Euler.Y=&RigVMFunction_MathDoubleAdd(A=@n148.Result, B=@n152.Result).Result
     Euler.Z=&RigVMFunction_MathDoubleAdd(A=@n155.Result, B=@n156.Result).Result
    ).Result
   )
   n157 = RigVMVariableNode() {variable=spr_thigh_r_up_alpha}
   n158 = &RigVMFunction_MathFloatMake(Value=&RigVMFunction_MathBoolToFloat(Value=RigVMVariableNode() {variable=is_crouch}.Value).Result)
   n159 = RigVMVariableNode() {variable=spr_thigh_r_up_alpha}
   n160 = &RigVMFunction_MathDoubleSub(A=1, B=RigVMVariableNode() {variable=Dyn_pelvis_rotate_alfa}.Value)
   n161 = &RigVMFunction_MathDoubleAdd(
    A=&RigVMFunction_MathDoubleAdd(
     A=&RigVMFunction_MathDoubleMul(A=0, B=@n158.Value).Result
     B=&RigVMFunction_MathDoubleMul(A=-10, B=@n159.Value).Result
    ).Result
    B=&RigVMFunction_MathDoubleMul(A=0, B=@n160.Result).Result
   )
   n162 = RigVMVariableNode() {variable=Dyn_move_alfa_r}
   n163 = &RigVMFunction_MathDoubleMul(A=-15, B=@n162.Value)
   n164 = &RigVMFunction_MathDoubleAdd(
    A=&RigVMFunction_MathDoubleAdd(
     A=&RigVMFunction_MathDoubleMul(A=0, B=@n158.Value).Result
     B=&RigVMFunction_MathDoubleMul(A=15, B=@n159.Value).Result
    ).Result
    B=&RigVMFunction_MathDoubleMul(A=20, B=@n160.Result).Result
   )
   n165 = &RigVMFunction_MathDoubleMul(A=0, B=@n162.Value)
   n166 = &RigVMFunction_MathDoubleAdd(
    A=&RigVMFunction_MathDoubleAdd(
     A=&RigVMFunction_MathDoubleMul(A=0, B=@n158.Value).Result
     B=&RigVMFunction_MathDoubleMul(A=20, B=@n159.Value).Result
    ).Result
    B=&RigVMFunction_MathDoubleMul(A=0, B=@n160.Result).Result
   )
   n167 = &RigVMFunction_MathDoubleMul(A=15, B=@n162.Value)
   > &RigUnit_ModifyTransforms(
    ItemToModify=((Item=(Type=Bone,Name="dyn_straps_lock_front_2_r"),Transform=(Rotation=(X=0,Y=0,Z=0,W=1),Translation=(X=0,Y=0,Z=0),Scale3D=(X=1,Y=1,Z=1))))
    Weight=1
    WeightMinimum=0
    ItemToModify.0.Transform.Translation.Y=&RigVMFunction_MathDoubleMul(A=-0.5, B=@n157.Value).Result
    ItemToModify.0.Transform.Translation.X=&RigVMFunction_MathDoubleMul(A=-0.25, B=@n157.Value).Result
    ItemToModify.0.Transform.Translation.Z=&RigVMFunction_MathDoubleMul(A=0, B=@n157.Value).Result
    ItemToModify.0.Transform.Rotation=&RigVMFunction_MathQuaternionFromEuler(
     Euler=(X=0,Y=0,Z=45)
     Euler.X=&RigVMFunction_MathDoubleAdd(A=@n161.Result, B=@n163.Result).Result
     Euler.Y=&RigVMFunction_MathDoubleAdd(A=@n164.Result, B=@n165.Result).Result
     Euler.Z=&RigVMFunction_MathDoubleAdd(A=@n166.Result, B=@n167.Result).Result
    ).Result
   )
   n168 = RigVMVariableNode() {variable=spr_thigh_r_up_alpha}
   n169 = &RigVMFunction_MathFloatMake(Value=&RigVMFunction_MathBoolToFloat(Value=RigVMVariableNode() {variable=is_crouch}.Value).Result)
   n170 = RigVMVariableNode() {variable=spr_thigh_r_up_alpha}
   n171 = &RigVMFunction_MathDoubleSub(A=1, B=RigVMVariableNode() {variable=Dyn_pelvis_rotate_alfa}.Value)
   n172 = &RigVMFunction_MathDoubleAdd(
    A=&RigVMFunction_MathDoubleAdd(
     A=&RigVMFunction_MathDoubleMul(A=0, B=@n169.Value).Result
     B=&RigVMFunction_MathDoubleMul(A=10, B=@n170.Value).Result
    ).Result
    B=&RigVMFunction_MathDoubleMul(A=0, B=@n171.Result).Result
   )
   n173 = RigVMVariableNode() {variable=Dyn_move_alfa_r}
   n174 = &RigVMFunction_MathDoubleMul(A=0, B=@n173.Value)
   n175 = &RigVMFunction_MathDoubleAdd(
    A=&RigVMFunction_MathDoubleAdd(
     A=&RigVMFunction_MathDoubleMul(A=0, B=@n169.Value).Result
     B=&RigVMFunction_MathDoubleMul(A=-10, B=@n170.Value).Result
    ).Result
    B=&RigVMFunction_MathDoubleMul(A=-30, B=@n171.Result).Result
   )
   n176 = &RigVMFunction_MathDoubleMul(A=-20, B=@n173.Value)
   n177 = &RigVMFunction_MathDoubleAdd(
    A=&RigVMFunction_MathDoubleAdd(
     A=&RigVMFunction_MathDoubleMul(A=0, B=@n169.Value).Result
     B=&RigVMFunction_MathDoubleMul(A=0, B=@n170.Value).Result
    ).Result
    B=&RigVMFunction_MathDoubleMul(A=0, B=@n171.Result).Result
   )
   n178 = &RigVMFunction_MathDoubleMul(A=10, B=@n173.Value)
   > &RigUnit_ModifyTransforms(
    ItemToModify=((Item=(Type=Bone,Name="dyn_straps_lock_back_1_r"),Transform=(Rotation=(X=0,Y=0,Z=0,W=1),Translation=(X=-0.25,Y=0,Z=-0.15),Scale3D=(X=1,Y=1,Z=1))))
    Weight=1
    WeightMinimum=0
    ItemToModify.0.Transform.Translation.Y=&RigVMFunction_MathDoubleMul(A=0, B=@n168.Value).Result
    ItemToModify.0.Transform.Translation.X=&RigVMFunction_MathDoubleMul(A=-0.25, B=@n168.Value).Result
    ItemToModify.0.Transform.Translation.Z=&RigVMFunction_MathDoubleMul(A=-0.3, B=@n168.Value).Result
    ItemToModify.0.Transform.Rotation=&RigVMFunction_MathQuaternionFromEuler(
     Euler=(X=0,Y=0,Z=45)
     Euler.X=&RigVMFunction_MathDoubleAdd(A=@n172.Result, B=@n174.Result).Result
     Euler.Y=&RigVMFunction_MathDoubleAdd(A=@n175.Result, B=@n176.Result).Result
     Euler.Z=&RigVMFunction_MathDoubleAdd(A=@n177.Result, B=@n178.Result).Result
    ).Result
   )
   > &RigUnit_ModifyTransforms(
    ItemToModify=((Item=(Type=Bone,Name="dyn_straps_lock_back_2_r"),Transform=(Rotation=(X=0,Y=0,Z=0,W=1),Translation=(X=0,Y=0,Z=0),Scale3D=(X=1,Y=1,Z=1))))
    WeightMinimum=0
    ItemToModify.0.Transform.Rotation=&RigVMFunction_MathQuaternionFromEuler(Euler=(X=0,Y=-10,Z=0)).Result
    Weight=RigVMVariableNode() {variable=spr_thigh_r_up_alpha}.Value
   )
   > &RigUnit_ParentConstraint(
    AdvancedSettings=(RotationOrderForFilter=XZY)
    Child=(Type=Null,Name="dyn_straps_front_1_r_null")
    Parents=((Item=(Type=Bone,Name="dyn_straps_front_base_r"),Weight=1))
    Weight=1
   )
   > &RigUnit_ParentConstraint(
    AdvancedSettings=(RotationOrderForFilter=XZY)
    Child=(Type=Null,Name="dyn_straps_back_1_r_null")
    Parents=((Item=(Type=Bone,Name="dyn_straps_back_base_r"),Weight=1))
    Weight=1
   )
   &RigVMFunction_MathDoubleAdd(A=0, B=-30)
   &RigUnit_HierarchySetPoseItemArray(ElementType=All, ExecutePin=(), Pose=(HierarchyTopologyVersion=-1,PoseHash=-1), Weight=@n122.Value, ItemsToSet=@n120.Value)
   &RigUnit_ModifyTransforms(
    ExecutePin=()
    ItemToModify=((Item=(Type=Control,Name="dyn_straps_back_3_l_ctrl"),Transform=(Rotation=(X=0,Y=0,Z=0,W=1),Translation=(X=0,Y=0,Z=0.5),Scale3D=(X=1,Y=1,Z=1))))
    Weight=1
    WeightMinimum=0
   )
   &RigUnit_ModifyTransforms(
    ExecutePin=()
    ItemToModify=((Item=(Type=Control,Name="dyn_straps_Main_l_ctrl"),Transform=(Rotation=(X=0,Y=0,Z=0,W=1),Translation=(X=-5,Y=0,Z=0),Scale3D=(X=1,Y=1,Z=1))))
    WeightMinimum=0
    Weight=@Subtract.Result
   )
   &RigVMFunction_MathDoubleMul(A=30, B=1)
   &RigVMFunction_MathDoubleMul(A=5, B=1)
   &RigVMFunction_MathDoubleMul(A=30, B=1)
   &RigVMFunction_MathDoubleMul(A=-15, B=1)
   &RigVMFunction_MathDoubleMul(A=-10, B=1)
   &RigVMFunction_MathDoubleMul(A=5, B=1)
   &RigVMFunction_MathDoubleMul(A=1, B=RigVMVariableNode() {variable=spr_thigh_r_up_alpha}.Value)
   &RigUnit_ParentConstraint(
    AdvancedSettings=(InterpolationType=Average,RotationOrderForFilter=XZY)
    Child=(Type=Null,Name="dyn_straps_front_1_l_null")
    ExecutePin=()
    Parents=((Item=(Type=Bone,Name="dyn_straps_front_base_l"),Weight=1))
    Weight=1
   )
   RerouteNode_11_3 = RigVMDispatchNode(Values=((Type=Bone,Name="dyn_straps_back_1_l"),((),Type=Bone,Name="dyn_straps_back_2_l"),((),Type=Bone,Name="dyn_straps_back_3_l"),((),Type=Bone,Name="dyn_straps_back_4_l"),((),Type=Bone,Name="dyn_straps_back_5_l"),((),Type=Bone,Name="dyn_straps_back_6_l"))) {template="DISPATCH_RigVMDispatch_ArrayMake(in Values,out Array)"}
   &RigVMFunction_MathBoolToFloat(Value=RigVMVariableNode() {variable=is_crouch}.Value)
   RigVMFunction_Sequence = &RigVMFunction_Sequence(ExecuteContext=())
   &RigUnit_SphericalPoseReader(
    ActiveRegionScaleFactors=(PositiveWidth=1,NegativeWidth=0.5,PositiveHeight=1,NegativeHeight=1)
    ActiveRegionSize=0.1
    Debug=(bDrawDebug=true,DebugScale=25,DebugSegments=20,DebugThickness=0.25)
    DriverAxis=(X=1,Y=0,Z=0)
    DriverItem=(Type=Bone,Name="thigh_l")
    ExecutePin=()
    FalloffRegionScaleFactors=(PositiveWidth=1,NegativeWidth=1,PositiveHeight=1,NegativeHeight=1)
    FalloffSize=0.35
    FlipHeightScaling=false
    FlipWidthScaling=true
    OptionalParentItem=(Type=Bone)
    RotationOffset=(X=-35,Y=-30,Z=20)
   )
   &RigUnit_SphericalPoseReader(
    ActiveRegionScaleFactors=(PositiveWidth=1,NegativeWidth=3,PositiveHeight=2,NegativeHeight=2)
    ActiveRegionSize=0.1
    Debug=(bDrawDebug=true,DebugScale=25,DebugSegments=20,DebugThickness=0.25)
    DriverAxis=(X=-1,Y=0,Z=0)
    DriverItem=(Type=Bone,Name="thigh_l")
    ExecutePin=()
    FalloffRegionScaleFactors=(PositiveWidth=1,NegativeWidth=1,PositiveHeight=1,NegativeHeight=1)
    FalloffSize=0.45
    FlipHeightScaling=false
    FlipWidthScaling=true
    OptionalParentItem=(Type=Bone)
    RotationOffset=(X=90,Y=0,Z=-5)
   )
   RigVMVariableNode() {variable=SplineIK_ctrl_thigh_r_leg_slight_back_out_pose}
   RigVMVariableNode() {variable=spr_thigh_r_up_alpha}
  ]}
 RigVMModel Construction Event Graph {nodes [
   &RigUnit_ModifyTransforms(
    ExecutePin=()
    ItemToModify=((Item=(Type=Control,Name="dyn_back_chain_05_r_ctrl"),Transform=(Rotation=(X=0,Y=0,Z=0,W=1),Translation=(X=0,Y=-0.5,Z=1),Scale3D=(X=1,Y=1,Z=1))))
    Weight=1
   )
   > &RigUnit_ModifyTransforms(
    ItemToModify=((Item=(Type=Control,Name="dyn_chain_end_r_ctrl"),Transform=(Rotation=(X=0,Y=0,Z=0,W=1),Translation=(X=1,Y=-0.1,Z=0),Scale3D=(X=1,Y=1,Z=1))))
    Weight=1
   )
   > &RigUnit_ModifyTransforms(
    ItemToModify=((Item=(Type=Control,Name="dyn_front_chain_03_r_ctrl"),Transform=(Rotation=(X=0,Y=0,Z=0,W=1),Translation=(X=0,Y=0,Z=0),Scale3D=(X=1,Y=1,Z=1))))
    Weight=1
   )
   > &RigUnit_ModifyTransforms(
    ItemToModify=((Item=(Type=Control,Name="dyn_front_chain_05_r_ctrl"),Transform=(Rotation=(X=0,Y=0,Z=0,W=1),Translation=(X=0,Y=0,Z=0),Scale3D=(X=1,Y=1,Z=1))))
    Weight=1
   )
   &RigUnit_ModifyTransforms(
    ExecutePin=()
    ItemToModify=((Item=(Type=Control,Name="dyn_straps_back_5_r_ctrl"),Transform=(Rotation=(X=0,Y=0,Z=0,W=1),Translation=(X=0,Y=0,Z=0),Scale3D=(X=1,Y=1,Z=1))))
    Weight=1
   )
   > &RigUnit_ModifyTransforms(
    ItemToModify=((Item=(Type=Control,Name="dyn_straps_front_6_r_ctrl"),Transform=(Rotation=(X=0,Y=0,Z=0,W=1),Translation=(X=0,Y=0,Z=0),Scale3D=(X=1,Y=1,Z=1))))
    Weight=1
   )
   > &RigUnit_ModifyTransforms(
    ItemToModify=((Item=(Type=Control,Name="dyn_straps_front_5_r_ctrl"),Transform=(Rotation=(X=0,Y=0,Z=0,W=1),Translation=(X=0,Y=0,Z=0),Scale3D=(X=1,Y=1,Z=1))))
    Weight=1
   )
   > &RigUnit_ModifyTransforms(
    ItemToModify=((Item=(Type=Control,Name="dyn_straps_front_3_r_ctrl"),Transform=(Rotation=(X=0,Y=0,Z=0,W=1),Translation=(X=0,Y=0,Z=0),Scale3D=(X=1,Y=1,Z=1))))
    Weight=1
   )
   n1 = RigVMVariableNode() {variable=SplineIK_chain_r_ctrls}
   > RigVMVariableNode(Value=/Script/ControlRig.RigUnit_HierarchyGetPoseItemArray(ElementType=All, ItemsToGet=@n1.Value) {is=struct RigVMUnitNode}.Pose) {variable=SplineIK_ctrl_thigh_r_leg_slight_back_out_pose}
   n2 = &RigUnit_PrepareForExecution()
   RigVMFunction_Sequence = &RigVMFunction_Sequence(ExecuteContext=@n2.ExecutePin)
   SequenceAggregate_5 = &RigVMFunction_Sequence(ExecuteContext=@RigVMFunction_Sequence.A)
   SequenceAggregate_3 = &RigVMFunction_Sequence(ExecuteContext=@SequenceAggregate_5.A)
   RerouteNode_13_2 = RigVMDispatchNode(Values="((Type=Control,Name=\"dyn_straps_front_1_l_ctrl\"),((),Type=Control,Name=\"dyn_straps_front_3_l_ctrl\"),((),Type=Control,Name=\"dyn_straps_front_5_l_ctrl\"),((),Type=Control,Name=\"dyn_straps_front_6_l_ctrl\"),((),Type=Control,Name=\"dyn_straps_back_5_l_ctrl\"),((), ...(truncated)") {template="DISPATCH_RigVMDispatch_ArrayMake(in Values,out Array)"}
   RigVMVariableNode(Value=@RerouteNode_13_2.Array, ExecuteContext=@SequenceAggregate_3.A) {variable=SplineIK_chain_l_ctrls}
   RigVMVariableNode(
    Value=&RigUnit_HierarchyGetPoseItemArray(ElementType=All, ItemsToGet=RigVMVariableNode() {variable=SplineIK_chain_l_ctrls}.Value).Pose
    ExecuteContext=@SequenceAggregate_3.B
   ) {variable=SplineIK_ctrls_l_bind_pose}
   SequenceAggregate_4 = &RigVMFunction_Sequence(ExecuteContext=@SequenceAggregate_5.B)
   Multiply = &RigVMFunction_MathDoubleMul(B=0, A=RigVMVariableNode() {variable=spr_thigh_l_up_alpha}.Value)
   &RigUnit_ModifyTransforms(
    ItemToModify=((Item=(Type=Control,Name="dyn_straps_front_3_l_ctrl"),Transform=(Rotation=(X=0,Y=0,Z=0,W=1),Translation=(X=0,Y=0,Z=1),Scale3D=(X=1,Y=1,Z=1))))
    ExecutePin=@SequenceAggregate_4.A
    Weight=@Multiply.Result
   )
   > &RigUnit_ModifyTransforms(
    ItemToModify=((Item=(Type=Control,Name="dyn_straps_front_5_l_ctrl"),Transform=(Rotation=(X=0,Y=0,Z=0,W=1),Translation=(X=0,Y=0,Z=2.5),Scale3D=(X=1,Y=1,Z=1))))
    Weight=@Multiply.Result
   )
   > &RigUnit_ModifyTransforms(
    ItemToModify=((Item=(Type=Control,Name="dyn_straps_back_3_l_ctrl"),Transform=(Rotation=(X=0,Y=0,Z=0,W=1),Translation=(X=0,Y=0,Z=0),Scale3D=(X=1,Y=1,Z=1))))
    Weight=@Multiply.Result
   )
   > &RigUnit_ModifyTransforms(
    ItemToModify=((Item=(Type=Control,Name="dyn_straps_back_5_l_ctrl"),Transform=(Rotation=(X=0,Y=0,Z=0,W=1),Translation=(X=0,Y=0,Z=0),Scale3D=(X=1,Y=1,Z=1))))
    Weight=@Multiply.Result
   )
   > &RigUnit_ModifyTransforms(
    ItemToModify=((Item=(Type=Control,Name="dyn_straps_Main_l_ctrl"),Transform=(Rotation=(X=-0.25,Y=-0.25,Z=-0.067,W=0.933),Translation=(X=-3,Y=0,Z=0),Scale3D=(X=1,Y=1,Z=1))))
    Weight=@Multiply.Result
   )
   > RigVMVariableNode(Value=&RigUnit_HierarchyGetPoseItemArray(ElementType=All, ItemsToGet=RigVMVariableNode() {variable=SplineIK_chain_l_ctrls}.Value).Pose) {variable=SplineIK_ctrl_thigh_l_up_pose}
   &RigUnit_HierarchySetPoseItemArray(
    ElementType=All
    Weight=1
    ExecutePin=@SequenceAggregate_4.B
    ItemsToSet=RigVMVariableNode() {variable=SplineIK_chain_l_ctrls}.Value
    Pose=RigVMVariableNode() {variable=SplineIK_ctrls_l_bind_pose}.Value
   )
   SequenceAggregate = &RigVMFunction_Sequence(ExecuteContext=@RigVMFunction_Sequence.B)
   SequenceAggregate_1 = &RigVMFunction_Sequence(ExecuteContext=@SequenceAggregate.A)
   RerouteNode_13_1 = RigVMDispatchNode(Values="((Type=Control,Name=\"dyn_straps_front_1_r_ctrl\"),((),Type=Control,Name=\"dyn_straps_front_3_r_ctrl\"),((),Type=Control,Name=\"dyn_straps_front_5_r_ctrl\"),((),Type=Control,Name=\"dyn_straps_front_6_r_ctrl\"),((),Type=Control,Name=\"dyn_straps_back_5_r_ctrl\"),((), ...(truncated)") {template="DISPATCH_RigVMDispatch_ArrayMake(in Values,out Array)"}
   RigVMVariableNode(Value=@RerouteNode_13_1.Array, ExecuteContext=@SequenceAggregate_1.A) {variable=SplineIK_chain_r_ctrls}
   RigVMVariableNode(
    Value=&RigUnit_HierarchyGetPoseItemArray(ElementType=All, ItemsToGet=RigVMVariableNode() {variable=SplineIK_chain_r_ctrls}.Value).Pose
    ExecuteContext=@SequenceAggregate_1.B
   ) {variable=SplineIK_ctrls_r_bind_pose}
   SequenceAggregate_2 = &RigVMFunction_Sequence(ExecuteContext=@SequenceAggregate.B)
   n3 = RigVMVariableNode() {variable=spr_thigh_r_up_alpha}
   &RigUnit_ModifyTransforms(
    ItemToModify=((Item=(Type=Control,Name="dyn_straps_front_3_r_ctrl"),Transform=(Rotation=(X=0,Y=0,Z=0,W=1),Translation=(X=0,Y=0,Z=0),Scale3D=(X=1,Y=1,Z=1))))
    ExecutePin=@SequenceAggregate_2.A
    Weight=@n3.Value
   )
   > &RigUnit_ModifyTransforms(
    ItemToModify=((Item=(Type=Control,Name="dyn_straps_front_5_r_ctrl"),Transform=(Rotation=(X=0,Y=0,Z=0,W=1),Translation=(X=0,Y=0,Z=0),Scale3D=(X=1,Y=1,Z=1))))
    Weight=@n3.Value
   )
   > &RigUnit_ModifyTransforms(
    ItemToModify=((Item=(Type=Control,Name="dyn_straps_back_3_r_ctrl"),Transform=(Rotation=(X=0,Y=0,Z=0,W=1),Translation=(X=0,Y=0,Z=20),Scale3D=(X=1,Y=1,Z=1))))
    Weight=@n3.Value
   )
   > &RigUnit_ModifyTransforms(
    ItemToModify=((Item=(Type=Control,Name="dyn_straps_back_5_r_ctrl"),Transform=(Rotation=(X=0,Y=0,Z=0,W=1),Translation=(X=0,Y=0,Z=0),Scale3D=(X=1,Y=1,Z=1))))
    Weight=@n3.Value
   )
   > &RigUnit_ModifyTransforms(
    ItemToModify=((Item=(Type=Control,Name="dyn_straps_Main_r_ctrl"),Transform=(Rotation=(X=0,Y=0,Z=0,W=1),Translation=(X=0,Y=0,Z=0),Scale3D=(X=1,Y=1,Z=1))))
    Weight=@n3.Value
   )
   > RigVMVariableNode(Value=&RigUnit_HierarchyGetPoseItemArray(ElementType=All, ItemsToGet=RigVMVariableNode() {variable=SplineIK_chain_r_ctrls}.Value).Pose) {variable=SplineIK_ctrl_thigh_r_up_pose}
   &RigUnit_HierarchySetPoseItemArray(
    ElementType=All
    Weight=1
    ExecutePin=@SequenceAggregate_2.B
    Pose=RigVMVariableNode() {variable=SplineIK_ctrls_r_bind_pose}.Value
    ItemsToSet=@n1.Value
   )
   RerouteNode_4_ArrayMake_1 = RigVMDispatchNode(Values=((Type=None,Name="None"))) {template="DISPATCH_RigVMDispatch_ArrayMake(in Values,out Array)"}
   &RigUnit_HierarchyGetPoseItemArray(ElementType=All, ItemsToGet=@RerouteNode_4_ArrayMake_1.Array)
   RerouteNode_4_ArrayMake = RigVMDispatchNode(Values=((Type=None,Name="None"))) {template="DISPATCH_RigVMDispatch_ArrayMake(in Values,out Array)"}
  ]}
]
hierarchy [
 root {loc=<0 0 0>, rot=<0 0 0 1>}
 pelvis {parent=root, loc=<0 -0.5301 98.0571>, rot=<0.0127 0.707 -0.0127 -0.707>}
 spine_01 {parent=pelvis, loc=<6.1208 -0.9328 0>, rot=<0 0 0.034 -0.9994>}
 spine_02 {parent=spine_01, loc=<6.5997 -0.2095 0>, rot=<0 0 0.0238 -0.9997>}
 spine_03 {parent=spine_02, loc=<6.4914 -0.1237 0>, rot=<0 0 0.0149 -0.9999>}
 spine_04 {parent=spine_03, loc=<6.9598 0.0794 0>, rot=<0 0 -0.0132 -0.9999>}
 spine_05 {parent=spine_04, loc=<6.9987 0 0>, rot=<0 0 -0.0015 -1>}
 attach_cape {parent=spine_05, loc=<9.017 -11.714 0>, rot=<0.0616 -0.7044 -0.0616 -0.7044>}
 neck_01 {parent=spine_05, loc=<11.3138 0.1529 0>, rot=<0 0 -0.2003 -0.9797>}
 neck_02 {parent=neck_01, loc=<4.5627 0 0>, rot=<0 0 0.0504 -0.9987>}
 head {parent=neck_02, loc=<5.3914 0 0>, rot=<0 0 0.115 -0.9934>}
 attach_backpack {parent=spine_05, loc=<-0.123 -13.473 0>, rot=<-0.0093 -0.707 0.0093 -0.707>}
 clavicle_l {parent=spine_05, loc=<5.4831 -0.1053 -3.2946>, rot=<0.0465 -0.7289 -0.009 -0.6829>}
 upperarm_l {parent=clavicle_l, loc=<9.8658 -1.2136 -0.4105>, rot=<0.0164 -0.3245 -0.0269 -0.9454>}
 lowerarm_l {parent=upperarm_l, loc=<26.465 -0.8345 0.259>, rot=<0.0036 0.0161 -0.2298 -0.9731>}
 hand_l {parent=lowerarm_l, loc=<22.7355 -0.3203 -0.6105>, rot=<-0.6784 -0.0223 0.0735 -0.7307>}
 thumb_01_l {parent=hand_l, loc=<1.3118 -0.3191 -2.459>, rot=<0.6546 -0.4042 0.0664 -0.6354>}
 thumb_02_l {parent=thumb_01_l, loc=<5.9336 -0.0448 0.0401>, rot=<-0.0037 0.0213 0.298 -0.9543>}
 thumb_03_l {parent=thumb_02_l, loc=<2.2223 0.3682 -0.0907>, rot=<0.0118 -0.013 0.0066 -0.9998>}
 index_metacarpal_l {parent=hand_l, loc=<1.6535 0.0183 -2.96>, rot=<0.0158 -0.0124 -0.0173 -0.9996>}
 index_01_l {parent=index_metacarpal_l, loc=<8.3008 0.0357 -0.0524>, rot=<-0.0044 0.0485 0.0924 -0.9945>}
 index_02_l {parent=index_01_l, loc=<3.3587 -0.3004 0.0452>, rot=<-0.002 0.0027 0.3225 -0.9466>}
 index_03_l {parent=index_02_l, loc=<2.5578 -0.123 -0.0252>, rot=<-0.002 0.0067 0.2558 -0.9667>}
 middle_metacarpal_l {parent=hand_l, loc=<2.1856 -0.1725 -0.5984>, rot=<0.006 0.0168 -0.0329 -0.9993>}
 middle_01_l {parent=middle_metacarpal_l, loc=<7.3275 -0.2447 0.0129>, rot=<-0.0208 0.0042 0.0858 -0.9961>}
 middle_02_l {parent=middle_01_l, loc=<4.0123 0.1401 0.1175>, rot=<-0.0218 0.0258 0.3512 -0.9357>}
 middle_03_l {parent=middle_02_l, loc=<2.6077 0.0534 0.1373>, rot=<0.0082 0.0188 0.1708 -0.9851>}
 ring_metacarpal_l {parent=hand_l, loc=<2.2074 -0.3625 1.3638>, rot=<-0.0284 0.0722 -0.0458 -0.9959>}
 ring_01_l {parent=ring_metacarpal_l, loc=<7.0295 -0.6197 -0.1446>, rot=<-0.0489 -0.0252 0.1892 -0.9804>}
 ring_02_l {parent=ring_01_l, loc=<3.6174 0.4647 -0.1027>, rot=<0 0 0.2806 -0.9598>}
 ring_03_l {parent=ring_02_l, loc=<2.4152 0.114 0.0199>, rot=<0 0 0.1189 -0.9929>}
 pinky_metacarpal_l {parent=hand_l, loc=<1.7616 -0.9753 3.3814>, rot=<-0.0397 0.0814 -0.069 -0.9935>}
 pinky_01_l {parent=pinky_metacarpal_l, loc=<7.1709 -0.4831 -0.1094>, rot=<-0.0654 -0.0035 0.2242 -0.9723>}
 pinky_02_l {parent=pinky_01_l, loc=<2.8274 0.0893 -0.1043>, rot=<0.0037 0.0011 0.2854 -0.9584>}
 pinky_03_l {parent=pinky_02_l, loc=<2.2046 0.0307 -0.0078>, rot=<-0.0002 0.0098 0.1445 -0.9895>}
 deform_wrist_dn_base_l {parent=hand_l, loc=<0 -1.5326 0>, rot=<0 0 0 -1>}
 deform_wrist_dn_l {parent=deform_wrist_dn_base_l, loc=<0 0 0>, rot=<0 0 0 1>}
 deform_wrist_up_base_l {parent=hand_l, loc=<0 2.3009 0>, rot=<0 0 0 -1>}
 deform_wrist_up_l {parent=deform_wrist_up_base_l, loc=<0 0 0>, rot=<0 0 0 1>}
 lowerarm_twist_02_l {parent=lowerarm_l, loc=<10.6707 0 0>, rot=<0 0 0 1>}
 lowerarm_twist_01_l {parent=lowerarm_l, loc=<18.316 0 0>, rot=<0 0 0 1>}
 deform_elbow_in_base_l {parent=lowerarm_l, loc=<0.1834 3.6062 -0.2572>, rot=<0 0 0 -1>}
 deform_elbow_in_l {parent=deform_elbow_in_base_l, loc=<0 0 0>, rot=<0 0 0 1>}
 deform_elbow_out_base_l {parent=lowerarm_l, loc=<0 -4 0>, rot=<0 0 0 1>}
 deform_elbow_out_l {parent=deform_elbow_out_base_l, loc=<0 0 0>, rot=<0 0 0 1>}
 upperarm_twist_01_l {parent=upperarm_l, loc=<7.2135 -0.1134 0>, rot=<0 0 0 1>}
 deform_upperarm_bk_base_l {parent=upperarm_twist_01_l, loc=<-6.4053 -4.3746 0.314>, rot=<0 0 0 -1>}
 deform_upperarm_bk_l {parent=deform_upperarm_bk_base_l, loc=<0 0 0>, rot=<0 0 0 1>}
 deform_upperarm_fr_base_l {parent=upperarm_twist_01_l, loc=<-5.6312 5.092 0.3354>, rot=<0 0 0 -1>}
 deform_upperarm_fr_l {parent=deform_upperarm_fr_base_l, loc=<0 0 0>, rot=<0 0 0 1>}
 deform_upperarm_up_base_l {parent=upperarm_twist_01_l, loc=<-8.891 0.4825 3.9517>, rot=<0 0 0 -1>}
 deform_upperarm_up_l {parent=deform_upperarm_up_base_l, loc=<0 0 0>, rot=<0 0 0 1>}
 upperarm_twist_02_l {parent=upperarm_l, loc=<19.7998 -0.1134 0>, rot=<0 0 0 1>}
 clavicle_r {parent=spine_05, loc=<5.4831 -0.1053 3.2946>, rot=<0.7289 0.0465 -0.6829 0.009>}
 upperarm_r {parent=clavicle_r, loc=<-9.8658 1.2136 0.4105>, rot=<0.0164 -0.3245 -0.0269 -0.9454>}
 lowerarm_r {parent=upperarm_r, loc=<-26.465 0.8345 -0.259>, rot=<0.0036 0.0161 -0.2298 -0.9731>}
 hand_r {parent=lowerarm_r, loc=<-22.7355 0.3203 0.6106>, rot=<-0.6784 -0.0223 0.0735 -0.7307>}
 thumb_01_r {parent=hand_r, loc=<-1.3118 0.3191 2.459>, rot=<0.6546 -0.4042 0.0664 -0.6354>}
 thumb_02_r {parent=thumb_01_r, loc=<-5.9336 0.0448 -0.0401>, rot=<-0.0037 0.0213 0.298 -0.9543>}
 thumb_03_r {parent=thumb_02_r, loc=<-2.2223 -0.3682 0.0907>, rot=<0.0118 -0.013 0.0066 -0.9998>}
 index_metacarpal_r {parent=hand_r, loc=<-1.6535 -0.0183 2.96>, rot=<0.0158 -0.0124 -0.0173 -0.9996>}
 index_01_r {parent=index_metacarpal_r, loc=<-8.3008 -0.0357 0.0524>, rot=<-0.0044 0.0485 0.0924 -0.9945>}
 index_02_r {parent=index_01_r, loc=<-3.3587 0.3004 -0.0452>, rot=<-0.002 0.0027 0.3225 -0.9466>}
 index_03_r {parent=index_02_r, loc=<-2.5578 0.123 0.0252>, rot=<-0.002 0.0067 0.2558 -0.9667>}
 middle_metacarpal_r {parent=hand_r, loc=<-2.1856 0.1725 0.5984>, rot=<0.006 0.0168 -0.0329 -0.9993>}
 middle_01_r {parent=middle_metacarpal_r, loc=<-7.3275 0.2447 -0.0129>, rot=<-0.0208 0.0042 0.0858 -0.9961>}
 middle_02_r {parent=middle_01_r, loc=<-4.0123 -0.1401 -0.1175>, rot=<-0.0218 0.0258 0.3512 -0.9357>}
 middle_03_r {parent=middle_02_r, loc=<-2.6077 -0.0534 -0.1373>, rot=<0.0082 0.0188 0.1708 -0.9851>}
 ring_metacarpal_r {parent=hand_r, loc=<-2.2074 0.3625 -1.3638>, rot=<-0.0284 0.0722 -0.0458 -0.9959>}
 ring_01_r {parent=ring_metacarpal_r, loc=<-7.0295 0.6197 0.1446>, rot=<-0.0489 -0.0252 0.1892 -0.9804>}
 ring_02_r {parent=ring_01_r, loc=<-3.6174 -0.4647 0.1027>, rot=<0 0 0.2806 -0.9598>}
 ring_03_r {parent=ring_02_r, loc=<-2.4152 -0.114 -0.0199>, rot=<0 0 0.1189 -0.9929>}
 pinky_metacarpal_r {parent=hand_r, loc=<-1.7616 0.9753 -3.3814>, rot=<-0.0397 0.0814 -0.069 -0.9935>}
 pinky_01_r {parent=pinky_metacarpal_r, loc=<-7.1709 0.4831 0.1094>, rot=<-0.0654 -0.0035 0.2242 -0.9723>}
 pinky_02_r {parent=pinky_01_r, loc=<-2.8274 -0.0893 0.1043>, rot=<0.0037 0.0011 0.2854 -0.9584>}
 pinky_03_r {parent=pinky_02_r, loc=<-2.2046 -0.0307 0.0078>, rot=<-0.0002 0.0098 0.1445 -0.9895>}
 deform_wrist_dn_base_r {parent=hand_r, loc=<-0.0001 1.5328 0>, rot=<0 0 0 -1>}
 deform_wrist_dn_r {parent=deform_wrist_dn_base_r, loc=<0 0 0>, rot=<0 0 0 1>}
 deform_wrist_up_base_r {parent=hand_r, loc=<-0.0002 -2.3007 -0.0001>, rot=<0 0 0 -1>}
 deform_wrist_up_r {parent=deform_wrist_up_base_r, loc=<0 0 0>, rot=<0 0 0 1>}
 lowerarm_twist_02_r {parent=lowerarm_r, loc=<-10.6707 0 0>, rot=<0 0 0 1>}
 lowerarm_twist_01_r {parent=lowerarm_r, loc=<-18.316 0 0>, rot=<0 0 0 1>}
 deform_elbow_in_base_r {parent=lowerarm_r, loc=<-0.1834 -3.6062 0.2572>, rot=<0 0 0 -1>}
 deform_elbow_in_r {parent=deform_elbow_in_base_r, loc=<0 0 0>, rot=<0 0 0 1>}
 deform_elbow_out_base_r {parent=lowerarm_r, loc=<0 4 -0.0001>, rot=<0 0 0 -1>}
 deform_elbow_out_r {parent=deform_elbow_out_base_r, loc=<0 0 0>, rot=<0 0 0 1>}
 upperarm_twist_01_r {parent=upperarm_r, loc=<-7.2135 0.1134 0>, rot=<0 0 0 1>}
 deform_upperarm_bk_base_r {parent=upperarm_twist_01_r, loc=<6.4053 4.3746 -0.314>, rot=<0 0 0 -1>}
 deform_upperarm_bk_r {parent=deform_upperarm_bk_base_r, loc=<0 0 0>, rot=<0 0 0 1>}
 deform_upperarm_fr_base_r {parent=upperarm_twist_01_r, loc=<5.6312 -5.092 -0.3354>, rot=<0 0 0 -1>}
 deform_upperarm_fr_r {parent=deform_upperarm_fr_base_r, loc=<0 0 0>, rot=<0 0 0 1>}
 deform_upperarm_up_base_r {parent=upperarm_twist_01_r, loc=<8.891 -0.4825 -3.9517>, rot=<0 0 0 -1>}
 deform_upperarm_up_r {parent=deform_upperarm_up_base_r, loc=<0 0 0>, rot=<0 0 0 1>}
 upperarm_twist_02_r {parent=upperarm_r, loc=<-19.7998 0.1134 0>, rot=<0 0 0 1>}
 dyn_necklace_1 {parent=spine_05, loc=<8.6 3.7185 -0.0042>, rot=<-0.2161 -0.7308 0.6206 0.1847>}
 dyn_necklace_2 {parent=dyn_necklace_1, loc=<6.7991 0 0>, rot=<0 0.0102 0 -0.9999>}
 deform_pec_aim_base_l {parent=spine_04, loc=<3.2456 9.3309 -1.2237>, rot=<0.0603 -0.4033 0.3318 -0.8507>}
 deform_pec_aim_l {parent=deform_pec_aim_base_l, loc=<0 0 0>, rot=<0 0 0 1>}
 deform_pec_base_l {parent=deform_pec_aim_l, loc=<9.3665 0 0>, rot=<0 0 0 1>}
 deform_pec_l {parent=deform_pec_base_l, loc=<0 0 0>, rot=<0 0 0 1>}
 deform_pec_aim_base_r {parent=spine_04, loc=<3.2452 9.3309 1.2237>, rot=<-0.4033 -0.0603 0.8507 0.3318>}
 deform_pec_aim_r {parent=deform_pec_aim_base_r, loc=<0 0 0>, rot=<0 0 0 1>}
 deform_pec_base_r {parent=deform_pec_aim_r, loc=<-9.3667 -0.0003 -0.0004>, rot=<0 0 0 1>}
 deform_pec_r {parent=deform_pec_base_r, loc=<0 0 0>, rot=<0 0 0 1>}
 thigh_l {parent=pelvis, loc=<-1.0855 -0.2109 -8.7053>, rot=<0.0009 0.0355 0.0155 -0.9993>}
 calf_l {parent=thigh_l, loc=<-42.5284 -0.0019 -0.7297>, rot=<0.0006 -0.0161 -0.0391 -0.9991>}
 foot_l {parent=calf_l, loc=<-46.7401 -2.5014 -0.592>, rot=<0.0427 -0.0238 0.0353 -0.9982>}
 ball_l {parent=foot_l, loc=<-3.5658 10.8066 -0.1544>, rot=<0.0001 -0.0001 -0.7063 -0.7079>}
 calf_twist_02_l {parent=calf_l, loc=<-19.1546 -1.7943 -0.2638>, rot=<0.0237 0.0095 -0.0035 -0.9997>}
 calf_twist_01_l {parent=calf_l, loc=<-35.8483 -1.9197 -0.5794>, rot=<0.0237 0.0095 -0.0035 -0.9997>}
 deform_knee_in_base_l {parent=calf_l, loc=<0.8031 -5.0231 0.2456>, rot=<0 0 0 -1>}
 deform_knee_in_l {parent=deform_knee_in_base_l, loc=<0 0 0>, rot=<0 0 0 1>}
 deform_knee_out_base_l {parent=calf_l, loc=<0 4 0>, rot=<0 0 0 -1>}
 deform_knee_out_l {parent=deform_knee_out_base_l, loc=<0 0 0>, rot=<0 0 0 1>}
 thigh_twist_01_l {parent=thigh_l, loc=<-27.952 0 0>, rot=<0.0013 0 0 -1>}
 deform_groin_base_l {parent=thigh_l, loc=<0.4801 5.6776 -0.2486>, rot=<0 0 0 -1>}
 deform_groin_l {parent=deform_groin_base_l, loc=<0 0 0>, rot=<0 0 0 1>}
 dyn_pocket_l {parent=thigh_l, loc=<-23.1503 0.3674 -9.8258>, rot=<0.0399 -0.9911 0.0055 -0.1267>}
 thigh_r {parent=pelvis, loc=<-1.0855 -0.2109 8.7053>, rot=<0.0355 -0.0009 0.9993 0.0155>}
 calf_r {parent=thigh_r, loc=<42.5284 0.0019 0.7297>, rot=<0.0006 -0.0161 -0.0391 -0.9991>}
 foot_r {parent=calf_r, loc=<46.7401 2.5014 0.592>, rot=<0.0427 -0.0238 0.0353 -0.9982>}
 ball_r {parent=foot_r, loc=<3.5658 -10.8066 0.1544>, rot=<0.0001 -0.0001 -0.7063 -0.7079>}
 calf_twist_02_r {parent=calf_r, loc=<19.1546 1.7943 0.2638>, rot=<0.0237 0.0095 -0.0035 -0.9997>}
 calf_twist_01_r {parent=calf_r, loc=<35.8483 1.9197 0.5794>, rot=<0.0237 0.0095 -0.0035 -0.9997>}
 deform_knee_in_base_r {parent=calf_r, loc=<-0.8031 5.0231 -0.2457>, rot=<0 0 0 -1>}
 deform_knee_in_r {parent=deform_knee_in_base_r, loc=<0 0 0>, rot=<0 0 0 1>}
 deform_knee_out_base_r {parent=calf_r, loc=<0 -4 0>, rot=<0 0 0 -1>}
 deform_knee_out_r {parent=deform_knee_out_base_r, loc=<0 0 0>, rot=<0 0 0 1>}
 thigh_twist_01_r {parent=thigh_r, loc=<27.952 0 0>, rot=<0.0013 0 0 -1>}
 deform_groin_base_r {parent=thigh_r, loc=<-0.4801 -5.6776 0.2486>, rot=<0 0 0 -1>}
 deform_groin_r {parent=deform_groin_base_r, loc=<0 0 0>, rot=<0 0 0 1>}
 dyn_pocket_r {parent=thigh_r, loc=<23.1503 -0.3674 9.8258>, rot=<-0.1733 0.1175 -0.0621 -0.9759>}
 deform_glute_base_l {parent=pelvis, loc=<-7.3466 -9.0213 -9.1245>, rot=<0.0172 0.031 -0.4477 -0.8935>}
 deform_glute_l {parent=deform_glute_base_l, loc=<0 0 0>, rot=<0 0 0 1>}
 deform_glute_base_r {parent=pelvis, loc=<-7.3464 -9.0213 9.1245>, rot=<0.031 -0.0172 0.8935 -0.4477>}
 deform_glute_r {parent=deform_glute_base_r, loc=<0 0 0>, rot=<0 0 0 1>}
 dyn_bear_pendant_base {parent=pelvis, loc=<5.7285 -0.9163 14.4484>, rot=<0.0785 -0.975 -0.0289 0.206>}
 dyn_bear_pendant_1 {parent=dyn_bear_pendant_base, loc=<0 0 0>, rot=<0.1271 -0.0366 0.005 -0.9912>}
 dyn_bear_pendant_2 {parent=dyn_bear_pendant_1, loc=<2.3091 0.0688 0.078>, rot=<-0.2581 -0.0559 -0.0371 -0.9638>}
 dyn_bear_pendant_3 {parent=dyn_bear_pendant_2, loc=<2.4569 0 0>, rot=<0 0 0 1>}
 dyn_straps_lock_front_1_l {parent=pelvis, loc=<4.832 7.3765 -8.7761>, rot=<-0.0647 -0.9608 0.2626 -0.0609>}
 dyn_straps_lock_front_2_l {parent=dyn_straps_lock_front_1_l, loc=<3.4816 0.6114 0>, rot=<0 0 0 1>}
 dyn_straps_front_base_l {parent=dyn_straps_lock_front_2_l, loc=<0 0 0>, rot=<0 0 0 1>}
 dyn_straps_front_1_l {parent=dyn_straps_front_base_l, loc=<0 0 0>, rot=<0.269 0.1065 -0.0304 -0.9568>}
 dyn_straps_front_2_l {parent=dyn_straps_front_1_l, loc=<5.6246 0 0>, rot=<-0.4686 0.0205 0.0099 -0.8831>}
 dyn_straps_front_3_l {parent=dyn_straps_front_2_l, loc=<5.6472 0 0>, rot=<0.0143 0.0994 -0.0014 -0.9949>}
 dyn_straps_front_4_l {parent=dyn_straps_front_3_l, loc=<5.9598 0 0>, rot=<-0.2241 0.1364 0.0317 -0.9644>}
 dyn_straps_front_5_l {parent=dyn_straps_front_4_l, loc=<5.1786 0 0>, rot=<-0.0731 0.3661 0.0289 -0.9273>}
 dyn_straps_front_6_l {parent=dyn_straps_front_5_l, loc=<6.2806 0 0>, rot=<0 0 0 1>}
 dyn_straps_lock_back_1_l {parent=pelvis, loc=<6.5613 -9.4518 -8.4863>, rot=<-0.0494 -0.247 0.9561 -0.1499>}
 dyn_straps_lock_back_2_l {parent=dyn_straps_lock_back_1_l, loc=<3.5244 0.8068 0>, rot=<0 0 0 1>}
 dyn_straps_back_base_l {parent=dyn_straps_lock_back_2_l, loc=<0 0 0>, rot=<0 0 0 1>}
 dyn_straps_back_1_l {parent=dyn_straps_back_base_l, loc=<0 0 0>, rot=<0.9558 0.0294 -0.0636 -0.2854>}
 dyn_straps_back_2_l {parent=dyn_straps_back_1_l, loc=<7.0829 0 0>, rot=<0.6018 0.1685 -0.1317 -0.7695>}
 dyn_straps_back_3_l {parent=dyn_straps_back_2_l, loc=<7.0076 0 0>, rot=<-0.0356 0.137 0.0049 -0.9899>}
 dyn_straps_back_4_l {parent=dyn_straps_back_3_l, loc=<5.7487 0 0>, rot=<0.2203 0.1649 -0.0378 -0.9607>}
 dyn_straps_back_5_l {parent=dyn_straps_back_4_l, loc=<6.4775 0 0>, rot=<-0.1214 0.2327 0.0293 -0.9645>}
 dyn_straps_back_6_l {parent=dyn_straps_back_5_l, loc=<6.9694 0 0>, rot=<0 0 0 1>}
 dyn_straps_base_l {parent=pelvis, loc=<2.3604 -0.9072 -10.4024>, rot=<-0.0355 -0.9767 0.004 -0.2118>}
 dyn_straps_l {parent=dyn_straps_base_l, loc=<0 0 0>, rot=<0 0 0 1>}
 dyn_straps_2_l {parent=dyn_straps_l, loc=<11.5944 0 -0.0055>, rot=<0 0 0 1>}
 dyn_straps_end_l {parent=dyn_straps_2_l, loc=<13.4225 0 -0.0062>, rot=<0 0 0 1>}
 dyn_straps_base_r {parent=pelvis, loc=<2.3609 -0.9072 10.4024>, rot=<0.2107 -0.0319 0.9764 0.0353>}
 dyn_straps_r {parent=dyn_straps_base_r, loc=<0 0 0>, rot=<0 0 0 1>}
 dyn_straps_2_r {parent=dyn_straps_r, loc=<11.594 -0.1428 -0.0095>, rot=<0 0 0 1>}
 dyn_straps_end_r {parent=dyn_straps_2_r, loc=<13.4214 -0.1653 -0.0106>, rot=<0 0 0 1>}
 dyn_straps_lock_front_1_r {parent=pelvis, loc=<4.8322 7.3765 8.7761>, rot=<-0.0648 -0.9608 -0.2626 0.0609>}
 dyn_straps_lock_front_2_r {parent=dyn_straps_lock_front_1_r, loc=<3.4818 0.6113 0.0001>, rot=<0 0 0 1>}
 dyn_straps_front_base_r {parent=dyn_straps_lock_front_2_r, loc=<0 0 0>, rot=<0 0 0 1>}
 dyn_straps_front_1_r {parent=dyn_straps_front_base_r, loc=<0 0 0>, rot=<-0.269 -0.1065 -0.0303 -0.9568>}
 dyn_straps_front_2_r {parent=dyn_straps_front_1_r, loc=<5.6246 0 0>, rot=<0.4684 -0.0192 0.0102 -0.8832>}
 dyn_straps_front_3_r {parent=dyn_straps_front_2_r, loc=<5.6472 0 0>, rot=<-0.0145 -0.0994 -0.0014 -0.9949>}
 dyn_straps_front_4_r {parent=dyn_straps_front_3_r, loc=<5.9598 0 0>, rot=<0.2242 -0.1364 0.0317 -0.9644>}
 dyn_straps_front_5_r {parent=dyn_straps_front_4_r, loc=<5.1786 0 0>, rot=<0.073 -0.3661 0.0288 -0.9273>}
 dyn_straps_front_6_r {parent=dyn_straps_front_5_r, loc=<6.2806 0 0>, rot=<0 0 0 1>}
 dyn_straps_lock_back_1_r {parent=pelvis, loc=<6.5615 -9.4518 8.4862>, rot=<0.0494 0.247 0.9561 -0.1498>}
 dyn_straps_lock_back_2_r {parent=dyn_straps_lock_back_1_r, loc=<3.5245 0.8069 0>, rot=<0 0 0 1>}
 dyn_straps_back_base_r {parent=dyn_straps_lock_back_2_r, loc=<0 0 0>, rot=<0 0 0 1>}
 dyn_straps_back_1_r {parent=dyn_straps_back_base_r, loc=<0 0 0>, rot=<0.9558 0.0294 0.0636 0.2854>}
 dyn_straps_back_2_r {parent=dyn_straps_back_1_r, loc=<7.0829 0 0>, rot=<-0.6018 -0.1685 -0.1317 -0.7695>}
 dyn_straps_back_3_r {parent=dyn_straps_back_2_r, loc=<7.0076 0 0>, rot=<0.0356 -0.137 0.0049 -0.9899>}
 dyn_straps_back_4_r {parent=dyn_straps_back_3_r, loc=<5.7486 0 0>, rot=<-0.2204 -0.1649 -0.0378 -0.9606>}
 dyn_straps_back_5_r {parent=dyn_straps_back_4_r, loc=<6.4775 0 0>, rot=<0.1215 -0.2327 0.0293 -0.9645>}
 dyn_straps_back_6_r {parent=dyn_straps_back_5_r, loc=<6.9694 0 0>, rot=<0 0 0 1>}
 dyn_straps_front_1_l_null {type=Null, loc=<8.8687 6.583 99.6725>, rot=<-0.6025 -0.411 0.3856 -0.5652>}
 dyn_straps_front_1_l_ctrl {controlType=EULER_TRANSFORM, parent=dyn_straps_front_1_l_null, type=Control, loc=<0 0 0>, rot=<0 0 0 1>}
 dyn_straps_front_3_l_null {type=Null, loc=<13.1118 7.0491 89.2431>, rot=<-0.4141 -0.4124 0.3184 -0.7464>}
 dyn_straps_front_3_l_ctrl {controlType=EULER_TRANSFORM, parent=dyn_straps_front_3_l_null, type=Control, loc=<0 0 0>, rot=<0 0 0 1>}
 dyn_straps_front_5_l_null {type=Null, loc=<17.8565 4.1474 79.7679>, rot=<-0.3192 -0.0338 0.6174 -0.7182>}
 dyn_straps_front_5_l_ctrl {controlType=EULER_TRANSFORM, parent=dyn_straps_front_5_l_null, type=Control, loc=<0 0 0>, rot=<0 0 0 1>}
 dyn_straps_back_1_l_null {type=Null, loc=<8.3601 -9.4973 101.9859>, rot=<-0.6312 0.4492 0.3666 0.5152>}
 dyn_straps_back_1_l_ctrl {controlType=EULER_TRANSFORM, parent=dyn_straps_back_1_l_null, type=Control, loc=<0 0 0>, rot=<0 0 0 1>}
 dyn_straps_back_3_l_null {type=Null, loc=<13.4358 -10.1016 89.3967>, rot=<-0.5073 0.3092 0.4468 0.6689>}
 dyn_straps_back_3_l_ctrl {controlType=EULER_TRANSFORM, parent=dyn_straps_back_3_l_null, type=Control, loc=<0 0 0>, rot=<0 0 0 1>}
 dyn_straps_back_5_l_null {type=Null, loc=<18.1584 -5.4902 79.7637>, rot=<-0.3408 -0.0141 0.6441 0.6847>}
 dyn_straps_back_5_l_ctrl {controlType=EULER_TRANSFORM, parent=dyn_straps_back_5_l_null, type=Control, loc=<0 0 0>, rot=<0 0 0 1>}
 dyn_straps_Main_l_null {type=Null, loc=<19.1137 -0.4747 77.4027>, rot=<0.0114 0.6312 0.014 0.7754>}
 dyn_straps_Main_l_ctrl {controlType=EULER_TRANSFORM, parent=dyn_straps_Main_l_null, type=Control, loc=<0 0 0>, rot=<0 0 0 1>}
 dyn_straps_front_6_l_ctrl {controlType=EULER_TRANSFORM, parent=dyn_straps_Main_l_ctrl, type=Control, loc=<0 0 0>, rot=<0 0 0 1>}
 dyn_straps_back_6_l_ctrl {controlType=EULER_TRANSFORM, parent=dyn_straps_Main_l_ctrl, type=Control, loc=<0 0 0>, rot=<0 0 0 1>}
 dyn_straps_front_1_r_null {type=Null, loc=<-8.8687 6.583 99.6725>, rot=<0.411 0.6025 -0.5652 0.3856>}
 dyn_straps_front_1_r_ctrl {controlType=EULER_TRANSFORM, parent=dyn_straps_front_1_r_null, type=Control, loc=<0 0 0>, rot=<0 0 0 1>}
 dyn_straps_front_3_r_null {type=Null, loc=<-13.1118 7.0491 89.2431>, rot=<0.4124 0.414 -0.7464 0.3184>}
 dyn_straps_front_3_r_ctrl {controlType=EULER_TRANSFORM, parent=dyn_straps_front_3_r_null, type=Control, loc=<0 0 0>, rot=<0 0 0 1>}
 dyn_straps_front_5_r_null {type=Null, loc=<-17.8565 4.1474 79.7679>, rot=<0.0339 0.3192 -0.7182 0.6174>}
 dyn_straps_front_5_r_ctrl {controlType=EULER_TRANSFORM, parent=dyn_straps_front_5_r_null, type=Control, loc=<0 0 0>, rot=<0 0 0 1>}
 dyn_straps_back_1_r_null {type=Null, loc=<-8.3601 -9.4973 101.986>, rot=<-0.4492 0.6312 0.5152 0.3666>}
 dyn_straps_back_1_r_ctrl {controlType=EULER_TRANSFORM, parent=dyn_straps_back_1_r_null, type=Control, loc=<0 0 0>, rot=<0 0 0 1>}
 dyn_straps_back_3_r_null {type=Null, loc=<-13.4358 -10.1016 89.3967>, rot=<-0.3092 0.5073 0.6689 0.4468>}
 dyn_straps_back_3_r_ctrl {controlType=EULER_TRANSFORM, parent=dyn_straps_back_3_r_null, type=Control, loc=<0 0 0>, rot=<0 0 0 1>}
 dyn_straps_back_5_r_null {type=Null, loc=<-18.1584 -5.4902 79.7637>, rot=<0.0142 0.3409 0.6846 0.6441>}
 dyn_straps_back_5_r_ctrl {controlType=EULER_TRANSFORM, parent=dyn_straps_back_5_r_null, type=Control, loc=<0 0 0>, rot=<0 0 0 1>}
 dyn_straps_Main_r_null {type=Null, loc=<-19.1137 -0.4747 77.4027>, rot=<0.6313 0.0331 -0.7748 0.0037>}
 dyn_straps_Main_r_ctrl {controlType=EULER_TRANSFORM, parent=dyn_straps_Main_r_null, type=Control, loc=<0 0 0>, rot=<0 0 0 1>}
 dyn_straps_front_6_r_ctrl {controlType=EULER_TRANSFORM, parent=dyn_straps_Main_r_ctrl, type=Control, loc=<0 0 0>, rot=<0 0 0 1>}
 dyn_straps_back_6_r_ctrl {controlType=EULER_TRANSFORM, parent=dyn_straps_Main_r_ctrl, type=Control, loc=<0 0 0>, rot=<0 0 0 1>}
]
variables [
 Dyn_move_alfa_l {type=float, default=0}
 Dyn_move_alfa_r {type=float, default=0}
 Dyn_pelvis_rotate_alfa {type=float, default=0}
 SplineIK_chain_l_ctrls {type="TArray<FRigElementKey>"}
 SplineIK_chain_r_ctrls {type="TArray<FRigElementKey>"}
 SplineIK_ctrl_thigh_l_up_pose {type=FRigPose, default=(Elements=,HierarchyTopologyVersion=-1,PoseHash=-1)}
 SplineIK_ctrl_thigh_r_leg_slight_back_out_pose {type=FRigPose, default=(Elements=,HierarchyTopologyVersion=-1,PoseHash=-1)}
 SplineIK_ctrl_thigh_r_up_pose {type=FRigPose, default=(Elements=,HierarchyTopologyVersion=-1,PoseHash=-1)}
 SplineIK_ctrls_l_bind_pose {type=FRigPose, default=(Elements=,HierarchyTopologyVersion=-1,PoseHash=-1)}
 SplineIK_ctrls_r_bind_pose {type=FRigPose, default=(Elements=,HierarchyTopologyVersion=-1,PoseHash=-1)}
 is_DBNO {type=bool, default=False}
 is_crouch {type=bool, default=False}
 is_emote {type=bool, default=False}
 is_fly {type=bool, default=False}
 is_front_end {type=bool, default=False}
 spr_thigh_l_slight_back_out_apha {type=double, default=0}
 spr_thigh_l_up_alpha {type=double, default=0}
 spr_thigh_r_slight_back_out_apha {type=double, default=0}
 spr_thigh_r_up_alpha {type=double, default=0}
]
