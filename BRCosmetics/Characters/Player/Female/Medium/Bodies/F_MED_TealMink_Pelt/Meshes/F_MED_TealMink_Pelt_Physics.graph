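# Physics asset for the F_MED_TealMink_Pelt female player body cosmetic, setting up secondary motion for a bear pendant and left and right straps.
# The pendant is a three-link chain from its base body, and each strap a two-link chain from its own base, joined by swing/twist-limited constraints that differ per side.
# Strap links use paired angled capsules.
# Unconstrained pelvis and thigh capsules provide body volumes for the dangling parts.
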PhysicsAsset /BRCosmetics/Characters/Player/Female/Medium/Bodies/F_MED_TealMink_Pelt/Meshes/F_MED_TealMink_Pelt_Physics
bodies {
 dyn_bear_pendant_1 [{
   shapeType=Capsule
   length=1.4114
   radius=0.8632
   shapeName=dyn_bear_pendant_1_capsule
   center=<1.7193 0.0806 0.0014>
   rotation=<rot 0.408 -91.9785 -90>
  }]
 dyn_bear_pendant_2 [{shapeType=Capsule, length=1, radius=0.75, shapeName=dyn_bear_pendant_2_capsule, center=<1.0149 0 0>, rotation=<rot 0.0003 -89.9993 89.9999>}]
 dyn_bear_pendant_3 [{
   shapeType=Capsule
   length=2
   radius=2.9801
   shapeName=dyn_bear_pendant_3_capsule
   center=<2.1993 0.5626 -1.322>
   rotation=<rot 0 -173.452 -80.7012>
  }]
 dyn_bear_pendant_base [{shapeType=Capsule, length=0, radius=2.005, shapeName=dyn_bear_pendant_base_capsule, rotation=<rot 0 -90 0>}]
 dyn_straps_2_l [
  {
   shapeType=Capsule
   length=18
   radius=4.25
   shapeName=dyn_straps_2_l_capsule
   center=<9.2498 -6.0391 0.6452>
   rotation=<rot 85.0001 40.0012 179.999>
  }
  {
   shapeType=Capsule
   length=16.75
   radius=5.25
   shapeName=dyn_straps_2_l_capsule
   center=<7.0962 6.393 -1.6251>
   rotation=<rot -65.707 -50.0725 35.6896>
  }
 ]
 dyn_straps_2_r [
  {
   shapeType=Capsule
   length=15.25
   radius=3.25
   shapeName=dyn_straps_2_r_capsule
   center=<7.7435 7.4417 1.7224>
   rotation=<rot 85 -30.0015 0.0002>
  }
  {
   shapeType=Capsule
   length=17.25
   radius=5.75
   shapeName=dyn_straps_2_r_capsule
   center=<7.6978 -8.9375 0.0859>
   rotation=<rot -18.9941 94.1321 110.4867>
  }
 ]
 dyn_straps_base_l [{shapeType=Capsule, length=0, radius=3.5, shapeName=dyn_straps_base_l_capsule, rotation=<rot 0 -90 0>}]
 dyn_straps_base_r [{shapeType=Capsule, length=0, radius=3.5, shapeName=dyn_straps_base_r_capsule, rotation=<rot 0 -90 0>}]
 dyn_straps_l [
  {
   shapeType=Capsule
   length=16
   radius=3.75
   shapeName=dyn_straps_l_capsule
   center=<10.0623 -7.285 4.7585>
   rotation=<rot -79.9999 90.0006 -180>
  }
  {
   shapeType=Capsule
   length=12.5
   radius=4.25
   shapeName=dyn_straps_l_capsule
   center=<7.6576 6.6139 -0.0398>
   rotation=<rot -7.053 7.108 -45.439>
  }
 ]
 dyn_straps_r [
  {
   shapeType=Capsule
   length=13.75
   radius=3.5
   shapeName=dyn_straps_r_capsule
   center=<11.6257 -6.8569 -0.3189>
   rotation=<rot -30.0003 89.9991 180>
  }
  {
   shapeType=Capsule
   length=15.5
   radius=2.5
   shapeName=dyn_straps_r_capsule
   center=<11.5154 5.2433 4.3087>
   rotation=<rot -79.9999 -90.0005 -180>
  }
 ]
 pelvis [{shapeType=Capsule, length=10.8378, radius=11.6134, center=<-5.0226 -3.0278 0.4655>, rotation=<rot 1.2652 -90 -0.6152>}]
 thigh_l [
  {shapeType=Capsule, length=33.4746, radius=8.4811, center=<-21.3801 0.8252 -0.1869>, rotation=<rot 0.8676 87.8275 -86.1644>}
  {shapeType=Capsule, length=12.5, radius=6.5, shapeName=thigh_l_capsule, center=<-8.1741 -5.245 -0.8064>, rotation=<rot -90 -0.0009 0.0004>}
 ]
 thigh_r [{shapeType=Capsule, length=31.3372, radius=8.8729, center=<20.0219 -0.8683 1.0764>, rotation=<rot -0.0001 88.481 -86.6298>}]
}
constraints [
 {bone1Name=dyn_bear_pendant_1, bone2Name=dyn_bear_pendant_base, swing1LimitDegrees=30, swing2LimitDegrees=30, twistLimitDegrees=5}
 {bone1Name=dyn_bear_pendant_2, bone2Name=dyn_bear_pendant_1, swing1LimitDegrees=15, swing2LimitDegrees=30, twistLimitDegrees=5}
 {bone1Name=dyn_bear_pendant_3, bone2Name=dyn_bear_pendant_2, swing1LimitDegrees=45, swing2LimitDegrees=45, twistLimitDegrees=45}
 {bone1Name=dyn_straps_r, bone2Name=dyn_straps_base_r, swing1LimitDegrees=30, swing2LimitDegrees=60, twistLimitDegrees=20}
 {bone1Name=dyn_straps_l, bone2Name=dyn_straps_base_l, swing1LimitDegrees=30, swing2LimitDegrees=30, twistLimitDegrees=30}
 {bone1Name=dyn_straps_2_l, bone2Name=dyn_straps_l, swing1LimitDegrees=30, swing2LimitDegrees=25, twistLimitDegrees=30}
 {bone1Name=dyn_straps_2_r, bone2Name=dyn_straps_r, swing1LimitDegrees=30, swing2LimitDegrees=30, twistLimitDegrees=5}
]
