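# Post-process style AnimBlueprint for the female TealMink Pelt body cosmetic, adding secondary motion over the pose copied from the attached parent mesh.
# Constraints and crouch offsets place the strap bases, then AnimDynamics drives necklace and pocket bones.
# A RigidBody simulation, front-end backpack nudge and state-driven Control Rig finish the pose.
# The event graph strengthens RigidBody gravity while skydiving and sets Dyn_crouch_move when crouched and stationary, though the AnimGraph never reads it.
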AnimBlueprint /BRCosmetics/Characters/Player/Female/Medium/Bodies/F_MED_TealMink_Pelt/Meshes/F_MED_TealMink_Pelt_AnimBP
target_skeleton=&/F_MED_TealMink_Pelt_Skeleton
graphs [
 AnimGraph {nodes [
   "Constraint - dyn_straps_base_l (PP)"!(ComponentPose, Alpha=1, ConstraintWeights_0=0.75, ConstraintWeights_1:float=0.25) -> (Pose) {
    alphaBoolBlend {blendOption=Linear}
    boneToModify {boneName=dyn_straps_base_l}
    constraintSetup [{targetBone {boneName=thigh_l}}, {targetBone {boneName=pelvis}}]
    constraintWeights [1,1]
   }
   n1 = Get Alpha() -> (Alpha:float)
   n2 = "Constraint - dyn_straps_base_l (PP)"(
    ComponentPose=Local To Component(LocalPose=Copy Pose From Mesh() {bUseAttachedParent=true})
    Alpha=@n1
    ConstraintWeights_0=0.5
    ConstraintWeights_1=0.5
   ) {
    alphaBoolBlend {blendOption=Linear}
    boneToModify {boneName=dyn_straps_base_l}
    constraintSetup [{targetBone {boneName=thigh_l}}, {targetBone {boneName=pelvis}}]
    constraintWeights [1,1]
   }
   n3 = "Constraint - dyn_straps_base_r (PP)"!(ComponentPose=@n2, Alpha=@n1, ConstraintWeights_0=0.5, ConstraintWeights_1:float=0.5) -> (Pose) {
    alphaBoolBlend {blendOption=Linear}
    boneToModify {boneName=dyn_straps_base_r}
    constraintSetup [{targetBone {boneName=thigh_r}}, {targetBone {boneName=pelvis}}]
    constraintWeights [1,1]
   }
   n4 = "Transform (Modify) Bone - Bone: dyn_straps_base_l"!(ComponentPose=@n3, bAlphaBoolEnabled:bool=Get bIsCrouching(), Translation="-5,5,5", Rotation="0,-60,0", Scale="1,1,1") -> (Pose) {
    alphaInputType=Bool
    rotationMode=BMM_Additive
    rotationSpace=BCS_BoneSpace
    translationMode=BMM_Additive
    translationSpace=BCS_BoneSpace
    alphaBoolBlend {blendOption=Linear}
    boneToModify {boneName=dyn_straps_base_l}
    rotation {yaw=-60}
    translation=<-5 5 5>
   }
   n5 = "Transform (Modify) Bone - Bone: dyn_necklace_1"(ComponentPose=@n4, Alpha=@n1, Translation="0,0,-0.5", Rotation="0,0,0", Scale="1,1,1") {
    rotationMode=BMM_Additive
    rotationSpace=BCS_BoneSpace
    translationMode=BMM_Additive
    translationSpace=BCS_BoneSpace
    alphaBoolBlend {blendOption=Linear}
    boneToModify {boneName=dyn_necklace_1}
    translation {z=-0.5}
   }
   n6 = Anim Dynamics - Bone: dyn_necklace_1(ComponentPose=@n5, Alpha=@n1, ExternalForce="0,0,0") {
    angularDampingOverride=0.9999
    angularSpringConstant=500
    bAngularSpring=true
    bGravityOverrideInSimSpace=true
    bOverrideAngularDamping=true
    simulationSpace=BoneRelative
    alphaBoolBlend {blendOption=Linear}
    boundBone {boneName=dyn_necklace_1}
    componentAppliedLinearAccClamp=<3500 3500 3500>
    componentLinearAccScale=<1.2 1.2 1.2>
    componentLinearVelScale=<1.2 1.2 1.2>
    physicsBodyDefinitions [{
      sphereCollisionRadius=0
      boundBone {boneName=dyn_necklace_1}
      boxExtents=<15 20 20>
      constraintSetup {angularLimitsMax=<1 30 5>, angularLimitsMin=<-1 0 -5>, angularTarget=<0 0 0>}
      localJointOffset=<4 0 0>
     }]
    relativeSpaceBone {boneName=root}
   }
   n7 = Anim Dynamics - Bone: dyn_necklace_2(ComponentPose=@n6, Alpha=@n1, ExternalForce="0,0,0") {
    angularDampingOverride=1
    angularSpringConstant=500
    bAngularSpring=true
    bGravityOverrideInSimSpace=true
    bOverrideAngularDamping=true
    simulationSpace=BoneRelative
    alphaBoolBlend {blendOption=Linear}
    boundBone {boneName=dyn_necklace_2}
    componentAppliedLinearAccClamp=<3500 3500 3500>
    componentLinearAccScale=<1.2 1.2 1.2>
    componentLinearVelScale=<1.2 1.2 1.2>
    physicsBodyDefinitions [{
      sphereCollisionRadius=0
      boundBone {boneName=dyn_necklace_2}
      boxExtents=<15 20 20>
      constraintSetup {angularLimitsMax=<1 30 10>, angularLimitsMin=<-1 0 -10>, angularTarget=<0 0 0>}
      localJointOffset=<4 0 0>
     }]
    relativeSpaceBone {boneName=root}
   }
   n8 = Anim Dynamics - Bone: dyn_pocket_l(ComponentPose=@n7, Alpha=0.5, ExternalForce="0,0,0") {
    angularDampingOverride=0.9999
    angularSpringConstant=1250
    bAngularSpring=true
    bGravityOverrideInSimSpace=true
    bOverrideAngularDamping=true
    bUseGravityOverride=true
    simulationSpace=BoneRelative
    alphaBoolBlend {blendOption=Linear}
    boundBone {boneName=dyn_pocket_l}
    componentAppliedLinearAccClamp=<3500 3500 3500>
    componentLinearAccScale=<1.2 1.2 1.2>
    componentLinearVelScale=<1.2 1.2 1.2>
    gravityOverride {z=-550}
    physicsBodyDefinitions [{
      sphereCollisionRadius=0
      boundBone {boneName=dyn_pocket_l}
      boxExtents=<10 20 17.5>
      constraintSetup {angularLimitsMax=<1 0 7.5>, angularLimitsMin=<-1 -7.5 -7.5>, angularTarget=<0 0 0>}
      localJointOffset=<2.5 -2 5>
     }]
    relativeSpaceBone {boneName=root}
   }
   n9 = Anim Dynamics - Bone: dyn_pocket_r(ComponentPose=@n8, Alpha=0.5, ExternalForce="0,0,0") {
    angularDampingOverride=1
    angularSpringConstant=1000
    bAngularSpring=true
    bGravityOverrideInSimSpace=true
    bOverrideAngularDamping=true
    bUseGravityOverride=true
    simulationSpace=BoneRelative
    alphaBoolBlend {blendOption=Linear}
    boundBone {boneName=dyn_pocket_r}
    componentAppliedLinearAccClamp=<3500 3500 3500>
    componentLinearAccScale=<1.2 1.2 1.2>
    componentLinearVelScale=<1.2 1.2 1.2>
    gravityOverride {z=-550}
    physicsBodyDefinitions [{
      sphereCollisionRadius=0
      boundBone {boneName=dyn_pocket_r}
      boxExtents=<10 20 17.5>
      constraintSetup {angularLimitsMax=<1 0 7.5>, angularLimitsMin=<-1 -7.5 -7.5>, angularTarget=<0 0 0>}
      localJointOffset=<3 1 3>
     }]
    relativeSpaceBone {boneName=root}
   }
   n10 = RigidBody!(
    ComponentPose=@n9
    Alpha=@n1
    OverrideWorldGravity:Vector=Get Dyn_Override_World_Gravity() -> (Dyn_Override_World_Gravity:Vector)
    ExternalForce
   ) -> (Pose) {
    bOverrideWorldGravity=true
    overridePhysicsAsset=&/F_MED_TealMink_Pelt_Physics
    simulationSpace=BaseBoneSpace
    alphaBoolBlend {blendOption=Linear}
    baseBoneRef {boneName=pelvis}
    componentAppliedLinearAccClamp=<1250 1250 1250>
    componentLinearAccScale=<1 1 1>
    componentLinearVelScale=<1 1 1>
    overrideWorldGravity {z=-980}
    simSpaceSettings {maxAngularAcceleration=50, maxAngularVelocity=10, maxLinearAcceleration=10, maxLinearVelocity=1, velocityScaleZ=0.25, worldAlpha=0.75}
   }
   n11 = "Transform (Modify) Bone - Bone: attach_backpack"!(ComponentPose=@n10, bAlphaBoolEnabled:bool=Get bIsInFrontEnd(), Translation="0,0.5,0", Rotation="0,0,0", Scale="1,1,1") -> (Pose) {
    alphaInputType=Bool
    rotationMode=BMM_Additive
    rotationSpace=BCS_BoneSpace
    translationMode=BMM_Additive
    translationSpace=BCS_BoneSpace
    alphaBoolBlend {blendOption=Linear}
    alphaScaleBiasClamp {bInterpResult=true}
    boneToModify {boneName=attach_backpack}
    translation {y=0.5}
   }
   n12 = Control Rig!(
    Source=Component To Local(ComponentPose=@n11)
    Alpha=@n1
    is_front_end:bool=Get bIsInFrontEnd()
    is_crouch:bool=Get bIsCrouching()
    is_DBNO:bool=Get bIsDBNO()
    is_emote:bool=Get bIsPlayingEmote()
   ) -> (Pose) {
    alphaBoolBlend {blendOption=Linear}
    controlRigAssetReference {blueprintRigClass=&/F_MED_TealMink_Pelt_CtrlRig.F_MED_TealMink_Pelt_CtrlRig_C}
   }
   Output Pose(Result=Component To Local(ComponentPose=Local To Component(LocalPose=@n12))) {layerGroup=None}
  ]}
 EventGraph {nodes [
   Event BlueprintUpdateAnimation()
   > n1 = Sequence()
   > n2 = Branch(Condition=Get bIsSkyDiving())
   > Set Dyn_Override_World_Gravity(Dyn_Override_World_Gravity:Vector="0,0,-1500") -> (Output_Get:Vector)
   Set Dyn_Override_World_Gravity(execute=@n2.else, Dyn_Override_World_Gravity="0,0,-980")
   n3 = Get bIsMoving2D()
   n4 = Get bIsCrouching()
   n5 = Branch(execute=@n1.then_1, Condition=AND Boolean(A=NOT Boolean(A=@n3), B=@n4))
   > Set Dyn_crouch_move(Dyn_crouch_move:bool=true) -> (Output_Get:bool)
   Set Dyn_crouch_move(execute=@n5.else, Dyn_crouch_move=false)
   NAND Boolean(A=@n3, B=@n4) {is=title K2Node_CommutativeAssociativeBinaryOperator}
  ]}
]
variables [
 Alpha {type=number, default=1}
 Dyn_crouch_move {type=boolean}
 Dyn_Override_World_Gravity {type=Vector, default=<0 0 -980>}
]
