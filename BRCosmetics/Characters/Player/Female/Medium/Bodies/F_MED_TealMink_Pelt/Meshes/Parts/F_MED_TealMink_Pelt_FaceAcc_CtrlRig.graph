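# Control Rig for the F_MED_TealMink_Pelt face accessory that correctively offsets the front and back dreadlock base bones as the body poses.
# Spherical pose readers on the pelvis, head, neck and clavicles output weights for chained ModifyTransforms nodes.
# Dyn_crouch, Dyn_fly, Dyn_Emote, Dyn_Parachute and Dyn_FrontEnd variables scale those weights, adjust offsets or drive their own poses.
# The pelvis reader's result is stored in Dyn_pelvis_rotate_alfa to modulate later corrections.
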ControlRigBlueprint /BRCosmetics/Characters/Player/Female/Medium/Bodies/F_MED_TealMink_Pelt/Meshes/Parts/F_MED_TealMink_Pelt_FaceAcc_CtrlRig
graphs [
 RigVMFunctionLibrary {links [], nodes []}
 RigVMFunction_Sequence_1_ContainedGraph {nodes [
   n1 = RigVMFunctionEntryNode()
   RigVMFunction_Sequence = &RigVMFunction_Sequence(ExecuteContext=@n1.ExecuteContext)
   n2 = &RigVMFunction_Sequence(ExecuteContext=@RigVMFunction_Sequence.B)
   n3 = &RigVMFunction_Sequence(ExecuteContext=@n2.B)
   RigVMFunctionReturnNode(A=@RigVMFunction_Sequence.A, B=@n2.A, C=@n3.A, D=@n3.B)
  ]}
 RigVMFunction_Sequence_1_ContainedGraph {nodes [
   n1 = RigVMFunctionEntryNode()
   RigVMFunction_Sequence = &RigVMFunction_Sequence(ExecuteContext=@n1.ExecuteContext)
   n2 = &RigVMFunction_Sequence(ExecuteContext=@RigVMFunction_Sequence.B)
   n3 = &RigVMFunction_Sequence(ExecuteContext=@n2.B)
   n4 = &RigVMFunction_Sequence(ExecuteContext=@n3.B)
   n5 = &RigVMFunction_Sequence(ExecuteContext=@n4.B)
   RigVMFunctionReturnNode(A=@RigVMFunction_Sequence.A, B=@n2.A, C=@n3.A, D=@n4.A, E=@n5.A, F=@n5.B)
  ]}
 RigVMFunction_Sequence_4_ContainedGraph {nodes [
   n1 = RigVMFunctionEntryNode()
   n2 = &RigVMFunction_Sequence(ExecuteContext=@n1.ExecuteContext)
   n3 = &RigVMFunction_Sequence(ExecuteContext=@n2.B)
   RigVMFunctionReturnNode(B=@n3.A, C=@n3.B, A=@n2.A)
  ]}
 RigVMModel {nodes [
   RigUnit_BeginExecution = &RigUnit_BeginExecution()
   n1 = &RigUnit_SphericalPoseReader(
    ActiveRegionScaleFactors=(PositiveWidth=0.15,NegativeWidth=0.15,PositiveHeight=1,NegativeHeight=1)
    ActiveRegionSize=0.1
    Debug=(bDrawDebug=true,DebugScale=25,DebugSegments=20,DebugThickness=0.25)
    DriverAxis=(X=0,Y=-1,Z=0)
    DriverItem=(Type=Bone,Name="pelvis")
    FalloffRegionScaleFactors=(PositiveWidth=1,NegativeWidth=1,PositiveHeight=1,NegativeHeight=1)
    FalloffSize=0.2
    FlipHeightScaling=false
    FlipWidthScaling=true
    OptionalParentItem=(Type=Bone)
    RotationOffset=(X=-95,Y=0,Z=0)
    ExecutePin=@RigUnit_BeginExecution.ExecutePin
   )
   > RigVMVariableNode(Value=@n1.OutputParam) {variable=Dyn_pelvis_rotate_alfa}
   > RigVMFunction_Sequence = RigVMAggregateNode()
   RigVMFunction_Sequence_1 = RigVMAggregateNode(ExecuteContext=@RigVMFunction_Sequence.B)
   n2 = &RigUnit_SphericalPoseReader(
    ActiveRegionScaleFactors=(PositiveWidth=1,NegativeWidth=3,PositiveHeight=0.5,NegativeHeight=1)
    ActiveRegionSize=0.1
    Debug=(bDrawDebug=true,bDraw2D=False,bDrawLocalAxes=False,DebugScale=25,DebugSegments=20,DebugThickness=0.25)
    DriverAxis=(X=1,Y=0,Z=0)
    DriverItem=(Type=Bone,Name="head")
    FalloffRegionScaleFactors=(PositiveWidth=1,NegativeWidth=1,PositiveHeight=1,NegativeHeight=1)
    FalloffSize=0.3
    FlipHeightScaling=False
    FlipWidthScaling=False
    OptionalParentItem=(Type=Bone,Name="None")
    RotationOffset=(X=-70,Y=-115,Z=15)
    ExecutePin=@RigVMFunction_Sequence_1.A
   )
   > &RigUnit_ModifyTransforms(
    ItemToModify=((Item=(Type=Bone,Name="dyn_dread_front_1_base_l"),Transform=(Rotation=(X=0,Y=0,Z=0,W=1),Translation=(X=-1,Y=0,Z=0),Scale3D=(X=1,Y=1,Z=1))))
    WeightMinimum=0
    ItemToModify.0.Transform.Rotation=&RigVMFunction_MathQuaternionFromEuler(Euler=(X=0,Y=0,Z=0)).Result
    Weight=@n2.OutputParam
   )
   > &RigUnit_ModifyTransforms(
    ItemToModify=((Item=(Type=Bone,Name="dyn_dread_front_1_base_r"),Transform=(Rotation=(X=0,Y=0,Z=0,W=1),Translation=(X=-1,Y=0,Z=0),Scale3D=(X=1,Y=1,Z=1))))
    WeightMinimum=0
    ItemToModify.0.Transform.Rotation=&RigVMFunction_MathQuaternionFromEuler(Euler=(X=-30,Y=20,Z=0)).Result
    Weight=@n2.OutputParam
   )
   n3 = RigVMVariableNode() {variable=Dyn_crouch}
   n4 = &RigVMFunction_MathDoubleMul(A=-15, B=@n3.Value)
   n5 = &RigVMFunction_MathDoubleMul(A=40, B=@n3.Value)
   n6 = &RigVMFunction_MathDoubleMul(A=0, B=@n3.Value)
   > &RigUnit_ModifyTransforms(
    ItemToModify=((Item=(Type=Bone,Name="dyn_dread_back_1_base"),Transform=(Rotation=(X=0,Y=0,Z=0,W=1),Translation=(X=-5,Y=2,Z=5),Scale3D=(X=1,Y=1,Z=1))))
    WeightMinimum=0
    ItemToModify.0.Transform.Translation.Y=&RigVMFunction_MathDoubleAdd(A=-1, B=&RigVMFunction_MathDoubleMul(A=0, B=@n3.Value).Result).Result
    ItemToModify.0.Transform.Translation.Z=&RigVMFunction_MathDoubleAdd(A=0, B=&RigVMFunction_MathDoubleMul(A=1, B=@n3.Value).Result).Result
    ItemToModify.0.Transform.Translation.X=&RigVMFunction_MathDoubleAdd(A=-5, B=&RigVMFunction_MathDoubleMul(A=0, B=@n3.Value).Result).Result
    ItemToModify.0.Transform.Rotation=&RigVMFunction_MathQuaternionFromEuler(
     Euler=(X=-20,Y=40,Z=-60)
     Euler.Y=&RigVMFunction_MathDoubleAdd(A=30, B=@n4.Result).Result
     Euler.Z=&RigVMFunction_MathDoubleAdd(A=-60, B=@n5.Result).Result
     Euler.X=&RigVMFunction_MathDoubleAdd(A=-20, B=@n6.Result).Result
    ).Result
    Weight=@n2.OutputParam
   )
   n7 = RigVMVariableNode() {variable=Dyn_crouch}
   n8 = &RigVMFunction_MathDoubleMul(A=20, B=@n7.Value)
   n9 = &RigVMFunction_MathDoubleMul(A=0, B=@n7.Value)
   n10 = &RigVMFunction_MathDoubleMul(A=-45, B=@n7.Value)
   > &RigUnit_ModifyTransforms(
    ItemToModify=((Item=(Type=Bone,Name="dyn_dread_back_2_base"),Transform=(Rotation=(X=0,Y=0,Z=0,W=1),Translation=(X=-5,Y=0,Z=-3),Scale3D=(X=1,Y=1,Z=1))))
    WeightMinimum=0
    ItemToModify.0.Transform.Translation.Z=&RigVMFunction_MathDoubleAdd(A=0, B=&RigVMFunction_MathDoubleMul(A=-3, B=@n7.Value).Result).Result
    ItemToModify.0.Transform.Translation.X=&RigVMFunction_MathDoubleAdd(A=-3, B=&RigVMFunction_MathDoubleMul(B=-3, A=@n7.Value).Result).Result
    ItemToModify.0.Transform.Translation.Y=&RigVMFunction_MathDoubleAdd(A=0, B=&RigVMFunction_MathDoubleMul(A=0, B=@n7.Value).Result).Result
    ItemToModify.0.Transform.Rotation=&RigVMFunction_MathQuaternionFromEuler(
     Euler=(X=-20,Y=40,Z=0)
     Euler.Y=&RigVMFunction_MathDoubleAdd(A=0, B=@n8.Result).Result
     Euler.Z=&RigVMFunction_MathDoubleAdd(A=0, B=@n9.Result).Result
     Euler.X=&RigVMFunction_MathDoubleAdd(A=0, B=@n10.Result).Result
    ).Result
    Weight=@n2.OutputParam
   )
   n11 = RigVMVariableNode() {variable=Dyn_crouch}
   n12 = &RigVMFunction_MathDoubleMul(A=-45, B=@n11.Value)
   n13 = &RigVMFunction_MathDoubleMul(A=-30, B=@n11.Value)
   n14 = &RigVMFunction_MathDoubleMul(A=-15, B=@n11.Value)
   > &RigUnit_ModifyTransforms(
    ItemToModify=((Item=(Type=Bone,Name="dyn_dread_back_3_base"),Transform=(Rotation=(X=0,Y=0,Z=0,W=1),Translation=(X=-3,Y=-3,Z=2),Scale3D=(X=1,Y=1,Z=1))))
    WeightMinimum=0
    ItemToModify.0.Transform.Translation.Z=&RigVMFunction_MathDoubleAdd(A=2, B=&RigVMFunction_MathDoubleMul(A=0, B=@n11.Value).Result).Result
    ItemToModify.0.Transform.Translation.Y=&RigVMFunction_MathDoubleAdd(A=-3, B=&RigVMFunction_MathDoubleMul(A=0, B=@n11.Value).Result).Result
    ItemToModify.0.Transform.Translation.X=&RigVMFunction_MathDoubleAdd(A=-3, B=&RigVMFunction_MathDoubleMul(A=-1, B=@n11.Value).Result).Result
    ItemToModify.0.Transform.Rotation=&RigVMFunction_MathQuaternionFromEuler(
     Euler=(X=30,Y=30,Z=-60)
     Euler.Z=&RigVMFunction_MathDoubleAdd(A=-60, B=@n12.Result).Result
     Euler.Y=&RigVMFunction_MathDoubleAdd(A=30, B=@n13.Result).Result
     Euler.X=&RigVMFunction_MathDoubleAdd(A=30, B=@n14.Result).Result
    ).Result
    Weight=@n2.OutputParam
   )
   n15 = &RigUnit_SphericalPoseReader(
    ActiveRegionScaleFactors=(PositiveWidth=0.5,NegativeWidth=0.5,PositiveHeight=0.5,NegativeHeight=0.5)
    ActiveRegionSize=0.1
    Debug=(bDrawDebug=true,bDraw2D=False,bDrawLocalAxes=False,DebugScale=25,DebugSegments=20,DebugThickness=0.25)
    DriverAxis=(X=1,Y=0,Z=0)
    DriverItem=(Type=Bone,Name="head")
    FalloffRegionScaleFactors=(PositiveWidth=1,NegativeWidth=1,PositiveHeight=1,NegativeHeight=1)
    FalloffSize=0.25
    FlipHeightScaling=False
    FlipWidthScaling=False
    OptionalParentItem=(Type=Bone,Name="None")
    RotationOffset=(X=90,Y=0,Z=-65)
    ExecutePin=@RigVMFunction_Sequence_1.B
   )
   n16 = RigVMVariableNode() {variable=Dyn_crouch}
   n17 = &RigVMFunction_MathDoubleMul(A=40, B=@n16.Value)
   n18 = &RigVMFunction_MathDoubleMul(A=10, B=@n16.Value)
   n19 = &RigVMFunction_MathDoubleMul(A=30, B=@n16.Value)
   Clamp = &RigVMFunction_MathDoubleClamp(Minimum=0.5, Value=&RigVMFunction_MathDoubleMul(A=1, B=RigVMVariableNode() {variable=Dyn_pelvis_rotate_alfa}.Value).Result)
   n20 = &RigVMFunction_MathDoubleMul(
    A=&RigVMFunction_MathDoubleMul(B=@Clamp.Result, A=@n15.OutputParam).Result
    B=&RigVMFunction_MathDoubleSub(A=1, B=RigVMVariableNode() {variable=Dyn_fly}.Value).Result
   )
   > &RigUnit_ModifyTransforms(
    ItemToModify=((Item=(Type=Bone,Name="dyn_dread_back_1_base"),Transform=(Rotation=(X=0,Y=0,Z=0,W=1),Translation=(X=0,Y=-2,Z=0),Scale3D=(X=1,Y=1,Z=1))))
    WeightMinimum=0
    ItemToModify.0.Transform.Translation.Z=&RigVMFunction_MathDoubleAdd(A=0, B=&RigVMFunction_MathDoubleMul(A=0, B=@n16.Value).Result).Result
    ItemToModify.0.Transform.Translation.X=&RigVMFunction_MathDoubleAdd(A=-1, B=&RigVMFunction_MathDoubleMul(A=0, B=@n16.Value).Result).Result
    ItemToModify.0.Transform.Translation.Y=&RigVMFunction_MathDoubleAdd(A=-2, B=&RigVMFunction_MathDoubleMul(A=0, B=@n16.Value).Result).Result
    ItemToModify.0.Transform.Rotation=&RigVMFunction_MathQuaternionFromEuler(
     Euler=(X=0,Y=-10,Z=75)
     Euler.Z=&RigVMFunction_MathDoubleAdd(A=20, B=@n17.Result).Result
     Euler.X=&RigVMFunction_MathDoubleAdd(A=-30, B=@n18.Result).Result
     Euler.Y=&RigVMFunction_MathDoubleAdd(A=-30, B=@n19.Result).Result
    ).Result
    Weight=@n20.Result
   )
   n21 = RigVMVariableNode() {variable=Dyn_crouch}
   n22 = &RigVMFunction_MathDoubleMul(A=20, B=@n21.Value)
   n23 = &RigVMFunction_MathDoubleMul(A=0, B=@n21.Value)
   n24 = &RigVMFunction_MathDoubleMul(A=-10, B=@n21.Value)
   > &RigUnit_ModifyTransforms(
    ItemToModify=((Item=(Type=Bone,Name="dyn_dread_back_2_base"),Transform=(Rotation=(X=0,Y=0,Z=0,W=1),Translation=(X=-2,Y=-1,Z=3),Scale3D=(X=1,Y=1,Z=1))))
    WeightMinimum=0
    ItemToModify.0.Transform.Rotation=&RigVMFunction_MathQuaternionFromEuler(
     Euler=(X=-20,Y=-30,Z=-30)
     Euler.Y=&RigVMFunction_MathDoubleAdd(A=20, B=@n22.Result).Result
     Euler.Z=&RigVMFunction_MathDoubleAdd(A=0, B=@n23.Result).Result
     Euler.X=&RigVMFunction_MathDoubleAdd(A=-20, B=@n24.Result).Result
    ).Result
    Weight=@n20.Result
   )
   n25 = RigVMVariableNode() {variable=Dyn_crouch}
   n26 = &RigVMFunction_MathDoubleMul(A=60, B=@n25.Value)
   n27 = &RigVMFunction_MathDoubleMul(A=30, B=@n25.Value)
   n28 = &RigVMFunction_MathDoubleMul(A=0, B=@n25.Value)
   > &RigUnit_ModifyTransforms(
    ItemToModify=((Item=(Type=Bone,Name="dyn_dread_back_3_base"),Transform=(Rotation=(X=0,Y=0,Z=0,W=1),Translation=(X=-1,Y=1,Z=0),Scale3D=(X=1,Y=1,Z=1))))
    WeightMinimum=0
    ItemToModify.0.Transform.Rotation=&RigVMFunction_MathQuaternionFromEuler(
     Euler=(X=-60,Y=0,Z=30)
     Euler.X=&RigVMFunction_MathDoubleAdd(A=-60, B=@n26.Result).Result
     Euler.Z=&RigVMFunction_MathDoubleAdd(A=45, B=@n27.Result).Result
     Euler.Y=&RigVMFunction_MathDoubleAdd(A=0, B=@n28.Result).Result
    ).Result
    Weight=@n20.Result
   )
   n29 = &RigUnit_SphericalPoseReader(
    ActiveRegionScaleFactors=(PositiveWidth=1.5,NegativeWidth=1.5,PositiveHeight=1,NegativeHeight=0.15)
    ActiveRegionSize=0.1
    Debug=(bDrawDebug=true,bDraw2D=False,bDrawLocalAxes=False,DebugScale=25,DebugSegments=20,DebugThickness=0.25)
    DriverAxis=(X=0,Y=1,Z=0)
    DriverItem=(Type=Bone,Name="head")
    FalloffRegionScaleFactors=(PositiveWidth=1,NegativeWidth=1,PositiveHeight=1,NegativeHeight=1)
    FalloffSize=0.2
    FlipHeightScaling=False
    FlipWidthScaling=False
    OptionalParentItem=(Type=Bone,Name="None")
    RotationOffset=(X=120,Y=0,Z=-10)
    ExecutePin=@RigVMFunction_Sequence_1.C
   )
   n30 = RigVMVariableNode() {variable=Dyn_crouch}
   Add = &RigVMFunction_MathDoubleAdd(A=-2, B=&RigVMFunction_MathDoubleMul(A=-1, B=@n30.Value).Result)
   n31 = &RigVMFunction_MathDoubleAdd(A=0, B=&RigVMFunction_MathDoubleMul(A=20, B=@n30.Value).Result)
   n32 = &RigVMFunction_MathDoubleMul(A=20, B=RigVMVariableNode() {variable=Dyn_FrontEnd}.Value)
   n33 = &RigVMFunction_MathDoubleMul(A=-20, B=@n30.Value)
   n34 = &RigVMFunction_MathDoubleMul(A=-10, B=@n30.Value)
   n35 = &RigVMFunction_MathDoubleMul(A=@n29.OutputParam, B=&RigVMFunction_MathDoubleSub(A=1, B=RigVMVariableNode() {variable=Dyn_Emote}.Value).Result)
   > &RigUnit_ModifyTransforms(
    ItemToModify=((Item=(Type=Bone,Name="dyn_dread_back_1_base"),Transform=(Rotation=(X=0,Y=0,Z=0,W=1),Translation=(X=-1,Y=0,Z=0),Scale3D=(X=1,Y=1,Z=1))))
    WeightMinimum=0
    ItemToModify.0.Transform.Translation.Y=@Add.Result
    ItemToModify.0.Transform.Translation.X=&RigVMFunction_MathDoubleAdd(A=-2, B=&RigVMFunction_MathDoubleMul(A=-2, B=@n30.Value).Result).Result
    ItemToModify.0.Transform.Translation.Z=&RigVMFunction_MathDoubleAdd(A=0, B=&RigVMFunction_MathDoubleMul(A=2, B=@n30.Value).Result).Result
    ItemToModify.0.Transform.Rotation=&RigVMFunction_MathQuaternionFromEuler(
     Euler=(X=0,Y=-20,Z=0)
     Euler.Z=&RigVMFunction_MathDoubleAdd(A=@n31.Result, B=@n32.Result).Result
     Euler.X=&RigVMFunction_MathDoubleAdd(A=-10, B=@n33.Result).Result
     Euler.Y=&RigVMFunction_MathDoubleAdd(A=10, B=@n34.Result).Result
    ).Result
    Weight=@n35.Result
   )
   n36 = RigVMVariableNode() {variable=Dyn_crouch}
   n37 = &RigVMFunction_MathDoubleMul(A=15, B=@n36.Value)
   n38 = &RigVMFunction_MathDoubleMul(A=-45, B=@n36.Value)
   n39 = &RigVMFunction_MathDoubleMul(A=10, B=@n36.Value)
   > &RigUnit_ModifyTransforms(
    ItemToModify=((Item=(Type=Bone,Name="dyn_dread_back_2_base"),Transform=(Rotation=(X=0,Y=0,Z=0,W=1),Translation=(X=-1,Y=-2,Z=0),Scale3D=(X=1,Y=1,Z=1))))
    WeightMinimum=0
    ItemToModify.0.Transform.Translation.Z=&RigVMFunction_MathDoubleAdd(A=0, B=&RigVMFunction_MathDoubleMul(A=0, B=@n36.Value).Result).Result
    ItemToModify.0.Transform.Translation.X=&RigVMFunction_MathDoubleAdd(A=-1, B=&RigVMFunction_MathDoubleMul(A=-1, B=@n36.Value).Result).Result
    ItemToModify.0.Transform.Translation.Y=&RigVMFunction_MathDoubleAdd(A=0, B=&RigVMFunction_MathDoubleMul(A=-5, B=@n36.Value).Result).Result
    ItemToModify.0.Transform.Rotation=&RigVMFunction_MathQuaternionFromEuler(
     Euler=(X=60,Y=-10,Z=-30)
     Euler.Z=&RigVMFunction_MathDoubleAdd(A=-30, B=@n37.Result).Result
     Euler.X=&RigVMFunction_MathDoubleAdd(A=0, B=@n38.Result).Result
     Euler.Y=&RigVMFunction_MathDoubleAdd(A=20, B=@n39.Result).Result
    ).Result
    Weight=@n35.Result
   )
   n40 = RigVMVariableNode() {variable=Dyn_crouch}
   n41 = &RigVMFunction_MathDoubleMul(A=0, B=@n40.Value)
   n42 = RigVMVariableNode() {variable=Dyn_FrontEnd}
   n43 = &RigVMFunction_MathDoubleMul(A=-2, B=@n40.Value)
   n44 = &RigVMFunction_MathDoubleMul(A=2, B=@n40.Value)
   n45 = &RigVMFunction_MathDoubleAdd(A=0, B=&RigVMFunction_MathDoubleMul(A=-10, B=@n40.Value).Result)
   n46 = &RigVMFunction_MathDoubleMul(A=20, B=@n42.Value)
   n47 = &RigVMFunction_MathDoubleMul(A=10, B=@n40.Value)
   n48 = &RigVMFunction_MathDoubleMul(A=10, B=@n40.Value)
   > &RigUnit_ModifyTransforms(
    ItemToModify=((Item=(Type=Bone,Name="dyn_dread_back_3_base"),Transform=(Rotation=(X=0,Y=0,Z=0,W=1),Translation=(X=0,Y=0,Z=1.5),Scale3D=(X=1,Y=1,Z=1))))
    WeightMinimum=0
    ItemToModify.0.Transform.Translation.X=&RigVMFunction_MathDoubleAdd(
     A=&RigVMFunction_MathDoubleAdd(A=-1, B=@n41.Result).Result
     B=&RigVMFunction_MathDoubleMul(A=-1, B=@n42.Value).Result
    ).Result
    ItemToModify.0.Transform.Translation.Z=&RigVMFunction_MathDoubleAdd(
     A=&RigVMFunction_MathDoubleAdd(A=1.5, B=@n43.Result).Result
     B=&RigVMFunction_MathDoubleMul(A=-2.5, B=@n42.Value).Result
    ).Result
    ItemToModify.0.Transform.Translation.Y=&RigVMFunction_MathDoubleAdd(
     A=&RigVMFunction_MathDoubleAdd(A=0, B=@n44.Result).Result
     B=&RigVMFunction_MathDoubleMul(A=1, B=@n42.Value).Result
    ).Result
    ItemToModify.0.Transform.Rotation=&RigVMFunction_MathQuaternionFromEuler(
     Euler=(X=-10,Y=15,Z=-30)
     Euler.Z=&RigVMFunction_MathDoubleAdd(A=@n45.Result, B=@n46.Result).Result
     Euler.X=&RigVMFunction_MathDoubleAdd(A=0, B=@n47.Result).Result
     Euler.Y=&RigVMFunction_MathDoubleAdd(A=0, B=@n48.Result).Result
    ).Result
    Weight=@n35.Result
   )
   n49 = &RigUnit_SphericalPoseReader(
    ActiveRegionScaleFactors=(PositiveWidth=1.5,NegativeWidth=1.5,PositiveHeight=0.15,NegativeHeight=1)
    ActiveRegionSize=0.1
    Debug=(bDrawDebug=true,bDraw2D=False,bDrawLocalAxes=False,DebugScale=25,DebugSegments=20,DebugThickness=0.25)
    DriverAxis=(X=0,Y=1,Z=0)
    DriverItem=(Type=Bone,Name="head")
    FalloffRegionScaleFactors=(PositiveWidth=1,NegativeWidth=1,PositiveHeight=1,NegativeHeight=1)
    FalloffSize=0.2
    FlipHeightScaling=False
    FlipWidthScaling=False
    OptionalParentItem=(Type=Bone,Name="None")
    RotationOffset=(X=60,Y=0,Z=-10)
    ExecutePin=@RigVMFunction_Sequence_1.D
   )
   n50 = RigVMVariableNode() {variable=Dyn_crouch}
   n51 = &RigVMFunction_MathDoubleMul(A=0, B=@n50.Value)
   n52 = &RigVMFunction_MathDoubleMul(A=0, B=@n50.Value)
   n53 = &RigVMFunction_MathDoubleMul(A=0, B=@n50.Value)
   n54 = &RigVMFunction_MathDoubleMul(A=@n49.OutputParam, B=&RigVMFunction_MathDoubleSub(A=1, B=RigVMVariableNode() {variable=Dyn_Emote}.Value).Result)
   > &RigUnit_ModifyTransforms(
    ItemToModify=((Item=(Type=Bone,Name="dyn_dread_back_1_base"),Transform=(Rotation=(X=0,Y=0,Z=0,W=1),Translation=(X=-1,Y=0,Z=0),Scale3D=(X=1,Y=1,Z=1))))
    WeightMinimum=0
    ItemToModify.0.Transform.Translation.Y=&RigVMFunction_MathDoubleAdd(B=0, A=&RigVMFunction_MathDoubleMul(A=0, B=@n50.Value).Result).Result
    ItemToModify.0.Transform.Translation.X=&RigVMFunction_MathDoubleAdd(A=-2, B=&RigVMFunction_MathDoubleMul(A=0, B=@n50.Value).Result).Result
    ItemToModify.0.Transform.Rotation=&RigVMFunction_MathQuaternionFromEuler(
     Euler=(X=0,Y=-20,Z=0)
     Euler.Z=&RigVMFunction_MathDoubleAdd(A=-30, B=@n51.Result).Result
     Euler.X=&RigVMFunction_MathDoubleAdd(A=0, B=@n52.Result).Result
     Euler.Y=&RigVMFunction_MathDoubleAdd(A=0, B=@n53.Result).Result
    ).Result
    Weight=@n54.Result
   )
   n55 = RigVMVariableNode() {variable=Dyn_crouch}
   n56 = &RigVMFunction_MathDoubleAdd(A=0, B=&RigVMFunction_MathDoubleMul(A=25, B=@n55.Value).Result)
   n57 = &RigVMFunction_MathDoubleMul(A=-45, B=RigVMVariableNode() {variable=Dyn_FrontEnd}.Value)
   n58 = &RigVMFunction_MathDoubleMul(A=75, B=@n55.Value)
   n59 = &RigVMFunction_MathDoubleMul(A=-35, B=@n55.Value)
   > &RigUnit_ModifyTransforms(
    ItemToModify=((Item=(Type=Bone,Name="dyn_dread_back_2_base"),Transform=(Rotation=(X=0,Y=0,Z=0,W=1),Translation=(X=-1,Y=-2,Z=0),Scale3D=(X=1,Y=1,Z=1))))
    WeightMinimum=0
    ItemToModify.0.Transform.Translation.Z=&RigVMFunction_MathDoubleAdd(A=0, B=&RigVMFunction_MathDoubleMul(A=0, B=@n55.Value).Result).Result
    ItemToModify.0.Transform.Translation.X=&RigVMFunction_MathDoubleAdd(A=0, B=&RigVMFunction_MathDoubleMul(A=0, B=@n55.Value).Result).Result
    ItemToModify.0.Transform.Translation.Y=&RigVMFunction_MathDoubleAdd(A=0, B=&RigVMFunction_MathDoubleMul(A=0, B=@n55.Value).Result).Result
    ItemToModify.0.Transform.Rotation=&RigVMFunction_MathQuaternionFromEuler(
     Euler=(X=60,Y=-10,Z=-30)
     Euler.X=&RigVMFunction_MathDoubleAdd(B=@n56.Result, A=@n57.Result).Result
     Euler.Z=&RigVMFunction_MathDoubleAdd(A=0, B=@n58.Result).Result
     Euler.Y=&RigVMFunction_MathDoubleAdd(A=0, B=@n59.Result).Result
    ).Result
    Weight=@n54.Result
   )
   n60 = RigVMVariableNode() {variable=Dyn_crouch}
   n61 = &RigVMFunction_MathDoubleMul(A=0, B=@n60.Value)
   n62 = RigVMVariableNode() {variable=Dyn_FrontEnd}
   n63 = &RigVMFunction_MathDoubleAdd(A=0, B=&RigVMFunction_MathDoubleMul(A=0, B=@n60.Value).Result)
   n64 = &RigVMFunction_MathDoubleMul(A=15, B=@n62.Value)
   n65 = &RigVMFunction_MathDoubleMul(A=0, B=@n60.Value)
   n66 = &RigVMFunction_MathDoubleMul(A=45, B=@n60.Value)
   > &RigUnit_ModifyTransforms(
    ItemToModify=((Item=(Type=Bone,Name="dyn_dread_back_3_base"),Transform=(Rotation=(X=0,Y=0,Z=0,W=1),Translation=(X=0,Y=0,Z=1.5),Scale3D=(X=1,Y=1,Z=1))))
    WeightMinimum=0
    ItemToModify.0.Transform.Translation.X=&RigVMFunction_MathDoubleAdd(
     A=&RigVMFunction_MathDoubleAdd(A=0, B=@n61.Result).Result
     B=&RigVMFunction_MathDoubleMul(A=1, B=@n62.Value).Result
    ).Result
    ItemToModify.0.Transform.Translation.Y=&RigVMFunction_MathDoubleAdd(A=0, B=&RigVMFunction_MathDoubleMul(A=0, B=@n60.Value).Result).Result
    ItemToModify.0.Transform.Translation.Z=&RigVMFunction_MathDoubleAdd(A=0, B=&RigVMFunction_MathDoubleMul(A=0, B=@n60.Value).Result).Result
    ItemToModify.0.Transform.Rotation=&RigVMFunction_MathQuaternionFromEuler(
     Euler=(X=-10,Y=15,Z=-30)
     Euler.Y=&RigVMFunction_MathDoubleAdd(A=@n63.Result, B=@n64.Result).Result
     Euler.X=&RigVMFunction_MathDoubleAdd(A=0, B=@n65.Result).Result
     Euler.Z=&RigVMFunction_MathDoubleAdd(A=30, B=@n66.Result).Result
    ).Result
    Weight=@n54.Result
   )
   RigVMFunction_Sequence_3 = RigVMAggregateNode(ExecuteContext=@RigVMFunction_Sequence_1.F)
   n67 = RigVMVariableNode() {variable=Dyn_Emote}
   &RigUnit_ModifyTransforms(
    ItemToModify=((Item=(Type=Bone,Name="dyn_dread_back_1_base"),Transform=(Rotation=(X=0,Y=0,Z=0.2588,W=0.9659),Translation=(X=0,Y=0,Z=0),Scale3D=(X=1,Y=1,Z=1))))
    WeightMinimum=0
    ExecutePin=@RigVMFunction_Sequence_3.A
    Weight=@n67.Value
   )
   > &RigUnit_ModifyTransforms(
    ItemToModify=((Item=(Type=Bone,Name="dyn_dread_back_2_base"),Transform=(Rotation=(X=0,Y=0.2164,Z=0,W=0.9763),Translation=(X=0,Y=0,Z=0),Scale3D=(X=1,Y=1,Z=1))))
    WeightMinimum=0
    Weight=@n67.Value
   )
   n68 = &RigVMFunction_MathDoubleMul(B=1, A=RigVMVariableNode() {variable=Dyn_fly}.Value)
   &RigUnit_ModifyTransforms(
    ItemToModify=((Item=(Type=Bone,Name="dyn_dread_front_1_base_l"),Transform=(Rotation=(X=0,Y=0,Z=0,W=1),Translation=(X=0,Y=0,Z=0),Scale3D=(X=1,Y=1,Z=1))))
    WeightMinimum=0
    ItemToModify.0.Transform.Rotation=&RigVMFunction_MathQuaternionFromEuler(Euler=(X=0,Y=0,Z=-20)).Result
    ExecutePin=@RigVMFunction_Sequence_3.C
    Weight=@n68.Result
   )
   > &RigUnit_ModifyTransforms(
    ItemToModify=((Item=(Type=Bone,Name="dyn_dread_front_1_base_r"),Transform=(Rotation=(X=0,Y=0,Z=0,W=1),Translation=(X=0,Y=0,Z=0),Scale3D=(X=1,Y=1,Z=1))))
    WeightMinimum=0
    ItemToModify.0.Transform.Rotation=&RigVMFunction_MathQuaternionFromEuler(Euler=(X=0,Y=0,Z=-20)).Result
    Weight=@n68.Result
   )
   > &RigUnit_ModifyTransforms(
    ItemToModify=((Item=(Type=Bone,Name="dyn_dread_back_1_base"),Transform=(Rotation=(X=0,Y=0,Z=0,W=1),Translation=(X=-5,Y=0,Z=0),Scale3D=(X=1,Y=1,Z=1))))
    WeightMinimum=0
    ItemToModify.0.Transform.Rotation=&RigVMFunction_MathQuaternionFromEuler(Euler=(X=-20,Y=30,Z=-10)).Result
    Weight=@n68.Result
   )
   > &RigUnit_ModifyTransforms(
    ItemToModify=((Item=(Type=Bone,Name="dyn_dread_back_2_base"),Transform=(Rotation=(X=0,Y=0,Z=0,W=1),Translation=(X=-3,Y=0,Z=0),Scale3D=(X=1,Y=1,Z=1))))
    WeightMinimum=0
    ItemToModify.0.Transform.Rotation=&RigVMFunction_MathQuaternionFromEuler(Euler=(X=0,Y=-30,Z=30)).Result
    Weight=@n68.Result
   )
   > &RigUnit_ModifyTransforms(
    ItemToModify=((Item=(Type=Bone,Name="dyn_dread_back_3_base"),Transform=(Rotation=(X=0,Y=0,Z=0,W=1),Translation=(X=-3,Y=-3,Z=2),Scale3D=(X=1,Y=1,Z=1))))
    WeightMinimum=0
    ItemToModify.0.Transform.Rotation=&RigVMFunction_MathQuaternionFromEuler(Euler=(X=0,Y=15,Z=60)).Result
    Weight=@n68.Result
   )
   n69 = &RigVMFunction_MathDoubleMul(
    A=&RigVMFunction_MathDoubleClamp(Value=&RigVMFunction_MathDoubleSub(B=0.25, A=@n2.OutputParam).Result).Result
    B=RigVMVariableNode() {variable=Dyn_crouch}.Value
   )
   Subtract = &RigVMFunction_MathDoubleSub(A=1, B=RigVMVariableNode() {variable=Dyn_pelvis_rotate_alfa}.Value)
   n70 = &RigVMFunction_MathDoubleMul(
    B=&RigVMFunction_MathDoubleMul(B=&RigVMFunction_MathDoubleMul(B=1, A=@n69.Result).Result, A=@Subtract.Result).Result
    A=&RigVMFunction_MathDoubleSub(A=1, B=RigVMVariableNode() {variable=Dyn_fly}.Value).Result
   )
   &RigUnit_ModifyTransforms(
    ItemToModify=((Item=(Type=Bone,Name="dyn_dread_back_1_base"),Transform=(Rotation=(X=0,Y=0,Z=0,W=1),Translation=(X=-1,Y=0,Z=-1),Scale3D=(X=1,Y=1,Z=1))))
    WeightMinimum=0
    ItemToModify.0.Transform.Translation.Y=&RigVMFunction_MathDoubleAdd(A=-2, B=0).Result
    ItemToModify.0.Transform.Translation.Z=&RigVMFunction_MathDoubleAdd(A=0, B=0).Result
    ItemToModify.0.Transform.Translation.X=&RigVMFunction_MathDoubleAdd(A=-3, B=0).Result
    ItemToModify.0.Transform.Rotation=&RigVMFunction_MathQuaternionFromEuler(
     Euler=(X=0,Y=-10,Z=-75)
     Euler.Y=&RigVMFunction_MathDoubleAdd(A=0, B=0).Result
     Euler.Z=&RigVMFunction_MathDoubleAdd(A=-60, B=0).Result
     Euler.X=&RigVMFunction_MathDoubleAdd(A=0, B=0).Result
    ).Result
    ExecutePin=@RigVMFunction_Sequence_3.B
    Weight=@n70.Result
   )
   > &RigUnit_ModifyTransforms(
    ItemToModify=((Item=(Type=Bone,Name="dyn_dread_back_2_base"),Transform=(Rotation=(X=0,Y=0,Z=0,W=1),Translation=(X=-2,Y=0,Z=-1),Scale3D=(X=1,Y=1,Z=1))))
    WeightMinimum=0
    ItemToModify.0.Transform.Translation.Y=&RigVMFunction_MathDoubleAdd(A=0, B=0).Result
    ItemToModify.0.Transform.Translation.Z=&RigVMFunction_MathDoubleAdd(A=0, B=0).Result
    ItemToModify.0.Transform.Translation.X=&RigVMFunction_MathDoubleAdd(A=-2, B=0).Result
    ItemToModify.0.Transform.Rotation=&RigVMFunction_MathQuaternionFromEuler(
     Euler=(X=0,Y=45,Z=0)
     Euler.Y=&RigVMFunction_MathDoubleAdd(A=60, B=0).Result
     Euler.Z=&RigVMFunction_MathDoubleAdd(A=-20, B=0).Result
     Euler.X=&RigVMFunction_MathDoubleAdd(A=0, B=0).Result
    ).Result
    Weight=@n70.Result
   )
   > &RigUnit_ModifyTransforms(
    ItemToModify=((Item=(Type=Bone,Name="dyn_dread_back_3_base"),Transform=(Rotation=(X=0,Y=0,Z=0,W=1),Translation=(X=0,Y=0,Z=0),Scale3D=(X=1,Y=1,Z=1))))
    WeightMinimum=0
    ItemToModify.0.Transform.Translation.Y=&RigVMFunction_MathDoubleAdd(A=0, B=0).Result
    ItemToModify.0.Transform.Translation.Z=&RigVMFunction_MathDoubleAdd(A=0, B=0).Result
    ItemToModify.0.Transform.Translation.X=&RigVMFunction_MathDoubleAdd(A=-2, B=0).Result
    ItemToModify.0.Transform.Rotation=&RigVMFunction_MathQuaternionFromEuler(
     Euler=(X=-30,Y=0,Z=-30)
     Euler.Y=&RigVMFunction_MathDoubleAdd(A=-45, B=0).Result
     Euler.Z=&RigVMFunction_MathDoubleAdd(A=-60, B=0).Result
     Euler.X=&RigVMFunction_MathDoubleAdd(A=0, B=0).Result
    ).Result
    Weight=@n70.Result
   )
   n71 = &RigUnit_SphericalPoseReader(
    ActiveRegionScaleFactors=(PositiveWidth=0.1,NegativeWidth=0.1,PositiveHeight=0.1,NegativeHeight=0.1)
    ActiveRegionSize=0.1
    Debug=(bDrawDebug=true,bDraw2D=False,bDrawLocalAxes=False,DebugScale=25,DebugSegments=20,DebugThickness=0.25)
    DriverAxis=(X=1,Y=0,Z=0)
    DriverItem=(Type=Bone,Name="neck_01")
    FalloffRegionScaleFactors=(PositiveWidth=1,NegativeWidth=1,PositiveHeight=1,NegativeHeight=1)
    FalloffSize=0.05
    FlipHeightScaling=False
    FlipWidthScaling=False
    OptionalParentItem=(Type=Bone,Name="None")
    RotationOffset=(X=-70,Y=-40,Z=47.5)
    ExecutePin=@RigVMFunction_Sequence_1.E
   )
   Multiply = &RigVMFunction_MathDoubleMul(B=@n2.OutputParam, A=@n71.OutputParam)
   > &RigUnit_ModifyTransforms(
    ItemToModify=((Item=(Type=Bone,Name="dyn_dread_front_1_base_l"),Transform=(Rotation=(X=0,Y=0,Z=0,W=1),Translation=(X=0,Y=0,Z=0),Scale3D=(X=1,Y=1,Z=1))))
    WeightMinimum=0
    ItemToModify.0.Transform.Rotation=&RigVMFunction_MathQuaternionFromEuler(Euler=(X=0,Y=0,Z=0)).Result
    Weight=@Multiply.Result
   )
   > &RigUnit_ModifyTransforms(
    ItemToModify=((Item=(Type=Bone,Name="dyn_dread_front_1_base_r"),Transform=(Rotation=(X=0,Y=0,Z=0,W=1),Translation=(X=0,Y=0,Z=0),Scale3D=(X=1,Y=1,Z=1))))
    WeightMinimum=0
    ItemToModify.0.Transform.Rotation=&RigVMFunction_MathQuaternionFromEuler(Euler=(X=0,Y=0,Z=0)).Result
    Weight=@Multiply.Result
   )
   n72 = RigVMVariableNode() {variable=Dyn_crouch}
   n73 = &RigVMFunction_MathDoubleMul(A=0, B=@n72.Value)
   n74 = &RigVMFunction_MathDoubleMul(A=0, B=@n72.Value)
   n75 = &RigVMFunction_MathDoubleMul(A=0, B=@n72.Value)
   > &RigUnit_ModifyTransforms(
    ItemToModify=((Item=(Type=Bone,Name="dyn_dread_back_1_base"),Transform=(Rotation=(X=0,Y=0,Z=0,W=1),Translation=(X=-5,Y=2,Z=5),Scale3D=(X=1,Y=1,Z=1))))
    WeightMinimum=0
    ItemToModify.0.Transform.Translation.Y=&RigVMFunction_MathDoubleAdd(A=0, B=&RigVMFunction_MathDoubleMul(A=0, B=@n72.Value).Result).Result
    ItemToModify.0.Transform.Translation.Z=&RigVMFunction_MathDoubleAdd(A=2, B=&RigVMFunction_MathDoubleMul(A=0, B=@n72.Value).Result).Result
    ItemToModify.0.Transform.Translation.X=&RigVMFunction_MathDoubleAdd(A=-5, B=&RigVMFunction_MathDoubleMul(A=0, B=@n72.Value).Result).Result
    ItemToModify.0.Transform.Rotation=&RigVMFunction_MathQuaternionFromEuler(
     Euler=(X=-20,Y=40,Z=-60)
     Euler.Y=&RigVMFunction_MathDoubleAdd(A=-30, B=@n73.Result).Result
     Euler.Z=&RigVMFunction_MathDoubleAdd(A=0, B=@n74.Result).Result
     Euler.X=&RigVMFunction_MathDoubleAdd(A=0, B=@n75.Result).Result
    ).Result
    Weight=@Multiply.Result
   )
   n76 = RigVMVariableNode() {variable=Dyn_crouch}
   n77 = &RigVMFunction_MathDoubleMul(A=0, B=@n76.Value)
   n78 = &RigVMFunction_MathDoubleMul(A=-30, B=@n76.Value)
   n79 = &RigVMFunction_MathDoubleMul(A=60, B=@n76.Value)
   > &RigUnit_ModifyTransforms(
    ItemToModify=((Item=(Type=Bone,Name="dyn_dread_back_2_base"),Transform=(Rotation=(X=0,Y=0,Z=0,W=1),Translation=(X=-5,Y=0,Z=-3),Scale3D=(X=1,Y=1,Z=1))))
    WeightMinimum=0
    ItemToModify.0.Transform.Translation.X=&RigVMFunction_MathDoubleAdd(A=0, B=&RigVMFunction_MathDoubleMul(B=0, A=@n76.Value).Result).Result
    ItemToModify.0.Transform.Translation.Y=&RigVMFunction_MathDoubleAdd(A=0, B=&RigVMFunction_MathDoubleMul(A=0, B=@n76.Value).Result).Result
    ItemToModify.0.Transform.Translation.Z=&RigVMFunction_MathDoubleAdd(A=0, B=&RigVMFunction_MathDoubleMul(A=0, B=@n76.Value).Result).Result
    ItemToModify.0.Transform.Rotation=&RigVMFunction_MathQuaternionFromEuler(
     Euler=(X=0,Y=0,Z=0)
     Euler.Y=&RigVMFunction_MathDoubleAdd(A=60, B=@n77.Result).Result
     Euler.Z=&RigVMFunction_MathDoubleAdd(A=-20, B=@n78.Result).Result
     Euler.X=&RigVMFunction_MathDoubleAdd(A=0, B=@n79.Result).Result
    ).Result
    Weight=@Multiply.Result
   )
   n80 = RigVMVariableNode() {variable=Dyn_crouch}
   n81 = &RigVMFunction_MathDoubleMul(A=30, B=@n80.Value)
   n82 = &RigVMFunction_MathDoubleMul(A=0, B=@n80.Value)
   n83 = &RigVMFunction_MathDoubleMul(A=0, B=@n80.Value)
   > &RigUnit_ModifyTransforms(
    ItemToModify=((Item=(Type=Bone,Name="dyn_dread_back_3_base"),Transform=(Rotation=(X=0,Y=0,Z=0,W=1),Translation=(X=-3,Y=-3,Z=2),Scale3D=(X=1,Y=1,Z=1))))
    WeightMinimum=0
    ItemToModify.0.Transform.Translation.Z=&RigVMFunction_MathDoubleAdd(A=2, B=&RigVMFunction_MathDoubleMul(A=1, B=@n80.Value).Result).Result
    ItemToModify.0.Transform.Translation.X=&RigVMFunction_MathDoubleAdd(A=-1, B=&RigVMFunction_MathDoubleMul(A=1, B=@n80.Value).Result).Result
    ItemToModify.0.Transform.Translation.Y=&RigVMFunction_MathDoubleAdd(A=-3, B=&RigVMFunction_MathDoubleMul(A=1, B=@n80.Value).Result).Result
    ItemToModify.0.Transform.Rotation=&RigVMFunction_MathQuaternionFromEuler(
     Euler=(X=0,Y=0,Z=0)
     Euler.Z=&RigVMFunction_MathDoubleAdd(A=-60, B=@n81.Result).Result
     Euler.Y=&RigVMFunction_MathDoubleAdd(A=0, B=@n82.Result).Result
     Euler.X=&RigVMFunction_MathDoubleAdd(A=0, B=@n83.Result).Result
    ).Result
    Weight=@Multiply.Result
   )
   n84 = &RigVMFunction_Sequence(ExecuteContext=@RigVMFunction_Sequence.C)
   n85 = &RigUnit_SphericalPoseReader(
    ActiveRegionScaleFactors=(PositiveWidth=0.5,NegativeWidth=0.5,PositiveHeight=0.5,NegativeHeight=0.5)
    ActiveRegionSize=0.1
    Debug=(bDrawDebug=true,bDraw2D=False,bDrawLocalAxes=False,DebugScale=25,DebugSegments=20,DebugThickness=0.25)
    DriverAxis=(X=1,Y=0,Z=0)
    DriverItem=(Type=Bone,Name="clavicle_l")
    FalloffRegionScaleFactors=(PositiveWidth=1,NegativeWidth=1,PositiveHeight=1,NegativeHeight=1)
    FalloffSize=0.25
    FlipHeightScaling=False
    FlipWidthScaling=False
    OptionalParentItem=(Type=Bone,Name="None")
    RotationOffset=(X=0,Y=-45,Z=0)
    ExecutePin=@n84.A
   )
   n86 = RigVMVariableNode() {variable=Dyn_crouch}
   n87 = &RigVMFunction_MathDoubleMul(A=10, B=@n86.Value)
   n88 = &RigVMFunction_MathDoubleMul(A=60, B=@n86.Value)
   n89 = &RigVMFunction_MathDoubleMul(A=0, B=@n86.Value)
   n90 = RigVMVariableNode() {variable=Dyn_Parachute}
   n91 = &RigVMFunction_MathDoubleMul(
    A=&RigVMFunction_MathDoubleMul(A=@n85.OutputParam, B=&RigVMFunction_MathDoubleSub(A=1, B=@n90.Value).Result).Result
    B=&RigVMFunction_MathDoubleSub(A=1, B=RigVMVariableNode() {variable=Dyn_Emote}.Value).Result
   )
   > &RigUnit_ModifyTransforms(
    ItemToModify=((Item=(Type=Bone,Name="dyn_dread_back_1_base"),Transform=(Rotation=(X=0,Y=0,Z=0,W=1),Translation=(X=0,Y=1.5,Z=0),Scale3D=(X=1,Y=1,Z=1))))
    WeightMinimum=0
    ItemToModify.0.Transform.Translation.Y=&RigVMFunction_MathDoubleAdd(A=2, B=&RigVMFunction_MathDoubleMul(A=-1, B=@n86.Value).Result).Result
    ItemToModify.0.Transform.Translation.Z=&RigVMFunction_MathDoubleAdd(A=0, B=&RigVMFunction_MathDoubleMul(A=0, B=@n86.Value).Result).Result
    ItemToModify.0.Transform.Translation.X=&RigVMFunction_MathDoubleAdd(A=0, B=&RigVMFunction_MathDoubleMul(A=2, B=@n86.Value).Result).Result
    ItemToModify.0.Transform.Rotation=&RigVMFunction_MathQuaternionFromEuler(
     Euler=(X=-5,Y=-45,Z=30)
     Euler.Z=&RigVMFunction_MathDoubleAdd(A=20, B=@n87.Result).Result
     Euler.Y=&RigVMFunction_MathDoubleAdd(A=-30, B=@n88.Result).Result
     Euler.X=&RigVMFunction_MathDoubleAdd(A=-20, B=@n89.Result).Result
    ).Result
    Weight=@n91.Result
   )
   n92 = RigVMVariableNode() {variable=Dyn_crouch}
   n93 = &RigVMFunction_MathDoubleAdd(A=0, B=&RigVMFunction_MathDoubleMul(A=-60, B=@n92.Value).Result)
   n94 = &RigVMFunction_MathDoubleMul(A=75, B=&RigVMFunction_MathDoubleSub(A=1, B=RigVMVariableNode() {variable=Dyn_pelvis_rotate_alfa}.Value).Result)
   n95 = &RigVMFunction_MathDoubleMul(A=0, B=@n92.Value)
   n96 = &RigVMFunction_MathDoubleMul(A=-30, B=@n92.Value)
   > &RigUnit_ModifyTransforms(
    ItemToModify=((Item=(Type=Bone,Name="dyn_dread_back_2_base"),Transform=(Rotation=(X=0,Y=0,Z=0,W=1),Translation=(X=-1,Y=0,Z=0),Scale3D=(X=1,Y=1,Z=1))))
    WeightMinimum=0
    ItemToModify.0.Transform.Translation.Y=&RigVMFunction_MathDoubleAdd(A=0, B=&RigVMFunction_MathDoubleMul(A=2, B=@n92.Value).Result).Result
    ItemToModify.0.Transform.Translation.Z=&RigVMFunction_MathDoubleAdd(A=2, B=&RigVMFunction_MathDoubleMul(A=-4, B=@n92.Value).Result).Result
    ItemToModify.0.Transform.Translation.X=&RigVMFunction_MathDoubleAdd(A=-1, B=&RigVMFunction_MathDoubleMul(A=1, B=@n92.Value).Result).Result
    ItemToModify.0.Transform.Rotation=&RigVMFunction_MathQuaternionFromEuler(
     Euler=(X=-20,Y=-25,Z=20)
     Euler.Y=&RigVMFunction_MathDoubleAdd(A=@n93.Result, B=@n94.Result).Result
     Euler.Z=&RigVMFunction_MathDoubleAdd(A=0, B=@n95.Result).Result
     Euler.X=&RigVMFunction_MathDoubleAdd(A=30, B=@n96.Result).Result
    ).Result
    Weight=@n91.Result
   )
   n97 = &RigUnit_SphericalPoseReader(
    ActiveRegionScaleFactors=(PositiveWidth=0.25,NegativeWidth=0.25,PositiveHeight=0.25,NegativeHeight=0.25)
    ActiveRegionSize=0.1
    Debug=(bDrawDebug=true,bDraw2D=False,bDrawLocalAxes=False,DebugScale=25,DebugSegments=20,DebugThickness=0.25)
    DriverAxis=(X=1,Y=0,Z=0)
    DriverItem=(Type=Bone,Name="clavicle_l")
    FalloffRegionScaleFactors=(PositiveWidth=1,NegativeWidth=1,PositiveHeight=1,NegativeHeight=1)
    FalloffSize=0.15
    FlipHeightScaling=False
    FlipWidthScaling=False
    OptionalParentItem=(Type=Bone,Name="None")
    RotationOffset=(X=10,Y=-100,Z=0)
    ExecutePin=@n84.B
   )
   n98 = RigVMVariableNode() {variable=Dyn_crouch}
   n99 = &RigVMFunction_MathDoubleMul(A=-10, B=@n98.Value)
   n100 = &RigVMFunction_MathDoubleMul(A=-10, B=@n98.Value)
   n101 = &RigVMFunction_MathDoubleMul(A=20, B=@n98.Value)
   n102 = &RigVMFunction_MathDoubleSub(A=1, B=RigVMVariableNode() {variable=Dyn_Parachute}.Value)
   n103 = RigVMVariableNode() {variable=Dyn_Emote}
   > &RigUnit_ModifyTransforms(
    ItemToModify=((Item=(Type=Bone,Name="dyn_dread_back_1_base"),Transform=(Rotation=(X=0,Y=0,Z=0,W=1),Translation=(X=0,Y=1.5,Z=0),Scale3D=(X=1,Y=1,Z=1))))
    WeightMinimum=0
    ItemToModify.0.Transform.Translation.Y=&RigVMFunction_MathDoubleAdd(A=0, B=&RigVMFunction_MathDoubleMul(A=0, B=@n98.Value).Result).Result
    ItemToModify.0.Transform.Translation.Z=&RigVMFunction_MathDoubleAdd(A=1, B=&RigVMFunction_MathDoubleMul(A=-1, B=@n98.Value).Result).Result
    ItemToModify.0.Transform.Translation.X=&RigVMFunction_MathDoubleAdd(A=0, B=&RigVMFunction_MathDoubleMul(A=0, B=@n98.Value).Result).Result
    ItemToModify.0.Transform.Rotation=&RigVMFunction_MathQuaternionFromEuler(
     Euler=(X=-5,Y=-45,Z=30)
     Euler.Z=&RigVMFunction_MathDoubleAdd(A=20, B=@n99.Result).Result
     Euler.Y=&RigVMFunction_MathDoubleAdd(A=-20, B=@n100.Result).Result
     Euler.X=&RigVMFunction_MathDoubleAdd(A=-20, B=@n101.Result).Result
    ).Result
    Weight=&RigVMFunction_MathDoubleMul(
     A=&RigVMFunction_MathDoubleMul(A=@n97.OutputParam, B=@n102.Result).Result
     B=&RigVMFunction_MathDoubleSub(A=1, B=@n103.Value).Result
    ).Result
   )
   n104 = &RigVMFunction_Sequence(ExecuteContext=@RigVMFunction_Sequence.D)
   n105 = &RigUnit_SphericalPoseReader(
    ActiveRegionScaleFactors=(PositiveWidth=0.5,NegativeWidth=0.1,PositiveHeight=0.5,NegativeHeight=0.5)
    ActiveRegionSize=0.1
    Debug=(bDrawDebug=true,bDraw2D=False,bDrawLocalAxes=False,DebugScale=25,DebugSegments=20,DebugThickness=0.25)
    DriverAxis=(X=-1,Y=0,Z=0)
    DriverItem=(Type=Bone,Name="clavicle_r")
    FalloffRegionScaleFactors=(PositiveWidth=1,NegativeWidth=1,PositiveHeight=1,NegativeHeight=1)
    FalloffSize=0.2
    FlipHeightScaling=False
    FlipWidthScaling=False
    OptionalParentItem=(Type=Bone,Name="None")
    RotationOffset=(X=-180,Y=-45,Z=0)
    ExecutePin=@n104.A
   )
   n106 = RigVMVariableNode() {variable=Dyn_Parachute}
   n107 = &RigVMFunction_MathDoubleMul(
    A=&RigVMFunction_MathDoubleMul(B=@n105.OutputParam, A=&RigVMFunction_MathDoubleSub(A=1, B=@n106.Value).Result).Result
    B=&RigVMFunction_MathDoubleSub(A=1, B=RigVMVariableNode() {variable=Dyn_Emote}.Value).Result
   )
   > &RigUnit_ModifyTransforms(
    ItemToModify=((Item=(Type=Bone,Name="dyn_dread_front_1_base_r"),Transform=(Rotation=(X=0,Y=0,Z=0,W=1),Translation=(X=0,Y=0,Z=0),Scale3D=(X=1,Y=1,Z=1))))
    WeightMinimum=0
    ItemToModify.0.Transform.Rotation=&RigVMFunction_MathQuaternionFromEuler(Euler=(X=0,Y=20,Z=0)).Result
    Weight=@n107.Result
   )
   > &RigUnit_ModifyTransforms(
    ItemToModify=((Item=(Type=Bone,Name="dyn_dread_back_1_base"),Transform=(Rotation=(X=0,Y=0,Z=0,W=1),Translation=(X=-1,Y=0,Z=0),Scale3D=(X=1,Y=1,Z=1))))
    WeightMinimum=0
    ItemToModify.0.Transform.Rotation=&RigVMFunction_MathQuaternionFromEuler(Euler=(X=0,Y=0,Z=0)).Result
    Weight=@n107.Result
   )
   n108 = RigVMVariableNode() {variable=Dyn_crouch}
   n109 = &RigVMFunction_MathDoubleMul(A=60, B=@n108.Value)
   n110 = &RigVMFunction_MathDoubleMul(A=-30, B=@n108.Value)
   n111 = &RigVMFunction_MathDoubleMul(A=0, B=@n108.Value)
   > &RigUnit_ModifyTransforms(
    ItemToModify=((Item=(Type=Bone,Name="dyn_dread_back_2_base"),Transform=(Rotation=(X=0,Y=0,Z=0,W=1),Translation=(X=-1,Y=-2,Z=0),Scale3D=(X=1,Y=1,Z=1))))
    WeightMinimum=0
    ItemToModify.0.Transform.Translation.X=&RigVMFunction_MathDoubleAdd(A=-1, B=&RigVMFunction_MathDoubleMul(A=1, B=@n108.Value).Result).Result
    ItemToModify.0.Transform.Translation.Y=&RigVMFunction_MathDoubleAdd(A=-2, B=&RigVMFunction_MathDoubleMul(A=0, B=@n108.Value).Result).Result
    ItemToModify.0.Transform.Translation.Z=&RigVMFunction_MathDoubleAdd(A=0, B=&RigVMFunction_MathDoubleMul(A=0, B=@n108.Value).Result).Result
    ItemToModify.0.Transform.Rotation=&RigVMFunction_MathQuaternionFromEuler(
     Euler=(X=-60,Y=30,Z=0)
     Euler.X=&RigVMFunction_MathDoubleAdd(A=-60, B=@n109.Result).Result
     Euler.Y=&RigVMFunction_MathDoubleAdd(A=30, B=@n110.Result).Result
     Euler.Z=&RigVMFunction_MathDoubleAdd(A=0, B=@n111.Result).Result
    ).Result
    Weight=@n107.Result
   )
   n112 = RigVMVariableNode() {variable=Dyn_crouch}
   n113 = &RigVMFunction_MathDoubleMul(A=-20, B=@n112.Value)
   n114 = &RigVMFunction_MathDoubleMul(A=30, B=@n112.Value)
   n115 = &RigVMFunction_MathDoubleMul(A=30, B=@n112.Value)
   > &RigUnit_ModifyTransforms(
    ItemToModify=((Item=(Type=Bone,Name="dyn_dread_back_3_base"),Transform=(Rotation=(X=0,Y=0,Z=0,W=1),Translation=(X=-1,Y=0,Z=1),Scale3D=(X=1,Y=1,Z=1))))
    WeightMinimum=0
    ItemToModify.0.Transform.Rotation=&RigVMFunction_MathQuaternionFromEuler(
     Euler=(X=-30,Y=-20,Z=-20)
     Euler.Z=&RigVMFunction_MathDoubleAdd(A=-20, B=@n113.Result).Result
     Euler.X=&RigVMFunction_MathDoubleAdd(A=-30, B=@n114.Result).Result
     Euler.Y=&RigVMFunction_MathDoubleAdd(A=-20, B=@n115.Result).Result
    ).Result
    Weight=@n107.Result
   )
   n116 = &RigUnit_SphericalPoseReader(
    ActiveRegionScaleFactors=(PositiveWidth=0.5,NegativeWidth=0.5,PositiveHeight=0.5,NegativeHeight=0.5)
    ActiveRegionSize=0.1
    Debug=(bDrawDebug=true,bDraw2D=False,bDrawLocalAxes=False,DebugScale=25,DebugSegments=20,DebugThickness=0.25)
    DriverAxis=(X=-1,Y=0,Z=0)
    DriverItem=(Type=Bone,Name="upperarm_r")
    FalloffRegionScaleFactors=(PositiveWidth=1,NegativeWidth=1,PositiveHeight=1,NegativeHeight=1)
    FalloffSize=0.2
    FlipHeightScaling=False
    FlipWidthScaling=False
    OptionalParentItem=(Type=Bone,Name="None")
    RotationOffset=(X=-115,Y=0,Z=45)
    ExecutePin=@n104.B
   )
   n117 = &RigVMFunction_MathDoubleMul(A=@n116.OutputParam, B=&RigVMFunction_MathDoubleSub(A=1, B=RigVMVariableNode() {variable=Dyn_Parachute}.Value).Result)
   > &RigUnit_ModifyTransforms(
    ItemToModify=((Item=(Type=Bone,Name="dyn_dread_front_1_base_r"),Transform=(Rotation=(X=0,Y=0,Z=0,W=1),Translation=(X=-1,Y=-1,Z=1),Scale3D=(X=1,Y=1,Z=1))))
    WeightMinimum=0
    ItemToModify.0.Transform.Rotation=&RigVMFunction_MathQuaternionFromEuler(Euler=(X=0,Y=0,Z=0)).Result
    Weight=@n117.Result
   )
   n118 = RigVMVariableNode() {variable=Dyn_crouch}
   n119 = &RigVMFunction_MathDoubleMul(A=-20, B=@n118.Value)
   n120 = &RigVMFunction_MathDoubleMul(A=0, B=@n118.Value)
   n121 = &RigVMFunction_MathDoubleMul(A=0, B=@n118.Value)
   > &RigUnit_ModifyTransforms(
    ItemToModify=((Item=(Type=Bone,Name="dyn_dread_back_1_base"),Transform=(Rotation=(X=0,Y=0,Z=0,W=1),Translation=(X=-2,Y=0,Z=0),Scale3D=(X=1,Y=1,Z=1))))
    WeightMinimum=0
    ItemToModify.0.Transform.Translation.Y=&RigVMFunction_MathDoubleAdd(A=0, B=&RigVMFunction_MathDoubleMul(A=0, B=@n118.Value).Result).Result
    ItemToModify.0.Transform.Translation.X=&RigVMFunction_MathDoubleAdd(A=0, B=&RigVMFunction_MathDoubleMul(A=0, B=@n118.Value).Result).Result
    ItemToModify.0.Transform.Rotation=&RigVMFunction_MathQuaternionFromEuler(
     Euler=(X=0,Y=0,Z=25)
     Euler.Z=&RigVMFunction_MathDoubleAdd(A=0, B=@n119.Result).Result
     Euler.Y=&RigVMFunction_MathDoubleAdd(A=0, B=@n120.Result).Result
     Euler.X=&RigVMFunction_MathDoubleAdd(A=0, B=@n121.Result).Result
    ).Result
    Weight=@n117.Result
   )
   n122 = RigVMVariableNode() {variable=Dyn_crouch}
   n123 = &RigVMFunction_MathDoubleMul(A=20, B=@n122.Value)
   n124 = &RigVMFunction_MathDoubleMul(A=0, B=@n122.Value)
   n125 = &RigVMFunction_MathDoubleMul(A=0, B=@n122.Value)
   > &RigUnit_ModifyTransforms(
    ItemToModify=((Item=(Type=Bone,Name="dyn_dread_back_2_base"),Transform=(Rotation=(X=0,Y=0,Z=0,W=1),Translation=(X=0,Y=0,Z=0),Scale3D=(X=1,Y=1,Z=1))))
    WeightMinimum=0
    ItemToModify.0.Transform.Translation.Y=&RigVMFunction_MathDoubleAdd(A=0, B=&RigVMFunction_MathDoubleMul(A=0, B=@n122.Value).Result).Result
    ItemToModify.0.Transform.Translation.X=&RigVMFunction_MathDoubleAdd(A=0, B=&RigVMFunction_MathDoubleMul(A=0, B=@n122.Value).Result).Result
    ItemToModify.0.Transform.Rotation=&RigVMFunction_MathQuaternionFromEuler(
     Euler=(X=-10,Y=0,Z=0)
     Euler.Z=&RigVMFunction_MathDoubleAdd(A=0, B=@n123.Result).Result
     Euler.Y=&RigVMFunction_MathDoubleAdd(A=0, B=@n124.Result).Result
     Euler.X=&RigVMFunction_MathDoubleAdd(A=-10, B=@n125.Result).Result
    ).Result
    Weight=@n117.Result
   )
   n126 = RigVMVariableNode() {variable=Dyn_crouch}
   n127 = &RigVMFunction_MathDoubleMul(A=0, B=@n126.Value)
   n128 = &RigVMFunction_MathDoubleMul(A=0, B=@n126.Value)
   n129 = &RigVMFunction_MathDoubleMul(A=20, B=@n126.Value)
   > &RigUnit_ModifyTransforms(
    ItemToModify=((Item=(Type=Bone,Name="dyn_dread_back_3_base"),Transform=(Rotation=(X=0,Y=0,Z=0,W=1),Translation=(X=-1,Y=2,Z=0),Scale3D=(X=1,Y=1,Z=1))))
    WeightMinimum=0
    ItemToModify.0.Transform.Translation.Y=&RigVMFunction_MathDoubleAdd(A=6, B=&RigVMFunction_MathDoubleMul(A=-4, B=@n126.Value).Result).Result
    ItemToModify.0.Transform.Translation.X=&RigVMFunction_MathDoubleAdd(A=0, B=&RigVMFunction_MathDoubleMul(A=0, B=@n126.Value).Result).Result
    ItemToModify.0.Transform.Rotation=&RigVMFunction_MathQuaternionFromEuler(
     Euler=(X=0,Y=0,Z=-30)
     Euler.Z=&RigVMFunction_MathDoubleAdd(A=20, B=@n127.Result).Result
     Euler.Y=&RigVMFunction_MathDoubleAdd(A=0, B=@n128.Result).Result
     Euler.X=&RigVMFunction_MathDoubleAdd(A=0, B=@n129.Result).Result
    ).Result
    Weight=@n117.Result
   )
   &RigVMFunction_MathDoubleAdd(A=-15, B=0)
   &RigVMFunction_MathDoubleAdd(A=20, B=0)
  ]}
]
hierarchy [
 root {loc=<0 0 0>, rot=<0 0 0 1>}
 pelvis {parent=root, loc=<0 -0.5301 98.0571>, rot=<0.0127 0.707 -0.0127 -0.707>}
 spine_01 {parent=pelvis, loc=<6.1208 -0.9328 0>, rot=<0 0 0.034 -0.9994>}
 spine_02 {parent=spine_01, loc=<6.5997 -0.2095 0>, rot=<0 0 0.0238 -0.9997>}
 spine_03 {parent=spine_02, loc=<6.4914 -0.1237 0>, rot=<0 0 0.0149 -0.9999>}
 spine_04 {parent=spine_03, loc=<6.9598 0.0794 0>, rot=<0 0 -0.0132 -0.9999>}
 spine_05 {parent=spine_04, loc=<6.9987 0 0>, rot=<0 0 -0.0015 -1>}
 neck_01 {parent=spine_05, loc=<11.3138 0.1529 0>, rot=<0 0 -0.2003 -0.9797>}
 neck_02 {parent=neck_01, loc=<4.5627 0 0>, rot=<0 0 0.0504 -0.9987>}
 head {parent=neck_02, loc=<5.3914 0 0>, rot=<0 0 0.115 -0.9934>}
 dyn_dread_front_1_base_l {parent=head, loc=<2.4456 3.0248 -7.5437>, rot=<0.042 0.001 0.999 0.0153>}
 dyn_dread_front_1_1_l {parent=dyn_dread_front_1_base_l, loc=<0 0 0>, rot=<0 0 0 1>}
 dyn_dread_front_1_2_l {parent=dyn_dread_front_1_1_l, loc=<11.5406 0.3972 0>, rot=<-0.1485 -0.037 0.1077 -0.9823>}
 dyn_dread_front_1_3_l {parent=dyn_dread_front_1_2_l, loc=<6.242 0 0>, rot=<0 0 0 1>}
 dyn_dread_front_1_base_r {parent=head, loc=<1.6113 3.2915 7.3464>, rot=<-0.0247 -0.0015 0.9983 0.0535>}
 dyn_dread_front_1_1_r {parent=dyn_dread_front_1_base_r, loc=<0 0 0>, rot=<0 0 0 1>}
 dyn_dread_front_1_2_r {parent=dyn_dread_front_1_1_r, loc=<10.8556 0.2831 0>, rot=<-0.4234 0.0661 0.01 -0.9035>}
 dyn_dread_front_1_3_r {parent=dyn_dread_front_1_2_r, loc=<7.6137 0 0>, rot=<0 0 0 1>}
 dyn_dread_back_1_base {parent=head, loc=<-0.3046 -4.305 -6.0159>, rot=<0.1507 -0.9866 -0.0088 0.0611>}
 dyn_dread_back_1_1 {parent=dyn_dread_back_1_base, loc=<0 0 0>, rot=<0 0 0 1>}
 dyn_dread_back_1_3 {parent=dyn_dread_back_1_1, loc=<8.2096 -0.0543 -0.1019>, rot=<0.0174 -0.0405 0.11 -0.9929>}
 dyn_dread_back_1_4 {parent=dyn_dread_back_1_3, loc=<8.523 0.3369 1.0282>, rot=<-0.0199 0.0744 -0.2504 -0.9651>}
 dyn_dread_back_1_5 {parent=dyn_dread_back_1_3, loc=<9.0949 -0.1816 -0.7502>, rot=<-0.0222 0.0857 -0.2585 -0.9619>}
 dyn_dread_back_1_6 {parent=dyn_dread_back_1_5, loc=<8.4243 0 0>, rot=<0 0 0 1>}
 dyn_dread_back_2_base {parent=head, loc=<-0.9063 -6.9384 0>, rot=<0.1308 -0.7639 -0.6125 0.1557>}
 dyn_dread_back_2_1 {parent=dyn_dread_back_2_base, loc=<0 0 0>, rot=<-0.6191 0 0 -0.7853>}
 dyn_dread_back_2_2 {parent=dyn_dread_back_2_1, loc=<7.5127 0.5531 0.5739>, rot=<-0.0728 0.0027 -0.0372 -0.9967>}
 dyn_dread_back_2_3 {parent=dyn_dread_back_2_2, loc=<7.9881 0 1.5229>, rot=<0 0 -0.1848 -0.9828>}
 dyn_dread_back_2_4 {parent=dyn_dread_back_2_2, loc=<9.1628 0.5711 -1.865>, rot=<0.1049 0.0112 -0.1528 -0.9826>}
 dyn_dread_back_2_5 {parent=dyn_dread_back_2_2, loc=<8.1976 0.5466 -0.2341>, rot=<0.1376 0.0074 -0.1498 -0.9791>}
 dyn_dread_back_2_6 {parent=dyn_dread_back_2_5, loc=<5.1104 0 0.2832>, rot=<0 0 0 1>}
 dyn_dread_back_3_base {parent=head, loc=<-0.9193 -5.1124 5.8978>, rot=<0.254 -0.9652 0.0154 -0.0604>}
 dyn_dread_back_3_1 {parent=dyn_dread_back_3_base, loc=<0 0 0>, rot=<0.0249 -0.0764 -0.1898 -0.9785>}
 dyn_dread_back_3_2 {parent=dyn_dread_back_3_1, loc=<7.5953 -0.5628 0.5223>, rot=<-0.1062 0.0642 0.1804 -0.9757>}
 dyn_dread_back_3_3 {parent=dyn_dread_back_3_2, loc=<4.878 0 0>, rot=<0 0 0 1>}
 dyn_dread_back_3_4 {parent=dyn_dread_back_3_1, loc=<7.5707 -2.5373 2.4138>, rot=<-0.3009 0.0603 0.116 -0.9446>}
 dyn_dread_back_3_5 {parent=dyn_dread_back_3_4, loc=<5.9905 0 0>, rot=<0.3193 -0.0434 -0.1274 -0.9381>}
 clavicle_l {parent=spine_05, loc=<5.4831 -0.1053 -3.2946>, rot=<0.0465 -0.7289 -0.009 -0.6829>}
 upperarm_l {parent=clavicle_l, loc=<9.8658 -1.2136 -0.4105>, rot=<0.0164 -0.3245 -0.0269 -0.9454>}
 lowerarm_l {parent=upperarm_l, loc=<26.465 -0.8345 0.259>, rot=<0.0036 0.0161 -0.2298 -0.9731>}
 clavicle_r {parent=spine_05, loc=<5.4831 -0.1053 3.2946>, rot=<0.7289 0.0465 -0.6829 0.009>}
 upperarm_r {parent=clavicle_r, loc=<-9.8658 1.2136 0.4105>, rot=<0.0164 -0.3245 -0.0269 -0.9454>}
 lowerarm_r {parent=upperarm_r, loc=<-26.465 0.8345 -0.259>, rot=<0.0036 0.0161 -0.2298 -0.9731>}
 deform_pec_aim_base_l {parent=spine_04, loc=<3.2456 9.3309 -1.2237>, rot=<0.0603 -0.4033 0.3318 -0.8507>}
 deform_pec_aim_l {parent=deform_pec_aim_base_l, loc=<0 0 0>, rot=<0 0 0 1>}
 deform_pec_aim_base_r {parent=spine_04, loc=<3.2452 9.3309 1.2237>, rot=<-0.4033 -0.0603 0.8507 0.3318>}
 deform_pec_aim_r {parent=deform_pec_aim_base_r, loc=<0 0 0>, rot=<0 0 0 1>}
]
variables [
 Dyn_Emote {type=double, default=0}
 Dyn_FrontEnd {type=double, default=1}
 Dyn_Parachute {type=double, default=0}
 Dyn_crouch {type=double, default=0}
 Dyn_fly {type=double, default=0}
 Dyn_pelvis_rotate_alfa {type=float, default=0}
]
