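# Physics asset for the TealMink Pelt female medium body's FaceAcc mesh, simulating its dreadlocks.
# Capsule bodies cover two front dreads, three back dreads, and the head, neck, upper spine, clavicles and arms.
# Swing and twist limited constraints link the dread segments to their bases and each other, with some branching.
# Extra constraints with free angular motion cross-link neighboring back dreads and some segments within a dread, tying the back group together.
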PhysicsAsset /BRCosmetics/Characters/Player/Female/Medium/Bodies/F_MED_TealMink_Pelt/Meshes/Parts/F_MED_TealMink_Pelt_FaceAcc_Physics
bodies {
 clavicle_l [{
   shapeType=Capsule
   length=8.505
   radius=5.505
   shapeName=clavicle_l_capsule
   center=<5.5068 -1.0688 -1.5986>
   rotation=<rot -0.6857 -84.9909 94.9656>
  }]
 clavicle_r [{
   shapeType=Capsule
   length=8.005
   radius=6.005
   shapeName=clavicle_r_capsule
   center=<-5.5849 1.1499 2.2535>
   rotation=<rot -0.2508 -89.9721 95.0063>
  }]
 dyn_dread_back_1_1 [{
   shapeType=Capsule
   length=6.9976
   radius=1.5242
   shapeName=dyn_dread_back_1_1_capsule
   center=<6.5058 1.7872 1.4658>
   rotation=<rot -21.6418 -97.173 -83.0454>
  }]
 dyn_dread_back_1_3 [{
   shapeType=Capsule
   length=5.4414
   radius=1.8831
   shapeName=dyn_dread_back_1_3_capsule
   center=<4.5012 0.1812 0.1813>
   rotation=<rot 0.0002 -102.8078 -93.4278>
  }]
 dyn_dread_back_1_4 [{
   shapeType=Capsule
   length=6.619
   radius=1.2402
   shapeName=dyn_dread_back_1_4_capsule
   center=<5.2429 -0.4819 0.4996>
   rotation=<rot 0 90.0499 -87.0794>
  }]
 dyn_dread_back_1_5 [{
   shapeType=Capsule
   length=6.2502
   radius=0.75
   shapeName=dyn_dread_back_1_5_capsule
   center=<4.6173 -0.0865 -0.0389>
   rotation=<rot -0.0218 90.7082 -91.103>
  }]
 dyn_dread_back_1_6 [{
   shapeType=Capsule
   length=3.5273
   radius=0.8543
   shapeName=dyn_dread_back_1_6_capsule
   center=<3.1179 0.0708 -0.0201>
   rotation=<rot 0 -88.9376 -89.8894>
  }]
 dyn_dread_back_1_base [{shapeType=Capsule, length=0, radius=2, shapeName=dyn_dread_back_1_base_capsule, rotation=<rot 0 -90 0>}]
 dyn_dread_back_2_1 [{
   shapeType=Capsule
   length=3.5
   radius=1.505
   shapeName=dyn_dread_back_2_1_capsule
   center=<5.9251 -0.0267 0.7549>
   rotation=<rot -5 -179.9996 180>
  }]
 dyn_dread_back_2_2 [{
   shapeType=Capsule
   length=3.346
   radius=2.3204
   shapeName=dyn_dread_back_2_2_capsule
   center=<3.5281 0.2001 -0.0153>
   rotation=<rot -0.1341 -107.0105 -90.5005>
  }]
 dyn_dread_back_2_3 [{
   shapeType=Capsule
   length=6.1007
   radius=0.5242
   shapeName=dyn_dread_back_2_3_capsule
   center=<4.2694 -0.113 -0.0234>
   rotation=<rot -0.1719 86.0617 -93.0349>
  }]
 dyn_dread_back_2_4 [{
   shapeType=Capsule
   length=7.8692
   radius=1.0749
   shapeName=dyn_dread_back_2_4_capsule
   center=<3.3555 0.3389 0.1463>
   rotation=<rot 0.4693 84.611 -87.6625>
  }]
 dyn_dread_back_2_5 [{
   shapeType=Capsule
   length=4.3372
   radius=0.8079
   shapeName=dyn_dread_back_2_5_capsule
   center=<2.4947 0.0404 0.2223>
   rotation=<rot 0.0561 -90.5349 -89.3599>
  }]
 dyn_dread_back_2_6 [{
   shapeType=Capsule
   length=4.9113
   radius=0.75
   shapeName=dyn_dread_back_2_6_capsule
   center=<3.5427 -0.0685 0.1214>
   rotation=<rot 0 89.9648 -87.5332>
  }]
 dyn_dread_back_2_base [{shapeType=Capsule, length=0, radius=1.5, shapeName=dyn_dread_back_2_base_capsule, rotation=<rot 0 -90 0>}]
 dyn_dread_back_3_1 [{
   shapeType=Capsule
   length=7.5888
   radius=1.8589
   shapeName=dyn_dread_back_3_1_capsule
   center=<6.5956 -0.0957 0.4067>
   rotation=<rot -7.5203 65.0424 -72.7294>
  }]
 dyn_dread_back_3_2 [{
   shapeType=Capsule
   length=5.6191
   radius=1.0945
   shapeName=dyn_dread_back_3_2_capsule
   center=<2.7241 0.3108 -0.017>
   rotation=<rot 0.7083 70.9233 -83.4661>
  }]
 dyn_dread_back_3_3 [{
   shapeType=Capsule
   length=4.7511
   radius=1.5808
   shapeName=dyn_dread_back_3_3_capsule
   center=<2.7184 -0.3597 -0.2844>
   rotation=<rot 14.9988 -83.2778 -89.2668>
  }]
 dyn_dread_back_3_4 [{
   shapeType=Capsule
   length=5.7518
   radius=1.3054
   shapeName=dyn_dread_back_3_4_capsule
   center=<3.9062 -0.5955 0.176>
   rotation=<rot 0 93.0774 -93.0024>
  }]
 dyn_dread_back_3_5 [{
   shapeType=Capsule
   length=4.3288
   radius=0.8971
   shapeName=dyn_dread_back_3_5_capsule
   center=<2.818 0.4294 0.0449>
   rotation=<rot -0.0124 -83.5062 -89.8588>
  }]
 dyn_dread_back_3_base [{shapeType=Capsule, length=0, radius=1.5, shapeName=dyn_dread_back_3_base_capsule, rotation=<rot 0 -90 0>}]
 dyn_dread_front_1_1_l [{
   shapeType=Capsule
   length=6.211
   radius=1.6751
   shapeName=dyn_dread_front_1_1_l_capsule
   center=<6.5197 -0.2456 -0.4478>
   rotation=<rot -4.2544 -16.1764 -77.5383>
  }]
 dyn_dread_front_1_1_r [{
   shapeType=Capsule
   length=4.708
   radius=1.6209
   shapeName=dyn_dread_front_1_1_r_capsule
   center=<5.2001 -0.2154 0.2443>
   rotation=<rot 8.7931 147.1722 -65.6428>
  }]
 dyn_dread_front_1_2_l [{
   shapeType=Capsule
   length=5.3312
   radius=1.4794
   shapeName=dyn_dread_front_1_2_l_capsule
   center=<3.3816 0.4684 0.1242>
   rotation=<rot -2.1751 78.1672 -89.8146>
  }]
 dyn_dread_front_1_2_r [{
   shapeType=Capsule
   length=5.4991
   radius=1.7346
   shapeName=dyn_dread_front_1_2_r_capsule
   center=<3.3303 0.3404 -0.1168>
   rotation=<rot -45.7426 -100.6 -81.9732>
  }]
 dyn_dread_front_1_3_l [{
   shapeType=Capsule
   length=5.0669
   radius=1.25
   shapeName=dyn_dread_front_1_3_l_capsule
   center=<4.2422 -1.1374 0.1229>
   rotation=<rot 0 71.9422 -88.4353>
  }]
 dyn_dread_front_1_3_r [{
   shapeType=Capsule
   length=7.6764
   radius=1.7353
   shapeName=dyn_dread_front_1_3_r_capsule
   center=<4.9099 -1.5368 1.2736>
   rotation=<rot 0 70.7377 -77.1987>
  }]
 dyn_dread_front_1_base_l [{shapeType=Capsule, length=0.505, radius=1.505, shapeName=dyn_dread_front_1_base_l_capsule, rotation=<rot 0 -90 0>}]
 dyn_dread_front_1_base_r [{shapeType=Capsule, length=0, radius=1.5, shapeName=dyn_dread_front_1_base_r_capsule, rotation=<rot 0 -90 0>}]
 head [{
   shapeType=Capsule
   length=4.5794
   radius=6.2844
   shapeName=head_capsule
   center=<2.4174 -0.6156 0.1172>
   rotation=<rot -0.5497 -90.0007 -87.4122>
  }]
 lowerarm_l [{
   shapeType=Capsule
   length=17.505
   radius=4.005
   shapeName=lowerarm_l_capsule
   center=<11.8676 2.6616 0.0875>
   rotation=<rot 0.4311 -80.0179 90.0562>
  }]
 lowerarm_r [{
   shapeType=Capsule
   length=17.505
   radius=4.005
   shapeName=lowerarm_r_capsule
   center=<-11.9659 -2.1502 0.105>
   rotation=<rot 0.4311 -80.0179 90.0561>
  }]
 neck_01 [{shapeType=Capsule, length=5.5, radius=4.5, shapeName=neck_01_capsule, rotation=<rot 0 -90.0002 90>}]
 spine_04 [{shapeType=Capsule, length=7.75, radius=11, shapeName=spine_04_capsule, center=<-1.106 1.4065 0>, rotation=<rot 0 -90 0>}]
 spine_05 [{shapeType=Capsule, length=14.75, radius=7.25, shapeName=spine_05_capsule, center=<1.7815 -0.6633 0>, rotation=<rot 0 -90 0>}]
 upperarm_l [{shapeType=Capsule, length=18.505, radius=4.005, shapeName=upperarm_l_capsule, center=<14.0596 -0.7318 0>, rotation=<rot -0.0003 -94.9997 90>}]
 upperarm_r [{shapeType=Capsule, length=19.005, radius=3.505, shapeName=upperarm_r_capsule, center=<-13.5909 1.1482 0>, rotation=<rot -0.0004 -89.9996 90>}]
}
constraints [
 {bone1Name=dyn_dread_front_1_1_r, bone2Name=dyn_dread_front_1_base_r, swing1LimitDegrees=30, swing2LimitDegrees=30, twistLimitDegrees=20}
 {bone1Name=dyn_dread_front_1_2_r, bone2Name=dyn_dread_front_1_1_r, swing1LimitDegrees=30, swing2LimitDegrees=30, twistLimitDegrees=15}
 {bone1Name=dyn_dread_front_1_3_r, bone2Name=dyn_dread_front_1_2_r, swing1LimitDegrees=15, swing2LimitDegrees=15, twistLimitDegrees=10}
 {bone1Name=dyn_dread_front_1_1_l, bone2Name=dyn_dread_front_1_base_l, swing1LimitDegrees=30, swing2LimitDegrees=20, twistLimitDegrees=10}
 {bone1Name=dyn_dread_front_1_3_l, bone2Name=dyn_dread_front_1_2_l, swing1LimitDegrees=30, swing2LimitDegrees=15, twistLimitDegrees=10}
 {bone1Name=dyn_dread_back_3_1, bone2Name=dyn_dread_back_3_base, swing1LimitDegrees=60, swing2LimitDegrees=20, twistLimitDegrees=20}
 {bone1Name=dyn_dread_back_3_3, bone2Name=dyn_dread_back_3_2, swing1LimitDegrees=30, swing2LimitDegrees=15, twistLimitDegrees=20}
 {bone1Name=dyn_dread_back_3_5, bone2Name=dyn_dread_back_3_4, swing1LimitDegrees=30, swing2LimitDegrees=20, twistLimitDegrees=30}
 {bone1Name=dyn_dread_back_2_1, bone2Name=dyn_dread_back_2_base, swing1LimitDegrees=45, swing2LimitDegrees=30, twistLimitDegrees=15}
 {bone1Name=dyn_dread_back_2_6, bone2Name=dyn_dread_back_2_5, swing1LimitDegrees=30, swing2LimitDegrees=20, twistLimitDegrees=45}
 {bone1Name=dyn_dread_back_1_1, bone2Name=dyn_dread_back_1_base, swing1LimitDegrees=45, swing2LimitDegrees=20, twistLimitDegrees=30}
 {bone1Name=dyn_dread_back_1_4, bone2Name=dyn_dread_back_1_3, swing1LimitDegrees=30, swing2LimitDegrees=20, twistLimitDegrees=20}
 {bone1Name=dyn_dread_back_1_5, bone2Name=dyn_dread_back_1_3, swing1LimitDegrees=30, swing2LimitDegrees=20, twistLimitDegrees=30}
 {bone1Name=dyn_dread_back_1_6, bone2Name=dyn_dread_back_1_5, swing1LimitDegrees=30, swing2LimitDegrees=10, twistLimitDegrees=30}
 {bone1Name=dyn_dread_back_3_2, bone2Name=dyn_dread_back_3_1, swing1LimitDegrees=45, swing2LimitDegrees=20, twistLimitDegrees=30}
 {bone1Name=dyn_dread_back_3_4, bone2Name=dyn_dread_back_3_1, swing1LimitDegrees=45, swing2LimitDegrees=20, twistLimitDegrees=30}
 {bone1Name=dyn_dread_back_1_3, bone2Name=dyn_dread_back_1_1, swing1LimitDegrees=45, swing2LimitDegrees=30, twistLimitDegrees=30}
 {
  bone1Name=dyn_dread_back_3_1
  bone2Name=dyn_dread_back_2_1
  swing1LimitDegrees=45
  swing1Motion=Free
  swing2LimitDegrees=45
  swing2Motion=Free
  twistLimitDegrees=45
  twistMotion=Free
 }
 {
  bone1Name=dyn_dread_back_1_1
  bone2Name=dyn_dread_back_2_1
  swing1LimitDegrees=45
  swing1Motion=Free
  swing2LimitDegrees=45
  swing2Motion=Free
  twistLimitDegrees=90
  twistMotion=Free
 }
 {bone1Name=dyn_dread_back_2_2, bone2Name=dyn_dread_back_2_1, swing1LimitDegrees=30, swing2LimitDegrees=30, twistLimitDegrees=30}
 {bone1Name=dyn_dread_back_2_3, bone2Name=dyn_dread_back_2_2, swing1LimitDegrees=20, swing2LimitDegrees=20, twistLimitDegrees=20}
 {bone1Name=dyn_dread_back_2_4, bone2Name=dyn_dread_back_2_2, swing1LimitDegrees=30, swing2LimitDegrees=20, twistLimitDegrees=45}
 {bone1Name=dyn_dread_back_2_5, bone2Name=dyn_dread_back_2_2, swing1LimitDegrees=30, swing2LimitDegrees=30, twistLimitDegrees=45}
 {
  bone1Name=dyn_dread_back_3_4
  bone2Name=dyn_dread_back_2_2
  swing1LimitDegrees=45
  swing1Motion=Free
  swing2LimitDegrees=45
  swing2Motion=Free
  twistLimitDegrees=45
  twistMotion=Free
 }
 {
  bone1Name=dyn_dread_back_1_3
  bone2Name=dyn_dread_back_2_2
  swing1LimitDegrees=30
  swing1Motion=Free
  swing2LimitDegrees=30
  swing2Motion=Free
  twistLimitDegrees=20
  twistMotion=Free
 }
 {
  bone1Name=dyn_dread_back_3_4
  bone2Name=dyn_dread_back_3_2
  swing1LimitDegrees=15
  swing1Motion=Free
  swing2LimitDegrees=15
  swing2Motion=Free
  twistLimitDegrees=30
  twistMotion=Free
 }
 {
  bone1Name=dyn_dread_back_3_5
  bone2Name=dyn_dread_back_2_4
  swing1LimitDegrees=45
  swing1Motion=Free
  swing2LimitDegrees=45
  swing2Motion=Free
  twistLimitDegrees=45
  twistMotion=Free
 }
 {bone1Name=dyn_dread_front_1_2_l, bone2Name=dyn_dread_front_1_1_l, swing1LimitDegrees=30, swing2LimitDegrees=20, twistLimitDegrees=10}
 {
  bone1Name=dyn_dread_back_1_5
  bone2Name=dyn_dread_back_2_3
  swing1LimitDegrees=20
  swing1Motion=Free
  swing2LimitDegrees=20
  swing2Motion=Free
  twistLimitDegrees=45
  twistMotion=Free
 }
 {
  bone1Name=dyn_dread_back_3_5
  bone2Name=dyn_dread_back_3_3
  swing1LimitDegrees=45
  swing1Motion=Free
  swing2LimitDegrees=45
  swing2Motion=Free
  twistLimitDegrees=45
  twistMotion=Free
 }
 {
  bone1Name=dyn_dread_back_1_4
  bone2Name=dyn_dread_back_1_5
  swing1LimitDegrees=45
  swing1Motion=Free
  swing2LimitDegrees=45
  swing2Motion=Free
  twistLimitDegrees=45
  twistMotion=Free
 }
]
