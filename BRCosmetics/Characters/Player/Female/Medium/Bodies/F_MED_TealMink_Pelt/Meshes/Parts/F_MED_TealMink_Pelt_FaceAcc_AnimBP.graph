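# Post-processing AnimBlueprint for the TealMink Pelt female FaceAcc cosmetic part, adding secondary motion including back dreads.
# Copies the attached parent's pose through a Control Rig, then a RigidBody simulation in spine_05 space.
# Position constraints then half-blend back dread bones toward spine_04/05 when a backbling is detected.
# EventGraph converts crouch, skydive, parachute, emote, front-end and rigid-body states into rig inputs, and computes gravity via a deprecated override function.
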AnimBlueprint /BRCosmetics/Characters/Player/Female/Medium/Bodies/F_MED_TealMink_Pelt/Meshes/Parts/F_MED_TealMink_Pelt_FaceAcc_AnimBP
target_skeleton=&/F_MED_TealMink_Pelt_FaceAcc_Skeleton
graphs [
 AnimGraph {nodes [
   n1 = Control Rig!(
    Source=Copy Pose From Mesh() {bUseAttachedParent=true}
    Alpha=Get Dyn_Alpha_Control_rig() -> (Dyn_Alpha_Control_rig:float)
    Dyn_crouch:double=Get Dyn_Crouch() -> (Dyn_Crouch:double)
    Dyn_fly:double=Get Dyn_Fly() -> (Dyn_Fly:double)
    Dyn_Emote:double=Get Dyn_Emote() -> (Dyn_Emote:double)
    Dyn_Parachute:double=Get Dyn_Parachute() -> (Dyn_Parachute:double)
    Dyn_FrontEnd:double=Get Dyn_Front_End() -> (Dyn_Front_End:double)
   ) -> (Pose) {
    alphaBoolBlend {blendOption=Linear}
    controlRigAssetReference {blueprintRigClass=&/F_MED_TealMink_Pelt_FaceAcc_CtrlRig.F_MED_TealMink_Pelt_FaceAcc_CtrlRig_C}
   }
   n2 = RigidBody!(
    ComponentPose=Local To Component(LocalPose=@n1)
    Alpha
    OverrideWorldGravity:Vector=Get Dyn_Override_World_Gravity() -> (Dyn_Override_World_Gravity:Vector)
    ExternalForce
   ) -> (Pose) {
    bOverrideWorldGravity=true
    overridePhysicsAsset=&/F_MED_TealMink_Pelt_FaceAcc_Physics
    simulationSpace=BaseBoneSpace
    alphaBoolBlend {blendOption=Linear}
    alphaScaleBiasClamp {bInterpResult=true}
    baseBoneRef {boneName=spine_05}
    componentAppliedLinearAccClamp=<1350 600 600>
    simSpaceSettings {maxAngularAcceleration=50, maxAngularVelocity=10, maxLinearAcceleration=1000, maxLinearVelocity=200, worldAlpha=0.66}
   }
   n3 = Get Dyn_Alpha_constr() -> (Dyn_Alpha_constr:float)
   n4 = "Constraint - dyn_dread_back_1_3 (P)"(ComponentPose=@n2, Alpha=@n3, ConstraintWeights_0=1) {
    alphaBoolBlend {blendOption=Linear}
    boneToModify {boneName=dyn_dread_back_1_3}
    constraintSetup [{targetBone {boneName=spine_05}}]
    constraintWeights [1]
   }
   n5 = "Constraint - dyn_dread_back_2_2 (P)"(ComponentPose=@n4, Alpha=@n3, ConstraintWeights_0=1) {
    alphaBoolBlend {blendOption=Linear}
    boneToModify {boneName=dyn_dread_back_2_2}
    constraintSetup [{targetBone {boneName=spine_05}}]
    constraintWeights [1]
   }
   n6 = "Constraint - dyn_dread_back_3_2 (P)"(ComponentPose=@n5, Alpha=@n3, ConstraintWeights_0=1) {
    alphaBoolBlend {blendOption=Linear}
    boneToModify {boneName=dyn_dread_back_3_2}
    constraintSetup [{targetBone {boneName=spine_05}}]
    constraintWeights [1]
   }
   n7 = "Constraint - dyn_dread_back_3_4 (P)"(ComponentPose=@n6, Alpha=@n3, ConstraintWeights_0=1) {
    alphaBoolBlend {blendOption=Linear}
    boneToModify {boneName=dyn_dread_back_3_4}
    constraintSetup [{targetBone {boneName=spine_05}}]
    constraintWeights [1]
   }
   n8 = "Constraint - dyn_dread_back_1_4 (P)"(ComponentPose=@n7, Alpha=@n3, ConstraintWeights_0=1) {
    alphaBoolBlend {blendOption=Linear}
    boneToModify {boneName=dyn_dread_back_1_4}
    constraintSetup [{targetBone {boneName=spine_04}}]
    constraintWeights [1]
   }
   n9 = "Constraint - dyn_dread_back_1_5 (P)"(ComponentPose=@n8, Alpha=@n3, ConstraintWeights_0=1) {
    alphaBoolBlend {blendOption=Linear}
    boneToModify {boneName=dyn_dread_back_1_5}
    constraintSetup [{targetBone {boneName=spine_04}}]
    constraintWeights [1]
   }
   n10 = "Constraint - dyn_dread_back_2_5 (P)"(ComponentPose=@n9, Alpha=@n3, ConstraintWeights_0=1) {
    alphaBoolBlend {blendOption=Linear}
    boneToModify {boneName=dyn_dread_back_2_5}
    constraintSetup [{targetBone {boneName=spine_04}}]
    constraintWeights [1]
   }
   n11 = "Constraint - dyn_dread_back_2_3 (P)"(ComponentPose=@n10, Alpha=@n3, ConstraintWeights_0=1) {
    alphaBoolBlend {blendOption=Linear}
    boneToModify {boneName=dyn_dread_back_2_3}
    constraintSetup [{targetBone {boneName=spine_04}}]
    constraintWeights [1]
   }
   n12 = "Constraint - dyn_dread_back_2_4 (P)"(ComponentPose=@n11, Alpha=@n3, ConstraintWeights_0=1) {
    alphaBoolBlend {blendOption=Linear}
    boneToModify {boneName=dyn_dread_back_2_4}
    constraintSetup [{targetBone {boneName=spine_04}}]
    constraintWeights [1]
   }
   n13 = "Constraint - dyn_dread_back_3_5 (P)"(ComponentPose=@n12, Alpha=@n3, ConstraintWeights_0=1) {
    alphaBoolBlend {blendOption=Linear}
    boneToModify {boneName=dyn_dread_back_3_5}
    constraintSetup [{targetBone {boneName=spine_05}}]
    constraintWeights [1]
   }
   n14 = "Constraint - dyn_dread_back_2_6 (P)"(ComponentPose=@n13, Alpha=@n3, ConstraintWeights_0=1) {
    alphaBoolBlend {blendOption=Linear}
    boneToModify {boneName=dyn_dread_back_2_6}
    constraintSetup [{targetBone {boneName=spine_04}}]
    constraintWeights [1]
   }
   n15 = "Constraint - dyn_dread_back_3_3 (P)"(ComponentPose=@n14, Alpha=@n3, ConstraintWeights_0=1) {
    alphaBoolBlend {blendOption=Linear}
    boneToModify {boneName=dyn_dread_back_3_3}
    constraintSetup [{targetBone {boneName=spine_04}}]
    constraintWeights [1]
   }
   Output Pose(Result=Component To Local(ComponentPose=@n15)) {layerGroup=None}
   Get Alpha() -> (Alpha:float)
  ]}
 EventGraph {nodes [
   Event BlueprintUpdateAnimation()
   > n1 = Sequence!(execute) -> (then_0, then_1, then_2:Exec, then_3:Exec, then_4:Exec, then_5:Exec, then_6:Exec)
   > n2 = Gravity Override State Function DEPRECATED(CharacterPartAnimInstance=Self-Reference(), GameTimeInSeconds=GetGameTimeInSeconds(), GravityOverrideParams=Get Gravity Override Params())
   > Set Dyn_Override_World_Gravity(Dyn_Override_World_Gravity:Vector=@n2) -> (Output_Get:Vector)
   n3 = Branch(execute=@n1.then_1, Condition=Get bIsCrouching())
   > Set Dyn_Crouch(Dyn_Crouch:double=1) -> (Output_Get:double)
   Set Dyn_Crouch(execute=@n3.else, Dyn_Crouch=0)
   n4 = Branch(execute=@n1.then_2, Condition=Get bIsSkyDiving())
   > Set Dyn_Fly(Dyn_Fly:double=1) -> (Output_Get:double)
   > n5 = Branch(Condition=Get bIsParachuteOpened())
   > Set Dyn_Parachute(Dyn_Parachute:double=1) -> (Output_Get:double)
   Set Dyn_Parachute(execute=@n5.else, Dyn_Parachute=0)
   Set Dyn_Fly(execute=@n4.else, Dyn_Fly=0)
   n6 = Branch(execute=@n1.then_3, Condition=Get bIsPlayingEmote())
   > Set Dyn_Emote(Dyn_Emote:double=1) -> (Output_Get:double)
   Set Dyn_Emote(execute=@n6.else, Dyn_Emote=0)
   n7 = Branch(execute=@n1.then_4, Condition=Get bRigidBodyEnabled())
   > Set Dyn_Alpha_Control_rig(Dyn_Alpha_Control_rig:float=1) -> (Output_Get:float)
   Set Dyn_Alpha_Control_rig(execute=@n7.else, Dyn_Alpha_Control_rig=0)
   n8 = Branch(execute=@n1.then_5, Condition=CheckForBackbling())
   > Set Dyn_Alpha_constr(Dyn_Alpha_constr:float=0.5) -> (Output_Get:float)
   Set Dyn_Alpha_constr(execute=@n8.else, Dyn_Alpha_constr=0)
   n9 = Branch(execute=@n1.then_6, Condition=Get bIsInFrontEnd())
   > Set Dyn_Front_End(Dyn_Front_End:double=1) -> (Output_Get:double)
   Set Dyn_Front_End(execute=@n9.else, Dyn_Front_End=0)
  ]}
]
variables [
 Alpha {type=number, default=1}
 Dyn_Alpha_constr {type=number}
 Dyn_Alpha_Control_rig {type=number, default=1}
 Dyn_Crouch {type=number}
 Dyn_Emote {type=number}
 Dyn_Fly {type=number}
 Dyn_Front_End {type=number}
 Dyn_Override_World_Gravity {type=Vector}
 Dyn_Parachute {type=number}
 Gravity Override Params {
  type=GravityOverrideParamsStruct
  default {
   skydive_FloatingDown {gravityOverride=<-500 -500 50>, windAmplitude=<200 -400 300>, windFrequency=<1.5 4 5>}
   skydive_Parachuting {windFrequency=<1.5 5.5 8>}
   tactical_Sprint {gravityOverride=<-120 -120 -1250>}
  }
 }
]
